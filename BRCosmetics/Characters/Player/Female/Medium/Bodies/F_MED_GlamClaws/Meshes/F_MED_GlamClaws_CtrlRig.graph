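# Control Rig for the F_MED_GlamClaws character body, adding deformation fixes and corrective posing for its belts.
# After the shared DeformRig function, weighted parent constraints bind main and right belt bones to spine, pelvis and leg bones.
# Thigh swing, separated from twist and remapped to weights, blends in corrective belt offsets.
# Left thigh angles drive extra left and right belt fixes, including one gated by On Grind Rail.
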ControlRigBlueprint /BRCosmetics/Characters/Player/Female/Medium/Bodies/F_MED_GlamClaws/Meshes/F_MED_GlamClaws_CtrlRig
graphs [
 CollapseNode_ContainedGraph {nodes [
   n1 = RigVMFunctionEntryNode()
   RigVMFunction_Sequence = &RigVMFunction_Sequence(ExecuteContext=@n1.ExecuteContext)
   RigVMFunction_MathQuaternionSwingTwist = &RigVMFunction_MathQuaternionSwingTwist(
    TwistAxis=(X=1,Y=0,Z=0)
    Input=&RigUnit_GetRelativeTransformForItem(Child=(Type=Bone,Name="thigh_l"), Parent=(Type=Null,Name="thigh_l_null"), bChildInitial=False, bParentInitial=False).RelativeTransform.Rotation
   )
   RigVMFunction_MathQuaternionToEuler = &RigVMFunction_MathQuaternionToEuler(RotationOrder=ZYX, Value=@RigVMFunction_MathQuaternionSwingTwist.Swing)
   RigVMVariableNode(ExecuteContext=@RigVMFunction_Sequence.A, Value=@RigVMFunction_MathQuaternionToEuler.Result.Y) {variable=l thighLocal ry float}
   > RigVMVariableNode(Value=@RigVMFunction_MathQuaternionToEuler.Result.Z) {variable=l thighLocal rz float}
   n2 = &RigVMFunction_MathQuaternionToEuler(
    RotationOrder=ZYX
    Value=&RigVMFunction_MathQuaternionSwingTwist(
     TwistAxis=(X=1,Y=0,Z=0)
     Input=&RigUnit_GetRelativeTransformForItem(Child=(Type=Bone,Name="thigh_r"), Parent=(Type=Null,Name="thigh_r_null"), bChildInitial=False, bParentInitial=False).RelativeTransform.Rotation
    ).Swing
   )
   > RigVMVariableNode(Value=@n2.Result.Y) {variable=r thighLocal ry float}
   > RigVMVariableNode(Value=@n2.Result.Z) {variable=r thighLocal rz float}
   RigVMFunctionReturnNode(ExecuteContext=@RigVMFunction_Sequence.B)
  ]}
 CollapseNode_ContainedGraph {nodes [
   RigVMFunctionEntryNode()
   > RigVMFunction_Sequence = RigVMAggregateNode()
   &RigUnit_ParentConstraint(
    AdvancedSettings=(RotationOrderForFilter=XZY)
    Child=(Type=Bone,Name="mainbelt_front")
    Parents="((Item=(Type=Bone,Name=\"spine_03\"),Weight=0.0019),(Item=(Type=Bone,Name=\"thigh_twist_01_r\"),Weight=0.0068),(Item=(Type=Bone,Name=\"deform_groin_l\"),Weight=0.0113),(Item=(Type=Bone,Name=\"thigh_twist_01_l\"),Weight=0.0125),(Item=(Type=Bone,Name=\"spine_02\"),Wei ...(truncated)"
    Weight=1
    ExecutePin=@RigVMFunction_Sequence.A
   )
   > &RigUnit_ParentConstraint(
    AdvancedSettings=(RotationOrderForFilter=XZY)
    Child=(Type=Bone,Name="mainbelt_L_01")
    Parents="((Item=(Type=Bone,Name=\"spine_03\"),Weight=0.0004),(Item=(Type=Bone,Name=\"thigh_twist_01_l\"),Weight=0.0005),(Item=(Type=Bone,Name=\"spine_02\"),Weight=0.0234),(Item=(Type=Bone,Name=\"thigh_twist_01_r\"),Weight=0.0823),(Item=(Type=Bone,Name=\"deform_groin_r\"),Wei ...(truncated)"
    Weight=1
   )
   > &RigUnit_ParentConstraint(
    AdvancedSettings=(RotationOrderForFilter=XZY)
    Child=(Type=Bone,Name="mainbelt_L_02")
    Parents=((Item=(Type=Bone,Name="spine_03"),Weight=0.0044),(Item=(Type=Bone,Name="thigh_twist_01_r"),Weight=0.033),(Item=(Type=Bone,Name="spine_02"),Weight=0.1188),(Item=(Type=Bone,Name="pelvis"),Weight=0.3454),(Item=(Type=Bone,Name="spine_01"),Weight=0.4983))
    Weight=1
   )
   > &RigUnit_ParentConstraint(
    AdvancedSettings=(RotationOrderForFilter=XZY)
    Child=(Type=Bone,Name="mainbelt_back")
    Parents="((Item=(Type=Bone,Name=\"thigh_twist_01_l\"),Weight=0.0003),(Item=(Type=Bone,Name=\"spine_03\"),Weight=0.0006),(Item=(Type=Bone,Name=\"thigh_twist_01_r\"),Weight=0.0015),(Item=(Type=Bone,Name=\"deform_glute_r\"),Weight=0.0083),(Item=(Type=Bone,Name=\"spine_02\"),Wei ...(truncated)"
    Weight=1
   )
   > &RigUnit_ParentConstraint(
    AdvancedSettings=(RotationOrderForFilter=XZY)
    Child=(Type=Bone,Name="mainbelt_R_01")
    Parents=((Item=(Type=Bone,Name="spine_03"),Weight=0.0044),(Item=(Type=Bone,Name="thigh_twist_01_l"),Weight=0.033),(Item=(Type=Bone,Name="spine_02"),Weight=0.1188),(Item=(Type=Bone,Name="pelvis"),Weight=0.3453),(Item=(Type=Bone,Name="spine_01"),Weight=0.4984))
    Weight=1
   )
   > &RigUnit_ParentConstraint(
    AdvancedSettings=(RotationOrderForFilter=XZY)
    Child=(Type=Bone,Name="mainbelt_R_02")
    Parents=((Item=(Type=Bone,Name="spine_04"),Weight=0.0005),(Item=(Type=Bone,Name="spine_03"),Weight=0.024),(Item=(Type=Bone,Name="pelvis"),Weight=0.1179),(Item=(Type=Bone,Name="spine_02"),Weight=0.3067),(Item=(Type=Bone,Name="spine_01"),Weight=0.5509))
    Weight=1
   )
   > &RigUnit_ParentConstraint(
    AdvancedSettings=(RotationOrderForFilter=XZY)
    Child=(Type=Bone,Name="mainbelt_R_03")
    Parents=((Item=(Type=Bone,Name="spine_03"),Weight=0.0038),(Item=(Type=Bone,Name="thigh_twist_01_r"),Weight=0.0068),(Item=(Type=Bone,Name="spine_02"),Weight=0.134),(Item=(Type=Bone,Name="pelvis"),Weight=0.3038),(Item=(Type=Bone,Name="spine_01"),Weight=0.5517))
    Weight=1
   )
   &RigUnit_ParentConstraint(
    AdvancedSettings=(RotationOrderForFilter=XZY)
    Child=(Type=Bone,Name="rightbelt_01")
    Parents="((Item=(Type=Bone,Name=\"thigh_twist_01_l\"),Weight=0.0014),(Item=(Type=Bone,Name=\"spine_02\"),Weight=0.0062),(Item=(Type=Bone,Name=\"thigh_twist_01_r\"),Weight=0.012),(Item=(Type=Bone,Name=\"deform_groin_r\"),Weight=0.018),(Item=(Type=Bone,Name=\"spine_01\"),Weigh ...(truncated)"
    Weight=1
    ExecutePin=@RigVMFunction_Sequence.B
   )
   > &RigUnit_ParentConstraint(
    AdvancedSettings=(RotationOrderForFilter=XZY)
    Child=(Type=Bone,Name="rightbelt_02")
    Parents="((Item=(Type=Bone,Name=\"spine_01\"),Weight=0.001),(Item=(Type=Bone,Name=\"thigh_twist_01_l\"),Weight=0.0015),(Item=(Type=Bone,Name=\"thigh_r\"),Weight=0.0281),(Item=(Type=Bone,Name=\"pelvis\"),Weight=0.2142),(Item=(Type=Bone,Name=\"deform_groin_r\"),Weight=0.22),(I ...(truncated)"
    Weight=1
   )
   > &RigUnit_ParentConstraint(
    AdvancedSettings=(RotationOrderForFilter=XZY)
    Child=(Type=Bone,Name="rightbelt_03")
    Parents="((Item=(Type=Bone,Name=\"deform_glute_r\"),Weight=0.0007),(Item=(Type=Bone,Name=\"deform_groin_r\"),Weight=0.0416),(Item=(Type=Bone,Name=\"pelvis\"),Weight=0.0973),(Item=(Type=Bone,Name=\"thigh_r\"),Weight=0.1046),(Item=(Type=Bone,Name=\"thigh_twist_01_r\"),Weight=0 ...(truncated)"
    Weight=1
   )
   > &RigUnit_ParentConstraint(
    AdvancedSettings=(RotationOrderForFilter=XZY)
    Child=(Type=Bone,Name="rightbelt_04")
    Parents="((Item=(Type=Bone,Name=\"mainbelt_L_02\"),Weight=0.0006),(Item=(Type=Bone,Name=\"spine_01\"),Weight=0.0008),(Item=(Type=Bone,Name=\"deform_groin_r\"),Weight=0.0211),(Item=(Type=Bone,Name=\"thigh_r\"),Weight=0.034),(Item=(Type=Bone,Name=\"deform_glute_r\"),Weight=0.0 ...(truncated)"
    Weight=1
   )
   > &RigUnit_ParentConstraint(
    AdvancedSettings=(RotationOrderForFilter=XZY)
    Child=(Type=Bone,Name="rightbelt_05")
    Parents="((Item=(Type=Bone,Name=\"mainbelt_back\"),Weight=0.0037),(Item=(Type=Bone,Name=\"mainbelt_L_02\"),Weight=0.0055),(Item=(Type=Bone,Name=\"spine_02\"),Weight=0.0104),(Item=(Type=Bone,Name=\"deform_glute_r\"),Weight=0.0459),(Item=(Type=Bone,Name=\"thigh_twist_01_r\"),W ...(truncated)"
    Weight=1
   )
   &RigUnit_ModifyTransforms(
    ItemToModify="((Item=(Type=Bone,Name=\"rightbelt_04\"),Transform=(Rotation=(X=0,Y=0,Z=0,W=-1),Translation=(X=1.6547,Y=0,Z=0),Scale3D=(X=1,Y=1,Z=1))),(Item=(Type=Bone,Name=\"rightbelt_02\"),Transform=(Rotation=(X=0,Y=0,Z=0,W=-1),Translation=(X=0.7631,Y=0,Z=0),Scale3D=(X=1,Y= ...(truncated)"
    Weight=1
    WeightMinimum=0
    ExecutePin=@RigVMFunction_Sequence.D
   )
   RigVMFunctionReturnNode(ExecuteContext=@RigVMFunction_Sequence.E)
  ]}
 CollapseNode_ContainedGraph {nodes [
   RigVMFunctionEntryNode()
   > RigVMFunction_Sequence = RigVMAggregateNode()
   DeltaRotation v4 = RigVMFunctionReferenceNode(
    Rotation Order=ZYX
    child=(Type=Bone,Name="thigh_r")
    outputMinMax=(X=0,Y=1)
    parent=(Type=Null,Name="thigh_r_null")
    premult=1
    rotationMinMax=(X=-20,Y=-70)
    triggerAxis=Z
    twistAxis=(X=1,Y=0,Z=0)
    ExecuteContext=@RigVMFunction_Sequence.A
   )
   > &RigUnit_ModifyTransforms(
    ItemToModify="((Item=(Type=Bone,Name=\"rightbelt_02\"),Transform=(Rotation=(X=-0.0444,Y=-0.0801,Z=-0.1083,W=-0.9899),Translation=(X=1.9949,Y=2.3176,Z=1.2695),Scale3D=(X=1,Y=1,Z=1))),(Item=(Type=Bone,Name=\"rightbelt_03\"),Transform=(Rotation=(X=0.0592,Y=0.1673,Z=-0.2597,W=0 ...(truncated)"
    WeightMinimum=0
    Weight=@"DeltaRotation v4".outputWeight
   )
   > &RigUnit_ModifyTransforms(
    ItemToModify="((Item=(Type=Bone,Name=\"mainbelt_front\"),Transform=(Rotation=(X=0,Y=0,Z=0,W=1),Translation=(X=0,Y=-0.1405,Z=0),Scale3D=(X=1,Y=1,Z=1))),(Item=(Type=Bone,Name=\"mainbelt_L_01\"),Transform=(Rotation=(X=0.0016,Y=0,Z=0,W=-1),Translation=(X=1.0936,Y=-1.9309,Z=1.29 ...(truncated)"
    WeightMinimum=0
    Weight=@"DeltaRotation v4".outputWeight
   )
   DeltaRotation v4_1 = RigVMFunctionReferenceNode(
    Rotation Order=ZYX
    child=(Type=Bone,Name="thigh_l")
    outputMinMax=(X=0,Y=1)
    parent=(Type=Null,Name="thigh_l_null")
    premult=1
    rotationMinMax=(X=-30,Y=-70)
    triggerAxis=Z
    twistAxis=(X=1,Y=0,Z=0)
    ExecuteContext=@RigVMFunction_Sequence.B
   )
   > &RigUnit_ModifyTransforms(
    ItemToModify="((Item=(Type=Bone,Name=\"mainbelt_front\"),Transform=(Rotation=(X=0,Y=0,Z=0,W=1),Translation=(X=0,Y=-0.5154,Z=-0.1756),Scale3D=(X=1,Y=1,Z=1))),(Item=(Type=Bone,Name=\"mainbelt_back\"),Transform=(Rotation=(X=0,Y=0,Z=0,W=1),Translation=(X=0,Y=0,Z=0),Scale3D=(X=1 ...(truncated)"
    WeightMinimum=0
    Weight=@"DeltaRotation v4_1".outputWeight
   )
   RigVMFunctionReturnNode(ExecuteContext=@RigVMFunction_Sequence.C)
  ]}
 CollapseNode_ContainedGraph {nodes [
   RigVMFunctionEntryNode()
   > RigVMFunction_Sequence = RigVMAggregateNode()
   n1 = &RigVMFunction_MathDoubleRemap(SourceMaximum=-90, SourceMinimum=-5, TargetMaximum=1, TargetMinimum=0, Value=RigVMVariableNode() {variable=l thighLocal ry float}.Value)
   &RigUnit_ModifyTransforms(
    ItemToModify="((Item=(Type=Bone,Name=\"None\"),Transform=(Rotation=(X=0,Y=0,Z=0,W=1),Translation=(X=0,Y=0,Z=0),Scale3D=(X=1,Y=1,Z=1))),(Item=(Type=Bone,Name=\"base_leftBelt\"),Transform=(Rotation=(X=0,Y=0,Z=0.3023,W=-0.9532),Translation=(X=-2.1731,Y=2.1978,Z=2.4122),Scale3D ...(truncated)"
    WeightMinimum=0
    Weight=@n1.Result
    ExecutePin=@RigVMFunction_Sequence.B
   )
   > &RigUnit_ModifyTransforms(
    ItemToModify="((Item=(Type=Bone,Name=\"None\"),Transform=(Rotation=(X=0,Y=0,Z=0,W=1),Translation=(X=0,Y=0,Z=0),Scale3D=(X=1,Y=1,Z=1))),(Item=(Type=Bone,Name=\"dyn_leftBelt_side_01\"),Transform=(Rotation=(X=0,Y=0,Z=-0.1227,W=-0.9924),Translation=(X=0.7882,Y=0.1979,Z=0),Scale ...(truncated)"
    WeightMinimum=0
    Weight=@n1.Result
   )
   RigVMFunctionReturnNode(ExecuteContext=@RigVMFunction_Sequence.C)
   Remap = &RigVMFunction_MathDoubleRemap(SourceMaximum=-90, SourceMinimum=25, TargetMaximum=1, TargetMinimum=0, Value=RigVMVariableNode() {variable=l thighLocal rz float}.Value)
   &RigUnit_ModifyTransforms(
    ItemToModify="((Item=(Type=Bone,Name=\"base_leftBelt\"),Transform=(Rotation=(X=0,Y=0,Z=-0.108,W=0.9941),Translation=(X=-2.0469,Y=-1.1271,Z=-0.0001),Scale3D=(X=1,Y=1,Z=1))),(Item=(Type=Bone,Name=\"dyn_leftBelt_02\"),Transform=(Rotation=(X=0,Y=0,Z=-0.0358,W=-0.9994),Translati ...(truncated)"
    WeightMinimum=0
    Weight=@Remap.Result
    ExecutePin=@RigVMFunction_Sequence.A
   )
   > &RigUnit_ModifyTransforms(
    ItemToModify="((Item=(Type=Bone,Name=\"dyn_leftBelt_side_01\"),Transform=(Rotation=(X=0,Y=0,Z=0,W=1),Translation=(X=6.6431,Y=0,Z=-2.7173),Scale3D=(X=1,Y=1,Z=1))),(Item=(Type=Bone,Name=\"dyn_leftBelt_side_02\"),Transform=(Rotation=(X=0.1137,Y=0,Z=0,W=0.9935),Translation=(X=0 ...(truncated)"
    WeightMinimum=0
    Weight=@Remap.Result
   )
   > &RigUnit_ModifyTransforms(
    ItemToModify="((Item=(Type=Bone,Name=\"rightbelt_02\"),Transform=(Rotation=(X=0,Y=0,Z=0,W=-1),Translation=(X=1.3149,Y=-0.1623,Z=-5.0758),Scale3D=(X=1,Y=1,Z=1))),(Item=(Type=Bone,Name=\"rightbelt_01\"),Transform=(Rotation=(X=-0.1596,Y=0,Z=0,W=0.9872),Translation=(X=2.8392,Y= ...(truncated)"
    WeightMinimum=0
    Weight=@Remap.Result
   )
   Multiply = &RigVMFunction_MathDoubleMul(B=&RigVMFunction_MathBoolToFloat(Value=RigVMVariableNode() {variable=On Grind Rail}.Value).Result, A=@Remap.Result)
   > &RigUnit_ModifyTransforms(
    ItemToModify="((Item=(Type=Bone,Name=\"dyn_leftBelt_side_03\"),Transform=(Rotation=(X=-0.0193,Y=0.1113,Z=-0.1696,W=0.979),Translation=(X=1.9885,Y=-2.5127,Z=-3.8165),Scale3D=(X=1,Y=1,Z=1))),(Item=(Type=Bone,Name=\"dyn_leftBelt_side_05\"),Transform=(Rotation=(X=0,Y=0,Z=0,W=-1 ...(truncated)"
    WeightMinimum=0
    Weight=@Multiply.Result
   )
  ]}
 New Function_ContainedGraph {nodes [
   n1 = RigVMFunctionEntryNode()
   RigVMFunction_MathBoolToInteger = &RigVMFunction_MathBoolToInteger(Value=@n1.onlyTwist)
   RigVMFunction_MathQuaternionSwingTwist = &RigVMFunction_MathQuaternionSwingTwist(
    TwistAxis=@n1.twistAxis
    Input=&RigUnit_GetRelativeTransformForItem(bChildInitial=False, bParentInitial=False, Child=@n1.child, Parent=@n1.parent).RelativeTransform.Rotation
   )
   n2 = RigVMDispatchNode(
    Values=((),())
    Index=@RigVMFunction_MathBoolToInteger.Result
    Values.0=@RigVMFunction_MathQuaternionSwingTwist.Swing
    Values.1=@RigVMFunction_MathQuaternionSwingTwist.Twist
   ) {template="DISPATCH_RigVMDispatch_SelectInt32(in Index,in Values,out Result)"}
   RigVMFunction_MathQuaternionToEuler = &RigVMFunction_MathQuaternionToEuler(Value=@n2.Result, RotationOrder=@n1.Rotation Order)
   > DISPATCH_RigVMDispatch_Print = RigVMDispatchNode(ScreenColor=(R=1,G=1,B=1,A=1), ScreenDuration=1.5, Enabled=@n1.debugRotation, Value=@RigVMFunction_MathQuaternionToEuler.Result) {template="DISPATCH_RigVMDispatch_Print(in Prefix,in Value,in Enabled,in ScreenDuration,in ScreenColor)"}
   Cast = RigVMDispatchNode(Value=@n1.triggerAxis) {template="DISPATCH_RigVMDispatch_CastEnumToInt(in Value,out Result)"}
   DISPATCH_RigVMDispatch_SelectInt32 = RigVMDispatchNode(
    Values=(0,0,0)
    Index=@Cast.Result
    Values.0=@RigVMFunction_MathQuaternionToEuler.Result.X
    Values.1=@RigVMFunction_MathQuaternionToEuler.Result.Y
    Values.2=@RigVMFunction_MathQuaternionToEuler.Result.Z
   ) {template="DISPATCH_RigVMDispatch_SelectInt32(in Index,in Values,out Result)"}
   Remap = &RigVMFunction_MathDoubleRemap(
    bClamp=False
    Value=@DISPATCH_RigVMDispatch_SelectInt32.Result
    TargetMinimum=@n1.outputMinMax.X
    TargetMaximum=@n1.outputMinMax.Y
    SourceMinimum=@n1.rotationMinMax.X
    SourceMaximum=@n1.rotationMinMax.Y
   )
   Absolute = &RigVMFunction_MathDoubleAbs(Value=@Remap.Result)
   Multiply = &RigVMFunction_MathDoubleMul(
    A=&RigVMFunction_MathDoubleClamp(Value=RigVMDispatchNode(True=@Absolute.Result, Condition=@n1.absoluteValue, False=@Remap.Result) {template="DISPATCH_RigVMDispatch_If(in Condition,in True,in False,out Result)"}.Result).Result
    B=@n1.premult
   )
   > RigVMFunctionReturnNode(outputQuat=@n2.Result, outputWeight=@Multiply.Result)
  ]}
 RigVMFunctionLibrary {nodes [DeltaRotation v4 = RigVMCollapseNode(
    ExecuteContext=()
    Rotation Order=ZYX
    child=(Type=Bone,Name="None")
    outputMinMax=(X=0,Y=1)
    parent=(Type=Bone,Name="None")
    premult=1
    rotationMinMax=(X=0,Y=0)
    twistAxis=(X=1,Y=0,Z=0)
   )]}
 RigVMFunction_Sequence_1_ContainedGraph {nodes [
   n1 = RigVMFunctionEntryNode()
   RigVMFunction_Sequence = &RigVMFunction_Sequence(ExecuteContext=@n1.ExecuteContext)
   n2 = &RigVMFunction_Sequence(ExecuteContext=@RigVMFunction_Sequence.B)
   n3 = &RigVMFunction_Sequence(ExecuteContext=@n2.B)
   RigVMFunctionReturnNode(A=@RigVMFunction_Sequence.A, B=@n2.A, D=@n3.A, E=@n3.B)
  ]}
 RigVMFunction_Sequence_1_ContainedGraph {nodes [
   n1 = RigVMFunctionEntryNode()
   RigVMFunction_Sequence = &RigVMFunction_Sequence(ExecuteContext=@n1.ExecuteContext)
   n2 = &RigVMFunction_Sequence(ExecuteContext=@RigVMFunction_Sequence.B)
   RigVMFunctionReturnNode(A=@RigVMFunction_Sequence.A, B=@n2.A, C=@n2.B)
  ]}
 RigVMFunction_Sequence_1_ContainedGraph {nodes [
   n1 = RigVMFunctionEntryNode()
   RigVMFunction_Sequence = &RigVMFunction_Sequence(ExecuteContext=@n1.ExecuteContext)
   n2 = &RigVMFunction_Sequence(ExecuteContext=@RigVMFunction_Sequence.B)
   RigVMFunctionReturnNode(A=@RigVMFunction_Sequence.A, B=@n2.A, C=@n2.B)
  ]}
 RigVMModel {nodes [
   RigUnit_BeginExecution = &RigUnit_BeginExecution()
   FN_DeformRig_FNC = RigVMFunctionReferenceNode(
    DeformRig_ Curve="(deformRig_l_upperarm_up_alpha_72_#1=False,deformRig_r_upperarm_up_alpha_73_#2=False,deformRig_l_upperarm_fr_alpha_74_#3=False,deformRig_r_upperarm_fr_alpha_75_CB51B8 ...(truncated)"
    DeformRig_Bool="(upperarm_1_#4=True,elbow_3_#5=True,wrist_7_#6=True,groin_8_#7=True,glute_17_#8=True,knee_9_40C408324CA261 ...(truncated)"
    DeformRig_Debug=(upperarm_debug_18_#4=False,wrist_debug_20_#6=False,lat_debug_26_#9=False,groinGlute_debug_27_#7=False)
    DeformRig_Value="(upperarm_up_18_#10=(Rotation=(X=0,Y=0,Z=0,W=1),Translation=(X=0,Y=0,Z=8),Scale3D=(X=0.7,Y=1,Z=1)),upperarm_fwd_fr_19_#11=(Rotation=(X=0,Y=0,Z=0,W=1),Translation=(X=0,Y=10,Z=0),Scale3D=(X=0.5,Y=1,Z= ...(truncated)"
    ExecuteContext=@RigUnit_BeginExecution.ExecutePin
   )
   > Variables = RigVMCollapseNode()
   > Belts_constrains = RigVMCollapseNode()
   > Belts_poses_fix = RigVMCollapseNode()
   > Left_belt_extra_fix = RigVMCollapseNode()
  ]}
]
hierarchy [
 root {loc=<0 0 0>, rot=<0 0 0 1>}
 pelvis {parent=root, loc=<0 -0.5301 98.0571>, rot=<0.0127 0.707 -0.0127 -0.707>}
 spine_01 {parent=pelvis, loc=<6.1208 -0.9328 0>, rot=<0 0 0.034 -0.9994>}
 spine_02 {parent=spine_01, loc=<6.5997 -0.2095 0>, rot=<0 0 0.0238 -0.9997>}
 spine_03 {parent=spine_02, loc=<6.4914 -0.1237 0>, rot=<0 0 0.0149 -0.9999>}
 spine_04 {parent=spine_03, loc=<6.9598 0.0794 0>, rot=<0 0 -0.0132 -0.9999>}
 spine_05 {parent=spine_04, loc=<6.9987 0 0>, rot=<0 0 -0.0015 -1>}
 neck_01 {parent=spine_05, loc=<11.3138 0.1529 0>, rot=<0 0 -0.2003 -0.9797>}
 neck_02 {parent=neck_01, loc=<4.5627 0 0>, rot=<0 0 0.0504 -0.9987>}
 head {parent=neck_02, loc=<5.3914 0 0>, rot=<0 0 0.115 -0.9934>}
 clavicle_l {parent=spine_05, loc=<5.4831 -0.1053 -3.2946>, rot=<0.0465 -0.7289 -0.009 -0.6829>}
 upperarm_l {parent=clavicle_l, loc=<9.8658 -1.2136 -0.4105>, rot=<0.0164 -0.3245 -0.0269 -0.9454>}
 lowerarm_l {parent=upperarm_l, loc=<26.465 -0.8345 0.259>, rot=<0.0036 0.0161 -0.2298 -0.9731>}
 hand_l {parent=lowerarm_l, loc=<22.7355 -0.3203 -0.6105>, rot=<-0.6784 -0.0223 0.0735 -0.7307>}
 thumb_01_l {parent=hand_l, loc=<1.3118 -0.3191 -2.459>, rot=<0.6546 -0.4042 0.0664 -0.6354>}
 thumb_02_l {parent=thumb_01_l, loc=<5.9336 -0.0448 0.0401>, rot=<-0.0037 0.0213 0.298 -0.9543>}
 thumb_03_l {parent=thumb_02_l, loc=<2.2223 0.3682 -0.0907>, rot=<0.0118 -0.013 0.0066 -0.9998>}
 index_metacarpal_l {parent=hand_l, loc=<1.6535 0.0183 -2.96>, rot=<0.0158 -0.0124 -0.0173 -0.9996>}
 index_01_l {parent=index_metacarpal_l, loc=<8.3008 0.0357 -0.0524>, rot=<-0.0044 0.0485 0.0924 -0.9945>}
 index_02_l {parent=index_01_l, loc=<3.3587 -0.3004 0.0452>, rot=<-0.002 0.0027 0.3225 -0.9466>}
 index_03_l {parent=index_02_l, loc=<2.5578 -0.123 -0.0252>, rot=<-0.002 0.0067 0.2558 -0.9667>}
 middle_metacarpal_l {parent=hand_l, loc=<2.1856 -0.1725 -0.5984>, rot=<0.006 0.0168 -0.0329 -0.9993>}
 middle_01_l {parent=middle_metacarpal_l, loc=<7.3275 -0.2447 0.0129>, rot=<-0.0208 0.0042 0.0858 -0.9961>}
 middle_02_l {parent=middle_01_l, loc=<4.0123 0.1401 0.1175>, rot=<-0.0218 0.0258 0.3512 -0.9357>}
 middle_03_l {parent=middle_02_l, loc=<2.6077 0.0534 0.1373>, rot=<0.0082 0.0188 0.1708 -0.9851>}
 ring_metacarpal_l {parent=hand_l, loc=<2.2074 -0.3625 1.3638>, rot=<-0.0284 0.0722 -0.0458 -0.9959>}
 ring_01_l {parent=ring_metacarpal_l, loc=<7.0295 -0.6197 -0.1446>, rot=<-0.0489 -0.0252 0.1892 -0.9804>}
 ring_02_l {parent=ring_01_l, loc=<3.6174 0.4647 -0.1027>, rot=<0 0 0.2806 -0.9598>}
 ring_03_l {parent=ring_02_l, loc=<2.4152 0.114 0.0199>, rot=<0 0 0.1189 -0.9929>}
 pinky_metacarpal_l {parent=hand_l, loc=<1.7616 -0.9753 3.3814>, rot=<-0.0397 0.0814 -0.069 -0.9935>}
 pinky_01_l {parent=pinky_metacarpal_l, loc=<7.1709 -0.4831 -0.1094>, rot=<-0.0654 -0.0035 0.2242 -0.9723>}
 pinky_02_l {parent=pinky_01_l, loc=<2.8274 0.0893 -0.1043>, rot=<0.0037 0.0011 0.2854 -0.9584>}
 pinky_03_l {parent=pinky_02_l, loc=<2.2046 0.0307 -0.0078>, rot=<-0.0002 0.0098 0.1445 -0.9895>}
 deform_wrist_up_base_l {parent=hand_l, loc=<0 2.3009 0>, rot=<0 0 0 -1>}
 deform_wrist_up_l {parent=deform_wrist_up_base_l, loc=<0 0 0>, rot=<0 0 0 1>}
 lowerarm_twist_02_l {parent=lowerarm_l, loc=<10.6707 0 0>, rot=<0 0 0 1>}
 lowerarm_twist_01_l {parent=lowerarm_l, loc=<18.316 0 0>, rot=<0 0 0 1>}
 upperarm_twist_01_l {parent=upperarm_l, loc=<7.2135 -0.1134 0>, rot=<0 0 0 1>}
 deform_upperarm_bk_base_l {parent=upperarm_twist_01_l, loc=<-6.4053 -4.3746 0.314>, rot=<0 0 0 -1>}
 deform_upperarm_bk_l {parent=deform_upperarm_bk_base_l, loc=<0 0 0>, rot=<0 0 0 1>}
 deform_upperarm_fr_base_l {parent=upperarm_twist_01_l, loc=<-5.6312 5.092 0.3354>, rot=<0 0 0 -1>}
 deform_upperarm_fr_l {parent=deform_upperarm_fr_base_l, loc=<0 0 0>, rot=<0 0 0 1>}
 deform_upperarm_up_base_l {parent=upperarm_twist_01_l, loc=<-8.891 0.4825 3.9517>, rot=<0 0 0 -1>}
 deform_upperarm_up_l {parent=deform_upperarm_up_base_l, loc=<0 0 0>, rot=<0 0 0 1>}
 base_TopSleeve_l {parent=upperarm_twist_01_l, loc=<1.4607 -0.765 2.4536>, rot=<0.7004 0.0186 0.0282 -0.713>}
 dyn_TopSleeve_l {parent=base_TopSleeve_l, loc=<0 0 0>, rot=<0 0 0 1>}
 base_BotSleeve_l {parent=upperarm_twist_01_l, loc=<3.941 1.2487 -4.2996>, rot=<-0.814 0.0222 -0.0191 -0.5801>}
 dyn_BotSleeve_l {parent=base_BotSleeve_l, loc=<0 0 0>, rot=<0 0 0 1>}
 base_RightSleeve_l {parent=upperarm_twist_01_l, loc=<2.5649 4.2693 -0.975>, rot=<0.9976 0.0279 0.0607 0.0179>}
 dyn_RightSleeve_l {parent=base_RightSleeve_l, loc=<0 0 0>, rot=<0 0 0 1>}
 base_LeftSleeve_l {parent=upperarm_twist_01_l, loc=<3.0514 -4.7957 -0.6197>, rot=<0.0179 0.0607 -0.0279 -0.9976>}
 dyn_LeftSleeve_l {parent=base_LeftSleeve_l, loc=<0 0 0>, rot=<0 0 0 1>}
 upperarm_twist_02_l {parent=upperarm_l, loc=<19.7998 -0.1134 0>, rot=<0 0 0 1>}
 clavicle_r {parent=spine_05, loc=<5.4831 -0.1053 3.2946>, rot=<0.7289 0.0465 -0.6829 0.009>}
 upperarm_r {parent=clavicle_r, loc=<-9.8658 1.2136 0.4105>, rot=<0.0164 -0.3245 -0.0269 -0.9454>}
 lowerarm_r {parent=upperarm_r, loc=<-26.465 0.8345 -0.259>, rot=<0.0036 0.0161 -0.2298 -0.9731>}
 hand_r {parent=lowerarm_r, loc=<-22.7355 0.3203 0.6106>, rot=<-0.6784 -0.0223 0.0735 -0.7307>}
 thumb_01_r {parent=hand_r, loc=<-1.3118 0.3191 2.459>, rot=<0.6546 -0.4042 0.0664 -0.6354>}
 thumb_02_r {parent=thumb_01_r, loc=<-5.9336 0.0448 -0.0401>, rot=<-0.0037 0.0213 0.298 -0.9543>}
 thumb_03_r {parent=thumb_02_r, loc=<-2.2223 -0.3682 0.0907>, rot=<0.0118 -0.013 0.0066 -0.9998>}
 index_metacarpal_r {parent=hand_r, loc=<-1.6535 -0.0183 2.96>, rot=<0.0158 -0.0124 -0.0173 -0.9996>}
 index_01_r {parent=index_metacarpal_r, loc=<-8.3008 -0.0357 0.0524>, rot=<-0.0044 0.0485 0.0924 -0.9945>}
 index_02_r {parent=index_01_r, loc=<-3.3587 0.3004 -0.0452>, rot=<-0.002 0.0027 0.3225 -0.9466>}
 index_03_r {parent=index_02_r, loc=<-2.5578 0.123 0.0252>, rot=<-0.002 0.0067 0.2558 -0.9667>}
 middle_metacarpal_r {parent=hand_r, loc=<-2.1856 0.1725 0.5984>, rot=<0.006 0.0168 -0.0329 -0.9993>}
 middle_01_r {parent=middle_metacarpal_r, loc=<-7.3275 0.2447 -0.0129>, rot=<-0.0208 0.0042 0.0858 -0.9961>}
 middle_02_r {parent=middle_01_r, loc=<-4.0123 -0.1401 -0.1175>, rot=<-0.0218 0.0258 0.3512 -0.9357>}
 middle_03_r {parent=middle_02_r, loc=<-2.6077 -0.0534 -0.1373>, rot=<0.0082 0.0188 0.1708 -0.9851>}
 ring_metacarpal_r {parent=hand_r, loc=<-2.2074 0.3625 -1.3638>, rot=<-0.0284 0.0722 -0.0458 -0.9959>}
 ring_01_r {parent=ring_metacarpal_r, loc=<-7.0295 0.6197 0.1446>, rot=<-0.0489 -0.0252 0.1892 -0.9804>}
 ring_02_r {parent=ring_01_r, loc=<-3.6174 -0.4647 0.1027>, rot=<0 0 0.2806 -0.9598>}
 ring_03_r {parent=ring_02_r, loc=<-2.4152 -0.114 -0.0199>, rot=<0 0 0.1189 -0.9929>}
 pinky_metacarpal_r {parent=hand_r, loc=<-1.7616 0.9753 -3.3814>, rot=<-0.0397 0.0814 -0.069 -0.9935>}
 pinky_01_r {parent=pinky_metacarpal_r, loc=<-7.1709 0.4831 0.1094>, rot=<-0.0654 -0.0035 0.2242 -0.9723>}
 pinky_02_r {parent=pinky_01_r, loc=<-2.8274 -0.0893 0.1043>, rot=<0.0037 0.0011 0.2854 -0.9584>}
 pinky_03_r {parent=pinky_02_r, loc=<-2.2046 -0.0307 0.0078>, rot=<-0.0002 0.0098 0.1445 -0.9895>}
 deform_wrist_up_base_r {parent=hand_r, loc=<-0.0002 -2.3007 -0.0001>, rot=<0 0 0 -1>}
 deform_wrist_up_r {parent=deform_wrist_up_base_r, loc=<0 0 0>, rot=<0 0 0 1>}
 lowerarm_twist_02_r {parent=lowerarm_r, loc=<-10.6707 0 0>, rot=<0 0 0 -1>}
 lowerarm_twist_01_r {parent=lowerarm_r, loc=<-18.316 0 0>, rot=<0 0 0 -1>}
 upperarm_twist_01_r {parent=upperarm_r, loc=<-7.2135 0.1134 0>, rot=<0 0 0 -1>}
 deform_upperarm_bk_base_r {parent=upperarm_twist_01_r, loc=<6.4053 4.3746 -0.314>, rot=<0 0 0 -1>}
 deform_upperarm_bk_r {parent=deform_upperarm_bk_base_r, loc=<0 0 0>, rot=<0 0 0 1>}
 deform_upperarm_fr_base_r {parent=upperarm_twist_01_r, loc=<5.6312 -5.092 -0.3354>, rot=<0 0 0 -1>}
 deform_upperarm_fr_r {parent=deform_upperarm_fr_base_r, loc=<0 0 0>, rot=<0 0 0 1>}
 deform_upperarm_up_base_r {parent=upperarm_twist_01_r, loc=<8.891 -0.4825 -3.9517>, rot=<0 0 0 -1>}
 deform_upperarm_up_r {parent=deform_upperarm_up_base_r, loc=<0 0 0>, rot=<0 0 0 1>}
 base_LeftSleeve_r {parent=upperarm_twist_01_r, loc=<-3.0515 4.7956 0.6198>, rot=<-0.0607 0.0179 0.9976 -0.0279>}
 dyn_LeftSleeve_r {parent=base_LeftSleeve_r, loc=<0 0 0>, rot=<0 0 0 -1>}
 base_RightSleeve_r {parent=upperarm_twist_01_r, loc=<-2.5652 -4.2693 0.9752>, rot=<0.0279 -0.9976 0.0179 -0.0607>}
 dyn_RightSleeve_r {parent=base_RightSleeve_r, loc=<0 0 0>, rot=<0 0 0 1>}
 base_BotSleeve_r {parent=upperarm_twist_01_r, loc=<-3.9407 -1.2487 4.2993>, rot=<-0.0222 -0.814 0.5801 -0.0191>}
 dyn_BotSleeve_r {parent=base_BotSleeve_r, loc=<0 0 0>, rot=<0 0 0 -1>}
 base_TopSleeve_r {parent=upperarm_twist_01_r, loc=<-1.4606 0.765 -2.4537>, rot=<-0.0186 0.7004 0.713 0.0282>}
 dyn_TopSleeve_r {parent=base_TopSleeve_r, loc=<0 0 0>, rot=<0 0 0 1>}
 upperarm_twist_02_r {parent=upperarm_r, loc=<-19.7998 0.1134 0>, rot=<0 0 0 -1>}
 attach_backpack {parent=spine_05, loc=<-0.123 -13.473 0>, rot=<-0.0093 -0.707 0.0093 -0.707>}
 Backpack_BR {parent=attach_backpack, type=Null, loc=<0 3.5263 0>, rot=<0 0 0 1>}
 attach_cape {parent=spine_05, loc=<9.017 -11.714 0>, rot=<0.0616 -0.7044 -0.0616 -0.7044>}
 cape {parent=attach_cape, type=Null, loc=<0 0 0>, rot=<0 0 0 1>}
 deform_pec_aim_base_l {parent=spine_04, loc=<3.2456 9.3309 -1.2237>, rot=<0.0603 -0.4033 0.3318 -0.8507>}
 deform_pec_aim_l {parent=deform_pec_aim_base_l, loc=<0 0 0>, rot=<0 0 0 1>}
 deform_pec_base_l {parent=deform_pec_aim_l, loc=<9.3665 0 0>, rot=<0 0 0 1>}
 deform_pec_l {parent=deform_pec_base_l, loc=<0 0 0>, rot=<0 0 0 1>}
 deform_pec_aim_base_r {parent=spine_04, loc=<3.2452 9.3309 1.2237>, rot=<-0.4033 -0.0603 0.8507 0.3318>}
 deform_pec_aim_r {parent=deform_pec_aim_base_r, loc=<0 0 0>, rot=<0 0 0 1>}
 deform_pec_base_r {parent=deform_pec_aim_r, loc=<-9.3667 -0.0003 -0.0004>, rot=<0 0 0 1>}
 deform_pec_r {parent=deform_pec_base_r, loc=<0 0 0>, rot=<0 0 0 1>}
 thigh_l {parent=pelvis, loc=<-1.0855 -0.2109 -8.7053>, rot=<0.0009 0.0355 0.0155 -0.9993>}
 calf_l {parent=thigh_l, loc=<-42.5284 -0.0019 -0.7297>, rot=<0.0006 -0.0161 -0.0391 -0.9991>}
 foot_l {parent=calf_l, loc=<-46.7401 -2.5014 -0.592>, rot=<0.0427 -0.0238 0.0353 -0.9982>}
 ball_l {parent=foot_l, loc=<-3.5658 10.8066 -0.1544>, rot=<0.0001 -0.0001 -0.7063 -0.7079>}
 calf_twist_02_l {parent=calf_l, loc=<-19.1546 -1.7943 -0.2638>, rot=<0.0237 0.0095 -0.0035 -0.9997>}
 calf_twist_01_l {parent=calf_l, loc=<-35.8483 -1.9197 -0.5794>, rot=<0.0237 0.0095 -0.0035 -0.9997>}
 deform_knee_in_base_l {parent=calf_l, loc=<0.8031 -5.0231 0.2456>, rot=<0 0 0 -1>}
 deform_knee_in_l {parent=deform_knee_in_base_l, loc=<0 0 0>, rot=<0 0 0 1>}
 deform_knee_out_base_l {parent=calf_l, loc=<0 4 0>, rot=<0 0 0 -1>}
 deform_knee_out_l {parent=deform_knee_out_base_l, loc=<0 0 0>, rot=<0 0 0 1>}
 thigh_twist_01_l {parent=thigh_l, loc=<-27.952 0 0>, rot=<0.0013 0 0 -1>}
 deform_groin_base_l {parent=thigh_l, loc=<0.4801 5.6776 -0.2486>, rot=<0 0 0 -1>}
 deform_groin_l {parent=deform_groin_base_l, loc=<0 0 0>, rot=<0 0 0 1>}
 thigh_r {parent=pelvis, loc=<-1.0855 -0.2109 8.7053>, rot=<0.0355 -0.0009 0.9993 0.0155>}
 calf_r {parent=thigh_r, loc=<42.5284 0.0019 0.7297>, rot=<0.0006 -0.0161 -0.0391 -0.9991>}
 foot_r {parent=calf_r, loc=<46.7401 2.5014 0.592>, rot=<0.0427 -0.0238 0.0353 -0.9982>}
 ball_r {parent=foot_r, loc=<3.5658 -10.8066 0.1544>, rot=<0.0001 -0.0001 -0.7063 -0.7079>}
 calf_twist_02_r {parent=calf_r, loc=<19.1546 1.7943 0.2638>, rot=<0.0237 0.0095 -0.0035 -0.9997>}
 calf_twist_01_r {parent=calf_r, loc=<35.8483 1.9197 0.5794>, rot=<0.0237 0.0095 -0.0035 -0.9997>}
 deform_knee_in_base_r {parent=calf_r, loc=<-0.8031 5.0231 -0.2457>, rot=<0 0 0 -1>}
 deform_knee_in_r {parent=deform_knee_in_base_r, loc=<0 0 0>, rot=<0 0 0 1>}
 deform_knee_out_base_r {parent=calf_r, loc=<0 -4 0>, rot=<0 0 0 -1>}
 deform_knee_out_r {parent=deform_knee_out_base_r, loc=<0 0 0>, rot=<0 0 0 1>}
 thigh_twist_01_r {parent=thigh_r, loc=<27.952 0 0>, rot=<0.0013 0 0 -1>}
 deform_groin_base_r {parent=thigh_r, loc=<-0.4801 -5.6776 0.2486>, rot=<0 0 0 -1>}
 deform_groin_r {parent=deform_groin_base_r, loc=<0 0 0>, rot=<0 0 0 1>}
 deform_glute_base_l {parent=pelvis, loc=<-7.3466 -9.0213 -9.1245>, rot=<0.0172 0.031 -0.4477 -0.8935>}
 deform_glute_l {parent=deform_glute_base_l, loc=<0 0 0>, rot=<0 0 0 1>}
 deform_glute_base_r {parent=pelvis, loc=<-7.3464 -9.0213 9.1245>, rot=<0.031 -0.0172 0.8935 -0.4477>}
 deform_glute_r {parent=deform_glute_base_r, loc=<0 0 0>, rot=<0 0 0 1>}
 mainbelt_front {parent=pelvis, loc=<5.2866 8.8383 -1.2004>, rot=<0 0 -0.6943 -0.7197>}
 mainbelt_L_01 {parent=pelvis, loc=<5.6956 6.6693 8.9487>, rot=<-0.2481 0.2571 -0.6485 -0.6722>}
 mainbelt_L_02 {parent=pelvis, loc=<6.4974 -0.1378 12.9627>, rot=<-0.6136 0.4384 -0.3624 -0.5477>}
 mainbelt_back {parent=pelvis, loc=<7.3159 -9.0868 1.5699>, rot=<-0.0127 -0.707 0.0127 -0.707>}
 mainbelt_R_01 {parent=pelvis, loc=<6.4579 0.6409 -13.0051>, rot=<-0.0127 -0.707 0.0127 -0.707>}
 mainbelt_R_02 {parent=pelvis, loc=<7.5696 -9.5564 -5.8425>, rot=<-0.1865 -0.7271 -0.19 -0.6328>}
 mainbelt_R_03 {parent=pelvis, loc=<7.3904 -6.6472 9.5818>, rot=<-0.0127 -0.707 0.0127 -0.707>}
 thigh_l_null {parent=pelvis, type=Null, loc=<-1.0855 -0.2109 -8.7053>, rot=<0.0009 0.0355 0.0155 -0.9993>}
 thigh_r_null {parent=pelvis, type=Null, loc=<-1.0855 -0.2109 8.7053>, rot=<0.0355 -0.0009 0.9993 0.0155>}
 base_leftBelt {parent=pelvis, loc=<7.108 -1.1978 1.1371>, rot=<-0.2432 0.6349 0.6919 0.2429>}
 dyn_leftBelt_01 {parent=base_leftBelt, loc=<0 0 0>, rot=<0 0 0 1>}
 dyn_leftBelt_02 {parent=dyn_leftBelt_01, loc=<6.2533 0 0>, rot=<0 0 0.1648 -0.9863>}
 dyn_leftBelt_03 {parent=dyn_leftBelt_02, loc=<6.6728 0 0>, rot=<0 0 0 1>}
 dyn_leftBelt_side_01 {parent=dyn_leftBelt_03, loc=<8.9202 -0.887 -0.5862>, rot=<-0.017 -0.0378 0.2895 -0.9563>}
 dyn_leftBelt_side_02 {parent=dyn_leftBelt_03, loc=<0.3145 2.4919 8.2846>, rot=<0.0576 0.6467 0.1068 -0.753>}
 base_keychain {parent=dyn_leftBelt_side_02, loc=<0.6792 2.1198 -0.2209>, rot=<0.7262 0.6848 0.0139 0.0597>}
 dyn_keychain_01 {parent=base_keychain, loc=<0 0 0>, rot=<0 0 0 1>}
 dyn_keychain_02 {parent=dyn_keychain_01, loc=<1.472 -0.0429 0.2996>, rot=<0.0015 0.1316 0.0115 -0.9912>}
 dyn_leftBelt_side_03 {parent=dyn_leftBelt_03, loc=<2.1411 -0.6098 -8.7912>, rot=<0.2658 -0.492 0.1544 -0.8145>}
 dyn_leftBelt_side_04 {parent=dyn_leftBelt_01, loc=<0.5422 5.3987 8.2608>, rot=<0.0332 0.7682 -0.0033 -0.6394>}
 dyn_leftBelt_side_05 {parent=dyn_leftBelt_01, loc=<5.8426 -3.2408 -8.2344>, rot=<0.2148 -0.8105 0.1348 -0.528>}
 base_rightBelt {parent=pelvis, loc=<2.8134 -0.4092 8.3038>, rot=<0.1622 -0.4938 0.8159 0.2533>}
 dyn_rightBelt_01 {parent=base_rightBelt, loc=<0 0 0>, rot=<0 0 0 1>}
 rightbelt_01 {parent=dyn_rightBelt_01, loc=<0.2126 -0.7866 -7.5406>, rot=<0.1775 -0.6548 0.1376 -0.7217>}
 rightbelt_05 {parent=dyn_rightBelt_01, loc=<-0.2126 0.7866 7.5406>, rot=<-0.1271 0.7029 0.1167 -0.69>}
 dyn_rightBelt_02 {parent=dyn_rightBelt_01, loc=<9.9316 0 0>, rot=<0 0 0.1071 -0.9942>}
 rightbelt_02 {parent=dyn_rightBelt_02, loc=<0.2019 0.1049 -5.2967>, rot=<0.1624 -0.6012 0.2274 -0.7487>}
 rightbelt_03 {parent=dyn_rightBelt_02, loc=<5.6774 0.0389 0.3614>, rot=<0.0822 0.1209 0.3559 -0.923>}
 rightbelt_04 {parent=dyn_rightBelt_02, loc=<-0.2201 -0.1089 6.156>, rot=<-0.1275 0.603 0.282 -0.7352>}
 Tail {parent=pelvis, type=Null, loc=<2.25 -9.5 0>, rot=<-0.1228 0.6964 0.1228 0.6964>}
]
variables [
 On Grind Rail {type=bool, default=False}
 l thighLocal ry float {type=double, default=0}
 l thighLocal rz float {type=double, default=0}
 r thighLocal ry float {type=double, default=0}
 r thighLocal rz float {type=double, default=0}
]
