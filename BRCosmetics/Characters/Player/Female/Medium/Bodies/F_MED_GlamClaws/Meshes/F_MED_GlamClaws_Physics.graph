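# Physics asset for the F_MED_GlamClaws female character body, set up for a dangling two-link keychain.
# A small base capsule anchors two box bodies chained in sequence, with narrow swing limits and zero twist limits.
# Capsules on the pelvis and left thigh cover the body around the keychain.
# The base and thigh each constrain to the pelvis with fully free angular motion, leaving only the chain links angularly restricted.
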PhysicsAsset /BRCosmetics/Characters/Player/Female/Medium/Bodies/F_MED_GlamClaws/Meshes/F_MED_GlamClaws_Physics
bodies {
 base_keychain [{shapeType=Capsule, length=0.505, radius=0.505, shapeName=base_keychain_capsule, rotation=<rot 0 -90 0>}]
 dyn_keychain_01 [{
   shapeType=Box
   extentX=0.4412
   extentY=2.5756
   extentZ=2.6453
   shapeName=dyn_keychain_01_box
   center=<0.4415 -0.0232 0.0377>
   rotation=<rot -0.7244 -89.9998 -13.4273>
  }]
 dyn_keychain_02 [{
   shapeType=Box
   extentX=0.9499
   extentY=4.4275
   extentZ=4.5459
   shapeName=dyn_keychain_02_box
   center=<2.465 0.1178 0.019>
   rotation=<rot 0 -89.8554 -89.5387>
  }]
 pelvis [{
   shapeType=Capsule
   length=14.4152
   radius=10.6638
   shapeName=pelvis_capsule
   center=<-7.0847 -3.0685 -0.0421>
   rotation=<rot 0 178.2378 -178.8964>
  }]
 thigh_l [{
   shapeType=Capsule
   length=12.2472
   radius=9.0741
   shapeName=thigh_l_capsule
   center=<-12.3034 -1.1805 -0.2552>
   rotation=<rot 1.1722 -100.185 -88.2835>
  }]
}
constraints [
 {
  bone1Name=base_keychain
  bone2Name=pelvis
  swing1LimitDegrees=45
  swing1Motion=Free
  swing2LimitDegrees=45
  swing2Motion=Free
  twistLimitDegrees=45
  twistMotion=Free
 }
 {
  bone1Name=thigh_l
  bone2Name=pelvis
  swing1LimitDegrees=45
  swing1Motion=Free
  swing2LimitDegrees=45
  swing2Motion=Free
  twistLimitDegrees=45
  twistMotion=Free
 }
 {bone1Name=dyn_keychain_01, bone2Name=base_keychain, swing1LimitDegrees=15, swing2LimitDegrees=10, twistLimitDegrees=0}
 {bone1Name=dyn_keychain_02, bone2Name=dyn_keychain_01, swing1LimitDegrees=0, swing2LimitDegrees=10, twistLimitDegrees=0}
]
