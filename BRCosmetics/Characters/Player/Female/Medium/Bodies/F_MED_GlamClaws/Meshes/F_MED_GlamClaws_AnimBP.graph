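# Post-process style AnimBP for the GlamClaws female cosmetic body, adding secondary motion to sleeves and belts.
# It copies the parent pose through a Control Rig, constrains sleeve base bones to arm bones, then runs RigidBody and Anim Dynamics.
# Outside the front end, updates feed motion to world-position-offset cloth materials, adding wind and turbulence while skydive-floating or parachuting.
# On customization load, it caches materials and hides any tail under capes or long backblings.
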AnimBlueprint /BRCosmetics/Characters/Player/Female/Medium/Bodies/F_MED_GlamClaws/Meshes/F_MED_GlamClaws_AnimBP
target_skeleton=&/F_MED_GlamClaws_Skeleton
graphs [
 AnimGraph {nodes [
   "Constraint - base_keychain (PPPPPP)"!(
    ComponentPose
    Alpha=1
    ConstraintWeights_0=0.8878
    ConstraintWeights_1:float=0.0967
    ConstraintWeights_2:float=0.0129
    ConstraintWeights_3:float=0.0011
    ConstraintWeights_4:float=0.0009
    ConstraintWeights_5:float=0.0007
   ) -> (Pose) {
    boneToModify {boneName=base_keychain}
    constraintSetup [
     {targetBone {boneName=dyn_leftBelt_side_02}}
     {targetBone {boneName=dyn_leftBelt_side_04}}
     {targetBone {boneName=dyn_leftBelt_side_01}}
     {targetBone {boneName=pelvis}}
     {targetBone {boneName=thigh_twist_01_l}}
     {targetBone {boneName=deform_groin_l}}
    ]
    constraintWeights [1,1,1,1,1,1]
   }
   Output Pose(Result=Use cached pose 'CRCache_dynamics'()) {layerGroup=None}
   n1 = Control Rig!(Source=Copy Pose From Mesh() {bUseAttachedParent=true}, Alpha, On Grind Rail:bool=false) -> (Pose) {controlRigAssetReference {blueprintRigClass=&/F_MED_GlamClaws_CtrlRig.F_MED_GlamClaws_CtrlRig_C}}
   Save cached pose 'CRCache_inputPose'(Pose=@n1) {cacheName=CRCache_inputPose}
   n2 = RigidBody(ComponentPose=Local To Component(LocalPose=Use cached pose 'CRCAache_preDynamics'())) {
    overridePhysicsAsset=&/F_MED_GlamClaws_Physics
    simulationSpace=BaseBoneSpace
    baseBoneRef {boneName=pelvis}
    componentAppliedLinearAccClamp=<3000 3000 3000>
    simSpaceSettings {maxAngularAcceleration=20, maxAngularVelocity=10, maxLinearAcceleration=800, maxLinearVelocity=400, worldAlpha=1}
   }
   n3 = Anim Dynamics - Bone: dyn_TopSleeve_l(ComponentPose=@n2, Alpha=0.4, ExternalForce="0,0,0") {
    angularDampingOverride=1
    angularSpringConstant=100
    bAngularSpring=true
    bOverrideAngularDamping=true
    bUseGravityOverride=true
    gravityScale=0.35
    simulationSpace=RootRelative
    alphaBoolBlend {blendOption=Linear}
    boundBone {boneName=dyn_TopSleeve_l}
    componentAppliedLinearAccClamp=<3000 3000 3000>
    componentLinearAccScale=<1 1 1>
    componentLinearVelScale=<1 1 1>
    gravityOverride {z=-150}
    physicsBodyDefinitions [{
      boundBone {boneName=dyn_TopSleeve_l}
      boxExtents=<12 30 12>
      constraintSetup {angularLimitsMax=<10 10 10>, angularLimitsMin=<-10 -10 -10>, angularTarget=<1 0 0>, linearAxesMax=<2 0 0>, linearAxesMin=<-0.5 0 0>}
      localJointOffset=<0.2 0 0>
     }]
    simSpaceSettings {maxAngularAcceleration=14, maxAngularVelocity=7, simSpaceAngularAlpha=1}
   }
   n4 = Anim Dynamics - Bone: dyn_BotSleeve_l(ComponentPose=@n3, Alpha=0.4, ExternalForce="0,0,0") {
    angularDampingOverride=1
    angularSpringConstant=100
    bAngularSpring=true
    bOverrideAngularDamping=true
    bUseGravityOverride=true
    gravityScale=0.35
    simulationSpace=RootRelative
    alphaBoolBlend {blendOption=Linear}
    boundBone {boneName=dyn_BotSleeve_l}
    componentAppliedLinearAccClamp=<3000 3000 3000>
    componentLinearAccScale=<1 1 1>
    componentLinearVelScale=<1 1 1>
    gravityOverride {z=-150}
    physicsBodyDefinitions [{
      boundBone {boneName=dyn_BotSleeve_l}
      boxExtents=<12 30 12>
      constraintSetup {angularLimitsMax=<10 10 10>, angularLimitsMin=<-10 -10 -10>, angularTarget=<1 0 0>, linearAxesMax=<2 0 0>, linearAxesMin=<-0.5 0 0>}
      localJointOffset=<0.2 0 0>
     }]
    simSpaceSettings {maxAngularAcceleration=14, maxAngularVelocity=7, simSpaceAngularAlpha=1}
   }
   n5 = Anim Dynamics - Bone: dyn_RightSleeve_l(ComponentPose=@n4, Alpha=0.4, ExternalForce="0,0,0") {
    angularDampingOverride=1
    angularSpringConstant=100
    bAngularSpring=true
    bOverrideAngularDamping=true
    bUseGravityOverride=true
    gravityScale=0.35
    simulationSpace=RootRelative
    alphaBoolBlend {blendOption=Linear}
    boundBone {boneName=dyn_RightSleeve_l}
    componentAppliedLinearAccClamp=<3000 3000 3000>
    componentLinearAccScale=<1 1 1>
    componentLinearVelScale=<1 1 1>
    gravityOverride {z=-150}
    physicsBodyDefinitions [{
      boundBone {boneName=dyn_RightSleeve_l}
      boxExtents=<12 30 12>
      constraintSetup {angularLimitsMax=<10 10 10>, angularLimitsMin=<-10 -10 -10>, angularTarget=<1 0 0>, linearAxesMax=<2 0 0>, linearAxesMin=<-0.5 0 0>}
      localJointOffset=<0.2 0 0>
     }]
    simSpaceSettings {maxAngularAcceleration=14, maxAngularVelocity=7, simSpaceAngularAlpha=1}
   }
   n6 = Anim Dynamics - Bone: dyn_LeftSleeve_l(ComponentPose=@n5, Alpha=0.4, ExternalForce="0,0,0") {
    angularDampingOverride=1
    angularSpringConstant=100
    bAngularSpring=true
    bOverrideAngularDamping=true
    bUseGravityOverride=true
    gravityScale=0.35
    simulationSpace=RootRelative
    alphaBoolBlend {blendOption=Linear}
    boundBone {boneName=dyn_LeftSleeve_l}
    componentAppliedLinearAccClamp=<3000 3000 3000>
    componentLinearAccScale=<1 1 1>
    componentLinearVelScale=<1 1 1>
    gravityOverride {z=-150}
    physicsBodyDefinitions [{
      boundBone {boneName=dyn_LeftSleeve_l}
      boxExtents=<12 30 12>
      constraintSetup {angularLimitsMax=<10 10 10>, angularLimitsMin=<-10 -10 -10>, angularTarget=<1 0 0>, linearAxesMax=<2 0 0>, linearAxesMin=<-0.5 0 0>}
      localJointOffset=<0.2 0 0>
     }]
    simSpaceSettings {maxAngularAcceleration=14, maxAngularVelocity=7, simSpaceAngularAlpha=1}
   }
   n7 = Anim Dynamics - Bone: dyn_BotSleeve_r(ComponentPose=@n6, Alpha=0.4, ExternalForce="0,0,0") {
    angularDampingOverride=1
    angularSpringConstant=100
    bAngularSpring=true
    bOverrideAngularDamping=true
    bUseGravityOverride=true
    gravityScale=0.35
    simulationSpace=RootRelative
    alphaBoolBlend {blendOption=Linear}
    boundBone {boneName=dyn_BotSleeve_r}
    componentAppliedLinearAccClamp=<3000 3000 3000>
    componentLinearAccScale=<1 1 1>
    componentLinearVelScale=<1 1 1>
    gravityOverride {z=-150}
    physicsBodyDefinitions [{
      boundBone {boneName=dyn_BotSleeve_r}
      boxExtents=<12 30 12>
      constraintSetup {angularLimitsMax=<10 10 10>, angularLimitsMin=<-10 -10 -10>, angularTarget=<1 0 0>, linearAxesMax=<3 0 0>, linearAxesMin=<-1.5 0 0>}
      localJointOffset=<0.2 0 0>
     }]
    simSpaceSettings {maxAngularAcceleration=14, maxAngularVelocity=7, simSpaceAngularAlpha=1}
   }
   n8 = Anim Dynamics - Bone: dyn_RightSleeve_r(ComponentPose=@n7, Alpha=0.4, ExternalForce="0,0,0") {
    angularDampingOverride=1
    angularSpringConstant=100
    bAngularSpring=true
    bOverrideAngularDamping=true
    bUseGravityOverride=true
    gravityScale=0.35
    simulationSpace=RootRelative
    alphaBoolBlend {blendOption=Linear}
    boundBone {boneName=dyn_RightSleeve_r}
    componentAppliedLinearAccClamp=<3000 3000 3000>
    componentLinearAccScale=<1 1 1>
    componentLinearVelScale=<1 1 1>
    gravityOverride {z=-150}
    physicsBodyDefinitions [{
      boundBone {boneName=dyn_RightSleeve_r}
      boxExtents=<12 30 12>
      constraintSetup {angularLimitsMax=<10 10 10>, angularLimitsMin=<-10 -10 -10>, angularTarget=<1 0 0>, linearAxesMax=<3 0 0>, linearAxesMin=<-1.5 0 0>}
      localJointOffset=<0.2 0 0>
     }]
    simSpaceSettings {maxAngularAcceleration=14, maxAngularVelocity=7, simSpaceAngularAlpha=1}
   }
   n9 = Anim Dynamics - Bone: dyn_TopSleeve_r(ComponentPose=@n8, Alpha=0.4, ExternalForce="0,0,0") {
    angularDampingOverride=1
    angularSpringConstant=100
    bAngularSpring=true
    bOverrideAngularDamping=true
    bUseGravityOverride=true
    gravityScale=0.35
    simulationSpace=RootRelative
    alphaBoolBlend {blendOption=Linear}
    boundBone {boneName=dyn_TopSleeve_r}
    componentAppliedLinearAccClamp=<3000 3000 3000>
    componentLinearAccScale=<1 1 1>
    componentLinearVelScale=<1 1 1>
    gravityOverride {z=-150}
    physicsBodyDefinitions [{
      boundBone {boneName=dyn_TopSleeve_r}
      boxExtents=<12 30 12>
      constraintSetup {angularLimitsMax=<10 10 10>, angularLimitsMin=<-10 -10 -10>, angularTarget=<1 0 0>, linearAxesMax=<3 0 0>, linearAxesMin=<-1.5 0 0>}
      localJointOffset=<0.2 0 0>
     }]
    simSpaceSettings {maxAngularAcceleration=14, maxAngularVelocity=7, simSpaceAngularAlpha=1}
   }
   n10 = Anim Dynamics - Bone: dyn_LeftSleeve_r(ComponentPose=@n9, Alpha=0.4, ExternalForce="0,0,0") {
    angularDampingOverride=1
    angularSpringConstant=100
    bAngularSpring=true
    bOverrideAngularDamping=true
    bUseGravityOverride=true
    gravityScale=0.35
    simulationSpace=RootRelative
    alphaBoolBlend {blendOption=Linear}
    boundBone {boneName=dyn_LeftSleeve_r}
    componentAppliedLinearAccClamp=<3000 3000 3000>
    componentLinearAccScale=<1 1 1>
    componentLinearVelScale=<1 1 1>
    gravityOverride {z=-150}
    physicsBodyDefinitions [{
      boundBone {boneName=dyn_LeftSleeve_r}
      boxExtents=<12 30 12>
      constraintSetup {angularLimitsMax=<10 10 10>, angularLimitsMin=<-10 -10 -10>, angularTarget=<1 0 0>, linearAxesMax=<3 0 0>, linearAxesMin=<-1.5 0 0>}
      localJointOffset=<0.2 0 0>
     }]
    simSpaceSettings {maxAngularAcceleration=14, maxAngularVelocity=7, simSpaceAngularAlpha=1}
   }
   n11 = Anim Dynamics - Bone: dyn_leftBelt_02!(ComponentPose=@n10, Alpha=1, GravityOverride:Vector="0,0,-980") -> (Pose) {
    angularDampingOverride=1
    angularSpringConstant=20000
    bOverrideAngularDamping=true
    bUseGravityOverride=true
    gravityScale=0.35
    simulationSpace=RootRelative
    alphaBoolBlend {blendOption=Linear}
    alphaScaleBiasClamp {bInterpResult=true}
    boundBone {boneName=dyn_leftBelt_02}
    componentAppliedLinearAccClamp=<3000 3000 3000>
    componentLinearAccScale=<1 1 1>
    componentLinearVelScale=<1 1 1>
    gravityOverride {z=-980}
    physicsBodyDefinitions [{
      boundBone {boneName=dyn_leftBelt_02}
      boxExtents=<30 30 30>
      constraintSetup {angularLimitsMax=<1 2 0>, angularLimitsMin=<-2 -2 -5>, angularTarget=<2 0 0>}
      localJointOffset=<1 0 0>
     }]
    relativeSpaceBone {boneName=pelvis}
    simSpaceSettings {maxAngularAcceleration=10, maxAngularVelocity=5, simSpaceAngularAlpha=1}
   }
   n12 = Anim Dynamics - Bone: dyn_leftBelt_03!(ComponentPose=@n11, Alpha=1, GravityOverride:Vector="0,0,-980") -> (Pose) {
    angularDampingOverride=1
    angularSpringConstant=20000
    bOverrideAngularDamping=true
    bUseGravityOverride=true
    gravityScale=0.35
    alphaBoolBlend {blendOption=Linear}
    alphaScaleBiasClamp {bInterpResult=true}
    boundBone {boneName=dyn_leftBelt_03}
    componentLinearAccScale=<1 1 1>
    componentLinearVelScale=<1 1 1>
    gravityOverride {z=-980}
    physicsBodyDefinitions [{
      boundBone {boneName=dyn_leftBelt_03}
      boxExtents=<28 28 28>
      constraintSetup {angularLimitsMax=<0 0 0>, angularLimitsMin=<0 0 -25>, angularTarget=<1 0 0>}
      localJointOffset=<2 0 0>
     }]
    relativeSpaceBone {boneName=pelvis}
    simSpaceSettings {maxAngularAcceleration=10, maxAngularVelocity=5, simSpaceAngularAlpha=1}
   }
   n13 = Anim Dynamics - Bone: dyn_rightBelt_01!(ComponentPose=@n12, Alpha=1, GravityOverride:Vector="0,0,-980") -> (Pose) {
    angularDampingOverride=1
    angularSpringConstant=20000
    bOverrideAngularDamping=true
    bUseGravityOverride=true
    gravityScale=0.35
    simulationSpace=RootRelative
    alphaBoolBlend {blendOption=Linear}
    alphaScaleBiasClamp {bInterpResult=true}
    boundBone {boneName=dyn_rightBelt_01}
    componentAppliedLinearAccClamp=<3000 3000 3000>
    componentLinearAccScale=<1 1 1>
    componentLinearVelScale=<1 1 1>
    gravityOverride {z=-980}
    physicsBodyDefinitions [{
      boundBone {boneName=dyn_rightBelt_01}
      boxExtents=<30 30 30>
      constraintSetup {angularLimitsMax=<1 2 0>, angularLimitsMin=<-2 -2 -5>, angularTarget=<2 0 0>}
      localJointOffset=<1 0 0>
     }]
    relativeSpaceBone {boneName=pelvis}
    simSpaceSettings {maxAngularAcceleration=10, maxAngularVelocity=5, simSpaceAngularAlpha=1}
   }
   n14 = Anim Dynamics - Bone: dyn_rightBelt_02!(ComponentPose=@n13, Alpha=1, GravityOverride:Vector="0,0,-980") -> (Pose) {
    angularDampingOverride=1
    angularSpringConstant=20000
    bOverrideAngularDamping=true
    bUseGravityOverride=true
    gravityScale=0.35
    simulationSpace=RootRelative
    alphaBoolBlend {blendOption=Linear}
    alphaScaleBiasClamp {bInterpResult=true}
    boundBone {boneName=dyn_rightBelt_02}
    componentLinearAccScale=<1 1 1>
    componentLinearVelScale=<1 1 1>
    gravityOverride {z=-980}
    physicsBodyDefinitions [{
      boundBone {boneName=dyn_rightBelt_02}
      boxExtents=<10 10 10>
      constraintSetup {angularLimitsMax=<0 0 0>, angularLimitsMin=<0 0 -10>, angularTarget=<1 0 0>}
      localJointOffset=<1 0 0>
     }]
    relativeSpaceBone {boneName=pelvis}
    simSpaceSettings {maxAngularAcceleration=10, maxAngularVelocity=5, simSpaceAngularAlpha=1}
   }
   Save cached pose 'CRCache_dynamics'(Pose=Component To Local(ComponentPose=@n14)) {cacheName=CRCache_dynamics}
   n15 = "Constraint - base_TopSleeve_l (PPPPPP)"!(
    ComponentPose=Local To Component(LocalPose=Use cached pose 'CRCache_inputPose'())
    Alpha=1
    ConstraintWeights_0=0.7771
    ConstraintWeights_1:float=0.2054
    ConstraintWeights_2:float=0.0105
    ConstraintWeights_3:float=0.0025
    ConstraintWeights_4:float=0.0023
    ConstraintWeights_5:float=0.0022
   ) -> (Pose) {
    boneToModify {boneName=base_TopSleeve_l}
    constraintSetup [
     {targetBone {boneName=upperarm_twist_01_l}}
     {targetBone {boneName=upperarm_twist_02_l}}
     {targetBone {boneName=clavicle_l}}
     {targetBone {boneName=upperarm_l}}
     {targetBone {boneName=deform_upperarm_up_l}}
     {targetBone {boneName=deform_upperarm_fr_l}}
    ]
    constraintWeights [1,1,1,1,1,1]
   }
   n16 = "Constraint - base_BotSleeve_l (PPPPPP)"!(
    ComponentPose=@n15
    Alpha=1
    ConstraintWeights_0=0.5417
    ConstraintWeights_1:float=0.3982
    ConstraintWeights_2:float=0.0532
    ConstraintWeights_3:float=0.0046
    ConstraintWeights_4:float=0.0018
    ConstraintWeights_5:float=0.0004
   ) -> (Pose) {
    boneToModify {boneName=base_BotSleeve_l}
    constraintSetup [
     {targetBone {boneName=upperarm_twist_01_l}}
     {targetBone {boneName=upperarm_twist_02_l}}
     {targetBone {boneName=upperarm_l}}
     {targetBone {boneName=clavicle_l}}
     {targetBone {boneName=deform_upperarm_fr_l}}
     {targetBone {boneName=spine_05}}
    ]
    constraintWeights [1,1,1,1,1,1]
   }
   n17 = "Constraint - base_RightSleeve_l (PPPPPP)"!(
    ComponentPose=@n16
    Alpha=1
    ConstraintWeights_0=0.6541
    ConstraintWeights_1:float=0.2985
    ConstraintWeights_2:float=0.0263
    ConstraintWeights_3:float=0.0125
    ConstraintWeights_4:float=0.0081
    ConstraintWeights_5:float=0.0005
   ) -> (Pose) {
    boneToModify {boneName=base_RightSleeve_l}
    constraintSetup [
     {targetBone {boneName=upperarm_twist_01_l}}
     {targetBone {boneName=upperarm_twist_02_l}}
     {targetBone {boneName=deform_upperarm_fr_l}}
     {targetBone {boneName=upperarm_l}}
     {targetBone {boneName=clavicle_l}}
     {targetBone {boneName=spine_05}}
    ]
    constraintWeights [1,1,1,1,1,1]
   }
   n18 = "Constraint - base_LeftSleeve_l (PPPPP)"!(
    ComponentPose=@n17
    Alpha=1
    ConstraintWeights_0=0.5597
    ConstraintWeights_1:float=0.4098
    ConstraintWeights_2:float=0.0274
    ConstraintWeights_3:float=0.0028
    ConstraintWeights_4:float=0.0003
   ) -> (Pose) {
    boneToModify {boneName=base_LeftSleeve_l}
    constraintSetup [
     {targetBone {boneName=upperarm_twist_01_l}}
     {targetBone {boneName=upperarm_twist_02_l}}
     {targetBone {boneName=upperarm_l}}
     {targetBone {boneName=clavicle_l}}
     {targetBone {boneName=deform_upperarm_bk_l}}
    ]
    constraintWeights [1,1,1,1,1]
   }
   n19 = "Constraint - base_LeftSleeve_r (PPPPPP)"!(
    ComponentPose=@n18
    Alpha=1
    ConstraintWeights_0=0.6031
    ConstraintWeights_1:float=0.3593
    ConstraintWeights_2:float=0.0218
    ConstraintWeights_3:float=0.0085
    ConstraintWeights_4:float=0.0072
    ConstraintWeights_5:float=0.0001
   ) -> (Pose) {
    boneToModify {boneName=base_LeftSleeve_r}
    constraintSetup [
     {targetBone {boneName=upperarm_twist_01_r}}
     {targetBone {boneName=upperarm_twist_02_r}}
     {targetBone {boneName=upperarm_r}}
     {targetBone {boneName=clavicle_r}}
     {targetBone {boneName=deform_upperarm_bk_r}}
     {targetBone {boneName=spine_05}}
    ]
    constraintWeights [1,1,1,1,1,1]
   }
   n20 = "Constraint - base_RightSleeve_r (PPPPPP)"!(
    ComponentPose=@n19
    Alpha=1
    ConstraintWeights_0=0.6374
    ConstraintWeights_1:float=0.3155
    ConstraintWeights_2:float=0.0236
    ConstraintWeights_3:float=0.0152
    ConstraintWeights_4:float=0.0079
    ConstraintWeights_5:float=0.0005
   ) -> (Pose) {
    boneToModify {boneName=base_RightSleeve_r}
    constraintSetup [
     {targetBone {boneName=upperarm_twist_01_r}}
     {targetBone {boneName=upperarm_twist_02_r}}
     {targetBone {boneName=deform_upperarm_fr_r}}
     {targetBone {boneName=upperarm_r}}
     {targetBone {boneName=clavicle_r}}
     {targetBone {boneName=spine_05}}
    ]
    constraintWeights [1,1,1,1,1,1]
   }
   n21 = "Constraint - base_BotSleeve_r (PPPPP)"!(
    ComponentPose=@n20
    Alpha=1
    ConstraintWeights_0=0.6362
    ConstraintWeights_1:float=0.3171
    ConstraintWeights_2:float=0.0304
    ConstraintWeights_3:float=0.0147
    ConstraintWeights_4:float=0.0017
   ) -> (Pose) {
    boneToModify {boneName=base_BotSleeve_r}
    constraintSetup [
     {targetBone {boneName=upperarm_twist_01_r}}
     {targetBone {boneName=upperarm_twist_02_r}}
     {targetBone {boneName=upperarm_r}}
     {targetBone {boneName=clavicle_r}}
     {targetBone {boneName=spine_05}}
    ]
    constraintWeights [1,1,1,1,1]
   }
   n22 = "Constraint - base_TopSleeve_r (PPPPPP)"!(
    ComponentPose=@n21
    Alpha=1
    ConstraintWeights_0=0.7584
    ConstraintWeights_1:float=0.2264
    ConstraintWeights_2:float=0.0092
    ConstraintWeights_3:float=0.004
    ConstraintWeights_4:float=0.0014
    ConstraintWeights_5:float=0.0005
   ) -> (Pose) {
    boneToModify {boneName=base_TopSleeve_r}
    constraintSetup [
     {targetBone {boneName=upperarm_twist_01_r}}
     {targetBone {boneName=upperarm_twist_02_r}}
     {targetBone {boneName=clavicle_r}}
     {targetBone {boneName=upperarm_r}}
     {targetBone {boneName=deform_upperarm_fr_r}}
     {targetBone {boneName=deform_upperarm_up_r}}
    ]
    constraintWeights [1,1,1,1,1,1]
   }
   Save cached pose 'CRCAache_preDynamics'(Pose=Component To Local(ComponentPose=@n22)) {cacheName=CRCAache_preDynamics}
  ]}
 EventGraph {nodes [
   n1 = Event BlueprintUpdateAnimation()
   > K2Node_Composite_0 = WPO_Setup!(execute, InputPin:float=@n1.DeltaTimeX) {is=title K2Node_Composite}
   n2 = Event OnCharacterCustomizationLoaded()
   > Set tailMesh(tailMesh:Skeletal Mesh Component Object Reference=Get Array Item with EFort Custom Part Type(Array=@n2.CustomMeshParts, PartType=MiscOrTail)) -> (Output_Get:Skeletal Mesh Component Object Reference)
   > n3 = Set SkelMesh(SkelMesh=GetPartSkeletalMeshComponent())
   > n4 = Sequence()
   n5 = Get Dynamic Material Instance(execute=@n4.then_1, SkeletalMeshComponent=@n3)
   > Set Mat Instance Dynamic(Mat Instance Dynamic=@n5)
   > n6 = GetOrCreateDynamicOverlayMaterial(MeshComponent=@n3, MaterialName=None)
   > Set Overlay Mat Instance Dynamic(Overlay Mat Instance Dynamic=@n6)
   n7 = CheckForBackbling()
   Set isWearingCape(execute=@n4.then_0, isWearingCape=OR Boolean(A=@n7.bIsBackblingACape, B=@n7.bIsBackblingLong))
   > Set characterHasTail(characterHasTail:bool=CheckForTail!(self:Custom Character Part Anim Instance Object Reference) -> (bHasTail:bool)) -> (Output_Get:bool)
   > Branch(Condition=Get characterHasTail() -> (characterHasTail:bool))
   > n8 = Branch(Condition=Get isWearingCape() -> (isWearingCape:bool))
   n9 = Get tailMesh() -> (tailMesh:Skeletal Mesh Component Object Reference)
   > SetVisibility!(execute, self:Scene Component Object Reference=@n9, bNewVisibility:bool=false, bPropagateToChildren:bool=false) -> (then)
   SetVisibility(execute=@n8.else, self=@n9, bNewVisibility=true, bPropagateToChildren=false)
  ]}
 WPO_Setup {
  path=EventGraph.K2Node_Composite_0.WPO_Setup
  nodes [
   n1 = Inputs!() -> (execute, InputPin:float=~)
   > n2 = Branch(Condition=Get bIsInFrontEnd())
   Set WPO Globals Function(
    execute=@n2.else
    Material=Get Mat Instance Dynamic()
    Overlay Material=Get Overlay Mat Instance Dynamic()
    Skeletal Mesh=Get SkelMesh()
    "Local Noise Transform  Location"=Get Local Noise Transform Location() -> (Local Noise Transform Location:Vector)
    Local Noise Offset=Get Local Noise Offset()
   )
   > Set WPO Regional Function!(
    execute
    Material:Material Instance Dynamic Object Reference=Get Mat Instance Dynamic()
    Overlay Material:Material Instance Dynamic Object Reference=Get Overlay Mat Instance Dynamic()
    Skeletal Mesh:Skeletal Mesh Component Object Reference=Get SkelMesh()
    Velocity Transform:Name=spine_05
    Velocity Interp Speed:double=Get Velocity Interp Speed()
    Collision Transform:Name=root
    Collision Transform Axis:EAxis Enum
    Invert Collision Transform Axis:bool=false
    Delta Time X:double=@n1
    Previous Velocity Transform Location:"Vector (by ref)"=Get R1 Velocity Transform Location()
    Previous Velocity Transform Velocity:"Vector (by ref)"=Get R1 Velocity Transform Velocity()
    Region:int=1
   ) -> (then)
   > Set WPO Regional Function(
    Material=Get Mat Instance Dynamic()
    Overlay Material=Get Overlay Mat Instance Dynamic()
    Skeletal Mesh=Get SkelMesh()
    Velocity Transform=foot_l
    Velocity Interp Speed=Get Velocity Interp Speed()
    Collision Transform=calf_l
    Invert Collision Transform Axis=true
    Delta Time X=@n1
    Previous Velocity Transform Location=Get R1 Velocity Transform Location()
    Previous Velocity Transform Velocity=Get R1 Velocity Transform Velocity()
    Region=1
   )
   > Set WPO Regional Function(
    Material=Get Mat Instance Dynamic()
    Overlay Material=Get Overlay Mat Instance Dynamic()
    Skeletal Mesh=Get SkelMesh()
    Velocity Transform=foot_r
    Velocity Interp Speed=Get Velocity Interp Speed()
    Collision Transform=calf_r
    Invert Collision Transform Axis=true
    Delta Time X=@n1
    Previous Velocity Transform Location=Get R2 Velocity Transform Location()
    Previous Velocity Transform Velocity=Get R2 Velocity Transform Velocity()
    Region=2
   )
   > n3 = Branch(Condition=OR Boolean(A=Get bIsSkydiveFloating(), B=Get bIsParachuteOpened()))
   > Set WPO Pho Wind Intensity(WPO Pho Wind Intensity=0.8)
   > n4 = Set WPO Pho Turbulence(WPO Pho Turbulence:double=4.2) -> (Output_Get:double)
   SetScalarParameterValue(
    execute=[@n6.then, @n4.then]
    self=[Get Overlay Mat Instance Dynamic(), Get Mat Instance Dynamic()]
    ParameterName=ClothSim_WindStrength
    Value=Get WPO Pho Wind Intensity()
   )
   n5 = Get WPO Pho Turbulence() -> (WPO Pho Turbulence:double)
   > SetScalarParameterValue(self=[Get Overlay Mat Instance Dynamic(), Get Mat Instance Dynamic()], ParameterName=ClothSim_R1_TurbulenceStrength, Value=@n5)
   > SetScalarParameterValue(self=[Get Mat Instance Dynamic(), Get Overlay Mat Instance Dynamic()], ParameterName=ClothSim_R2_TurbulenceStrength, Value=@n5)
   Set WPO Pho Wind Intensity(execute=@n3.else, WPO Pho Wind Intensity=0)
   > n6 = Set WPO Pho Turbulence(WPO Pho Turbulence=0)
   Outputs()
  ]
 }
]
variables [
 characterHasTail {type=boolean}
 isWearingCape {type=boolean}
 Local Noise Offset {type=Vector}
 Local Noise Transform Location {type=Vector}
 Mat Instance Dynamic {type=MaterialInstanceDynamic}
 Overlay Mat Instance Dynamic {type=MaterialInstanceDynamic}
 PlayerPawn {type=FortPlayerPawnAthena}
 R1 Velocity Transform Location {type=Vector}
 R1 Velocity Transform Velocity {type=Vector}
 R2 Velocity Transform Location {type=Vector}
 R2 Velocity Transform Velocity {type=Vector}
 SkelMesh {type=SkeletalMeshComponent}
 tailMesh {type=SkeletalMeshComponent}
 Velocity Interp Speed {type=number}
 WPO Pho Turbulence {type=number}
 WPO Pho Wind Intensity {type=number}
]
