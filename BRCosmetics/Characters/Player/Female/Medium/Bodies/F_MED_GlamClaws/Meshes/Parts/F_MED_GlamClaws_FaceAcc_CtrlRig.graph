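# Control Rig for the GlamClaws face accessory that corrects long-hair bones as the character moves.
# A Variables pass derives weights from head swing/twist and a curve-remapped pelvis height.
# Hair Chains Dynamics Disable selects between offsetting base hair bones and, apparently for backbling, parent-constraining dynamic chains to head, neck and spine before pose offsets.
# Offsets blend by those weights, plus movement-dependent smoothing, crouch, targeting and a spherical pose reader.
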ControlRigBlueprint /BRCosmetics/Characters/Player/Female/Medium/Bodies/F_MED_GlamClaws/Meshes/Parts/F_MED_GlamClaws_FaceAcc_CtrlRig
graphs [
 CollapseNode_ContainedGraph {nodes [
   n1 = RigVMFunctionEntryNode()
   RigVMFunction_Sequence = &RigVMFunction_Sequence(ExecuteContext=@n1.ExecuteContext)
   DeltaRotation v4 = RigVMFunctionReferenceNode(
    Rotation Order=ZYX
    child=(Type=Bone,Name="head")
    debugRotation=false
    outputMinMax=(X=0,Y=1)
    parent=(Type=Null,Name="head_null")
    premult=1
    rotationMinMax=(X=0,Y=-20)
    triggerAxis=Z
    twistAxis=(X=1,Y=0,Z=0)
    ExecuteContext=@RigVMFunction_Sequence.A
   )
   > RigVMVariableNode(Value=@"DeltaRotation v4".outputWeight) {variable=HeadUpDelta}
   > DeltaRotation v4_1 = RigVMFunctionReferenceNode(
    Rotation Order=ZYX
    child=(Type=Bone,Name="head")
    debugRotation=false
    onlyTwist=true
    outputMinMax=(X=0,Y=1)
    parent=(Type=Null,Name="head_null")
    premult=1
    rotationMinMax=(X=-15,Y=-45)
    twistAxis=(X=1,Y=0,Z=0)
   )
   > RigVMVariableNode(Value=@"DeltaRotation v4_1".outputWeight) {variable=HeadRotXLeft}
   > DeltaRotation v4_3 = RigVMFunctionReferenceNode(
    Rotation Order=ZYX
    child=(Type=Bone,Name="head")
    debugRotation=false
    onlyTwist=true
    outputMinMax=(X=0,Y=1)
    parent=(Type=Null,Name="head_null")
    premult=1
    rotationMinMax=(X=40,Y=70)
    twistAxis=(X=1,Y=0,Z=0)
   )
   > RigVMVariableNode(Value=@"DeltaRotation v4_3".outputWeight) {variable=HeadRotXRight}
   > DeltaRotation v4_4 = RigVMFunctionReferenceNode(
    Rotation Order=ZYX
    child=(Type=Bone,Name="head")
    debugRotation=false
    outputMinMax=(X=0,Y=1)
    parent=(Type=Bone,Name="spine_05")
    premult=1
    rotationMinMax=(X=0,Y=20)
    triggerAxis=Z
    twistAxis=(X=1,Y=0,Z=0)
   )
   > RigVMVariableNode(Value=@"DeltaRotation v4_4".outputWeight) {variable=HeadDownDelta}
   > DeltaRotation v4_2 = RigVMFunctionReferenceNode(
    Rotation Order=ZYX
    child=(Type=Bone,Name="head")
    debugRotation=false
    onlyTwist=false
    outputMinMax=(X=0,Y=1)
    parent=(Type=Null,Name="head_null")
    premult=1
    rotationMinMax=(X=-20,Y=-35)
    triggerAxis=Z
    twistAxis=(X=1,Y=0,Z=0)
   )
   > RigVMVariableNode(Value=@"DeltaRotation v4_2".outputWeight) {variable=HeadUpExtDelta}
   RigUnit_GetRelativeTransformForItem = &RigUnit_GetRelativeTransformForItem(Child=(Type=Bone,Name="pelvis"), Parent=(Type=Bone,Name="root"), bChildInitial=False, bParentInitial=False)
   Remap = &RigVMFunction_MathDoubleRemap(
    SourceMaximum=30
    SourceMinimum=60
    TargetMaximum=0
    TargetMinimum=1
    bClamp=False
    Value=@RigUnit_GetRelativeTransformForItem.RelativeTransform.Translation.Z
   )
   n2 = &RigVMFunction_AnimEvalRichCurve(
    Curve="(EditorCurveData=(Keys=((InterpMode=RCIM_Cubic,TangentMode=RCTM_Break,TangentWeightMode=RCTWM_WeightedNone,Time=0,Value=0,ArriveTangent=0,ArriveTangentWeight=0,LeaveTangent=0,LeaveTangentWeight=0),(InterpMode=RCIM_Cubic,TangentMode=RCTM_Break,TangentWeight ...(truncated)"
    SourceMinimum=0
    TargetMinimum=0
    Value=@Remap.Result
   )
   > RigVMVariableNode(Value=@n2.Result) {variable=CloseGroundRemap}
   > DeltaRotation v4_5 = RigVMFunctionReferenceNode(
    Rotation Order=ZYX
    child=(Type=Bone,Name="head")
    debugRotation=false
    onlyTwist=true
    outputMinMax=(X=0,Y=1)
    parent=(Type=Null,Name="head_null")
    premult=1
    rotationMinMax=(X=10,Y=40)
    twistAxis=(X=1,Y=0,Z=0)
   )
   > RigVMVariableNode(Value=@"DeltaRotation v4_5".outputWeight) {variable=Head right twist}
   RigVMFunctionReturnNode(ExecuteContext=@RigVMFunction_Sequence.B)
  ]}
 CollapseNode_ContainedGraph {nodes [
   n1 = RigVMFunctionEntryNode()
   RigVMFunction_Sequence = &RigVMFunction_Sequence(ExecuteContext=@n1.ExecuteContext)
   n2 = RigVMVariableNode() {variable=HeadUpDelta}
   &RigUnit_ModifyTransforms(
    ItemToModify="((Item=(Type=Bone,Name=\"base_longhair_R_A\"),Transform=(Rotation=(X=0,Y=0,Z=-0.1937,W=0.9811),Translation=(X=0,Y=0,Z=0),Scale3D=(X=1,Y=1,Z=1))),(Item=(Type=Bone,Name=\"base_longhair_R_B\"),Transform=(Rotation=(X=0.2264,Y=0.066,Z=-0.2009,W=0.9508),Translation= ...(truncated)"
    WeightMinimum=0
    ExecutePin=@RigVMFunction_Sequence.A
    Weight=@n2.Value
   )
   Multiply = &RigVMFunction_MathDoubleMul(A=@n2.Value, B=RigVMVariableNode() {variable=HeadRotXLeft}.Value)
   > &RigUnit_ModifyTransforms(
    ItemToModify="((Item=(Type=Bone,Name=\"base_longhair_R_A\"),Transform=(Rotation=(X=0,Y=-0.1431,Z=0,W=0.9897),Translation=(X=0,Y=0,Z=0),Scale3D=(X=1,Y=1,Z=1))),(Item=(Type=Bone,Name=\"base_longhair_R_B\"),Transform=(Rotation=(X=0,Y=0.2279,Z=0,W=-0.9737),Translation=(X=0,Y=0, ...(truncated)"
    WeightMinimum=0
    Weight=@Multiply.Result
   )
   n3 = &RigVMFunction_MathDoubleMul(B=RigVMVariableNode() {variable=HeadUpExtDelta}.Value, A=RigVMVariableNode() {variable=CloseGroundRemap}.Value)
   n4 = &RigVMFunction_AlphaInterp(InterpSpeedDecreasing=25, InterpSpeedIncreasing=10, Scale=1, bClampResult=true, Value=@n3.Result)
   n5 = &RigVMFunction_AlphaInterp(InterpSpeedDecreasing=100, InterpSpeedIncreasing=1, Scale=1, bClampResult=true, Value=@n3.Result)
   > &RigUnit_ModifyTransforms(
    ItemToModify="((Item=(Type=Bone,Name=\"base_longhair_C\"),Transform=(Rotation=(X=0,Y=0,Z=0.1295,W=-0.9916),Translation=(X=5.0479,Y=-1.8969,Z=-0.0001),Scale3D=(X=1,Y=1,Z=1))),(Item=(Type=Bone,Name=\"base_longhair_L_A\"),Transform=(Rotation=(X=0,Y=0,Z=0.1305,W=-0.9914),Transl ...(truncated)"
    WeightMinimum=0
    Weight=RigVMDispatchNode(False=@n4.Result, True=@n5.Result, Condition=RigVMVariableNode() {variable=Is Moving 2D}.Value) {template="DISPATCH_RigVMDispatch_If(in Condition,in True,in False,out Result)"}.Result
   )
   RigVMFunctionReturnNode(ExecuteContext=@RigVMFunction_Sequence.B)
  ]}
 CollapseNode_ContainedGraph {nodes [
   n1 = RigVMFunctionEntryNode()
   RigVMFunction_Sequence = &RigVMFunction_Sequence(ExecuteContext=@n1.ExecuteContext)
   Constraints_BP = RigVMCollapseNode(ExecutePin=@RigVMFunction_Sequence.A)
   &RigUnit_ModifyTransforms(
    ItemToModify="((Item=(Type=Bone,Name=\"dyn_longhair_L_B_04\"),Transform=(Rotation=(X=0,Y=0,Z=0,W=1),Translation=(X=0,Y=0,Z=1.5),Scale3D=(X=1,Y=1,Z=1))),(Item=(Type=Bone,Name=\"dyn_longhair_L_A_04\"),Transform=(Rotation=(X=0,Y=0,Z=0,W=1),Translation=(X=-1.5,Y=-1.5,Z=0),Scale ...(truncated)"
    WeightMinimum=0
    ExecutePin=@Constraints_BP.ExecutePin
    Weight=RigVMVariableNode() {variable=HeadDownDelta}.Value
   )
   n2 = RigVMVariableNode() {variable=HeadUpDelta}
   > AlterTransform = RigVMFunctionReferenceNode(
    inheritsTransform=false
    transforms="((Item=(Type=Bone,Name=\"dyn_longhair_L_A_03\"),Transform=(Rotation=(X=0,Y=0,Z=0,W=1),Translation=(X=-2,Y=-1.55,Z=0),Scale3D=(X=1,Y=1,Z=1))),(Item=(Type=Bone,Name=\"dyn_longhair_C_03\"),Transform=(Rotation=(X=0,Y=0,Z=0,W=1),Translation=(X=-3,Y=1.5,Z=1),Scale3D ...(truncated)"
    weights=@n2.Value
   )
   > &RigUnit_ModifyTransforms(
    ItemToModify="((Item=(Type=Bone,Name=\"dyn_longhair_L_A_04\"),Transform=(Rotation=(X=0,Y=0,Z=0,W=1),Translation=(X=-1,Y=0,Z=0),Scale3D=(X=1,Y=1,Z=1))),(Item=(Type=Bone,Name=\"dyn_longhair_C_04\"),Transform=(Rotation=(X=0,Y=0,Z=-0.5,W=0.866),Translation=(X=0,Y=0,Z=0),Scale3D ...(truncated)"
    WeightMinimum=0
    Weight=@n2.Value
   )
   n3 = &RigVMFunction_MathDoubleMul(A=&RigVMFunction_MathBoolToFloat(Value=RigVMVariableNode() {variable=IS Crouching}.Value).Result, B=@n2.Value)
   > AlterTransform_1 = RigVMFunctionReferenceNode(
    inheritsTransform=false
    transforms="((Item=(Type=Bone,Name=\"dyn_longhair_C_02\"),Transform=(Rotation=(X=0,Y=0,Z=0.1736,W=0.9848),Translation=(X=-1,Y=0,Z=0),Scale3D=(X=1,Y=1,Z=1))),(Item=(Type=Bone,Name=\"dyn_longhair_C_03\"),Transform=(Rotation=(X=0,Y=0,Z=0.1736,W=0.9848),Translation=(X=-1,Y=0, ...(truncated)"
    weights=@n3.Result
   )
   > AlterTransform_2 = RigVMFunctionReferenceNode(
    inheritsTransform=false
    transforms="((Item=(Type=Bone,Name=\"dyn_longhair_L_A_02\"),Transform=(Rotation=(X=0,Y=0,Z=0.1736,W=0.9848),Translation=(X=-1,Y=0,Z=0),Scale3D=(X=1,Y=1,Z=1))),(Item=(Type=Bone,Name=\"dyn_longhair_L_A_03\"),Transform=(Rotation=(X=0,Y=0,Z=0.1736,W=0.9848),Translation=(X=-1, ...(truncated)"
    weights=@n3.Result
   )
   > AlterTransform_3 = RigVMFunctionReferenceNode(
    inheritsTransform=false
    transforms="((Item=(Type=Bone,Name=\"dyn_longhair_R_A_02\"),Transform=(Rotation=(X=0,Y=0,Z=0.1736,W=0.9848),Translation=(X=-2,Y=0,Z=0),Scale3D=(X=1,Y=1,Z=1))),(Item=(Type=Bone,Name=\"dyn_longhair_R_A_03\"),Transform=(Rotation=(X=0,Y=0,Z=0.1736,W=0.9848),Translation=(X=-2, ...(truncated)"
    weights=@n3.Result
   )
   > n4 = &RigUnit_SphericalPoseReader(
    ActiveRegionScaleFactors=(PositiveWidth=1,NegativeWidth=1,PositiveHeight=1,NegativeHeight=1)
    ActiveRegionSize=0.05
    Debug=(bDrawDebug=false,bDraw2D=False,bDrawLocalAxes=False,DebugScale=25,DebugSegments=20,DebugThickness=0.25)
    DriverAxis=(X=1,Y=0,Z=0)
    DriverItem=(Type=Bone,Name="head")
    FalloffRegionScaleFactors=(PositiveWidth=1,NegativeWidth=1,PositiveHeight=1,NegativeHeight=1)
    FalloffSize=0.2
    FlipHeightScaling=False
    FlipWidthScaling=False
    OptionalParentItem=(Type=Bone,Name="spine_05")
    RotationOffset=(X=0,Y=-65,Z=-52)
   )
   Multiply_1 = RigVMAggregateNode(C=1, A=@n4.OutputParam, B=RigVMVariableNode() {variable=Is Targeting}.Value)
   n5 = RigVMVariableNode() {variable=IS Crouching}
   > &RigUnit_ModifyTransforms(
    ItemToModify="((Item=(Type=Bone,Name=\"dyn_longhair_R_A_02\"),Transform=(Rotation=(X=0,Y=0,Z=0,W=1),Translation=(X=-0.2971,Y=0,Z=0),Scale3D=(X=1.0802,Y=1.0802,Z=1.0802))),(Item=(Type=Bone,Name=\"dyn_longhair_R_A_03\"),Transform=(Rotation=(X=0,Y=0,Z=0,W=1),Translation=(X=-2. ...(truncated)"
    WeightMinimum=0
    Weight=RigVMDispatchNode(True=0, False=@Multiply_1.Result, Condition=@n5.Value) {template="DISPATCH_RigVMDispatch_If(in Condition,in True,in False,out Result)"}.Result
   )
   > &RigUnit_ModifyTransforms(
    ItemToModify="((Item=(Type=Bone,Name=\"dyn_longhair_L_B_03\"),Transform=(Rotation=(X=0.0005,Y=0.2173,Z=0.0567,W=-0.9745),Translation=(X=-1.0022,Y=0,Z=-0.215),Scale3D=(X=0.9182,Y=0.9182,Z=0.9182))),(Item=(Type=Bone,Name=\"dyn_longhair_L_A_04\"),Transform=(Rotation=(X=0,Y=0,Z ...(truncated)"
    WeightMinimum=0
    Weight=&RigVMFunction_MathDoubleMul(B=&RigVMFunction_MathBoolToFloat(Value=@n5.Value).Result, A=@Multiply_1.Result).Result
   )
   Multiply_4 = RigVMAggregateNode(
    C=RigVMVariableNode() {variable=CloseGroundRemap}.Value
    A=RigVMVariableNode() {variable=HeadUpExtDelta}.Value
    B=RigVMVariableNode() {variable=Is Targeting}.Value
   )
   > &RigUnit_ModifyTransforms(
    ItemToModify="((Item=(Type=Bone,Name=\"None\"),Transform=(Rotation=(X=0,Y=0,Z=0,W=1),Translation=(X=0,Y=0,Z=0),Scale3D=(X=1,Y=1,Z=1))),(Item=(Type=Bone,Name=\"dyn_longhair_C_02\"),Transform=(Rotation=(X=0,Y=0,Z=0,W=1),Translation=(X=0,Y=0,Z=0),Scale3D=(X=0.85,Y=0.85,Z=0.85) ...(truncated)"
    WeightMinimum=0
    Weight=@Multiply_4.Result
   )
   > &RigUnit_ModifyTransforms(
    ItemToModify="((Item=(Type=Bone,Name=\"dyn_longhair_C_03\"),Transform=(Rotation=(X=-0.0112,Y=-0.1471,Z=0.1223,W=0.9815),Translation=(X=-0.5206,Y=-0.1367,Z=-0.1606),Scale3D=(X=1,Y=1,Z=1))),(Item=(Type=Bone,Name=\"dyn_longhair_L_A_03\"),Transform=(Rotation=(X=0,Y=0,Z=0,W=1),T ...(truncated)"
    WeightMinimum=0
    Weight=RigVMVariableNode() {variable=HeadRotXLeft}.Value
   )
   > &RigUnit_ModifyTransforms(
    ItemToModify="((Item=(Type=Bone,Name=\"dyn_longhair_C_04\"),Transform=(Rotation=(X=0,Y=0.0912,Z=0,W=0.9958),Translation=(X=-0.9722,Y=0,Z=-9.3707),Scale3D=(X=0.8038,Y=0.8038,Z=0.8038))),(Item=(Type=Bone,Name=\"dyn_longhair_L_A_04\"),Transform=(Rotation=(X=0,Y=0,Z=0,W=-1),Tra ...(truncated)"
    WeightMinimum=0
    Weight=RigVMVariableNode() {variable=HeadRotXRight}.Value
   )
   > &RigUnit_ModifyTransforms(
    ItemToModify="((Item=(Type=Bone,Name=\"dyn_longhair_L_A_04\"),Transform=(Rotation=(X=0,Y=0,Z=0,W=-1),Translation=(X=-1.5385,Y=0,Z=0),Scale3D=(X=1,Y=1,Z=1))),(Item=(Type=Bone,Name=\"dyn_longhair_L_B_04\"),Transform=(Rotation=(X=0,Y=0,Z=0,W=1),Translation=(X=0,Y=0,Z=0),Scale3 ...(truncated)"
    WeightMinimum=0
    Weight=RigVMVariableNode() {variable=Head right twist}.Value
   )
   RigVMFunctionReturnNode(ExecuteContext=@RigVMFunction_Sequence.B)
  ]}
 CollapseNode_ContainedGraph {nodes [
   n1 = RigVMFunctionEntryNode()
   &RigUnit_ParentConstraint(
    AdvancedSettings=(InterpolationType=Average,RotationOrderForFilter=XZY)
    Child=(Type=Bone,Name="dyn_longhair_C_01")
    Parents=((Item=(Type=Bone,Name="head"),Weight=1))
    Weight=1
    ExecutePin=@n1.ExecutePin
   )
   > &RigUnit_ParentConstraint(
    AdvancedSettings=(InterpolationType=Average,RotationOrderForFilter=XZY)
    Child=(Type=Bone,Name="dyn_longhair_C_02")
    Parents=((Item=(Type=Bone,Name="head"),Weight=1))
    Weight=1
   )
   > &RigUnit_ParentConstraint(
    AdvancedSettings=(InterpolationType=Average,RotationOrderForFilter=XZY)
    Child=(Type=Bone,Name="dyn_longhair_C_03")
    Parents=((Item=(Type=Bone,Name="head"),Weight=0.9505),(Item=(Type=Bone,Name="neck_02"),Weight=0.0491),(Item=(Type=Bone,Name="neck_01"),Weight=0.0004))
    Weight=1
   )
   > &RigUnit_ParentConstraint(
    AdvancedSettings=(InterpolationType=Average,RotationOrderForFilter=XZY)
    Child=(Type=Bone,Name="dyn_longhair_C_04")
    Parents=((Item=(Type=Bone,Name="head"),Weight=0.0067),(Item=(Type=Bone,Name="neck_02"),Weight=0.0911),(Item=(Type=Bone,Name="neck_01"),Weight=0.1601),(Item=(Type=Bone,Name="spine_05"),Weight=0.7359),(Item=(Type=Bone,Name="spine_04"),Weight=0.0062))
    Weight=1
   )
   > &RigUnit_ParentConstraint(
    AdvancedSettings=(InterpolationType=Average,RotationOrderForFilter=XZY)
    Child=(Type=Bone,Name="dyn_longhair_L_A_01")
    Parents=((Item=(Type=Bone,Name="head"),Weight=1))
    Weight=1
   )
   > &RigUnit_ParentConstraint(
    AdvancedSettings=(InterpolationType=Average,RotationOrderForFilter=XZY)
    Child=(Type=Bone,Name="dyn_longhair_L_A_02")
    Parents=((Item=(Type=Bone,Name="head"),Weight=0.9995),(Item=(Type=Bone,Name="neck_02"),Weight=0.0005))
    Weight=1
   )
   > &RigUnit_ParentConstraint(
    AdvancedSettings=(InterpolationType=Average,RotationOrderForFilter=XZY)
    Child=(Type=Bone,Name="dyn_longhair_L_A_03")
    Parents=((Item=(Type=Bone,Name="head"),Weight=0.9435),(Item=(Type=Bone,Name="neck_02"),Weight=0.056),(Item=(Type=Bone,Name="neck_01"),Weight=0.0005))
    Weight=1
   )
   > &RigUnit_ParentConstraint(
    AdvancedSettings=(InterpolationType=Average,RotationOrderForFilter=XZY)
    Child=(Type=Bone,Name="dyn_longhair_L_A_04")
    Parents=((Item=(Type=Bone,Name="head"),Weight=0.0041),(Item=(Type=Bone,Name="neck_02"),Weight=0.1094),(Item=(Type=Bone,Name="neck_01"),Weight=0.4722),(Item=(Type=Bone,Name="spine_05"),Weight=0.4057),(Item=(Type=Bone,Name="spine_04"),Weight=0.0086))
    Weight=1
   )
   > &RigUnit_ParentConstraint(
    AdvancedSettings=(InterpolationType=Average,RotationOrderForFilter=XZY)
    Child=(Type=Bone,Name="dyn_longhair_L_B_01")
    Parents=((Item=(Type=Bone,Name="head"),Weight=1))
    Weight=1
   )
   > &RigUnit_ParentConstraint(
    AdvancedSettings=(InterpolationType=Average,RotationOrderForFilter=XZY)
    Child=(Type=Bone,Name="dyn_longhair_L_B_02")
    Parents=((Item=(Type=Bone,Name="head"),Weight=1))
    Weight=1
   )
   > &RigUnit_ParentConstraint(
    AdvancedSettings=(InterpolationType=Average,RotationOrderForFilter=XZY)
    Child=(Type=Bone,Name="dyn_longhair_L_B_03")
    Parents=((Item=(Type=Bone,Name="head"),Weight=0.9878),(Item=(Type=Bone,Name="neck_02"),Weight=0.0122))
    Weight=1
   )
   > &RigUnit_ParentConstraint(
    AdvancedSettings=(InterpolationType=Average,RotationOrderForFilter=XZY)
    Child=(Type=Bone,Name="dyn_longhair_L_A_04")
    Parents=((Item=(Type=Bone,Name="head"),Weight=0.8144),(Item=(Type=Bone,Name="neck_02"),Weight=0.1826),(Item=(Type=Bone,Name="neck_01"),Weight=0.003))
    Weight=1
   )
   > &RigUnit_ParentConstraint(
    AdvancedSettings=(InterpolationType=Average,RotationOrderForFilter=XZY)
    Child=(Type=Bone,Name="dyn_longhair_R_A_01")
    Parents=((Item=(Type=Bone,Name="head"),Weight=1))
    Weight=1
   )
   > &RigUnit_ParentConstraint(
    AdvancedSettings=(InterpolationType=Average,RotationOrderForFilter=XZY)
    Child=(Type=Bone,Name="dyn_longhair_R_A_02")
    Parents=((Item=(Type=Bone,Name="head"),Weight=0.9995),(Item=(Type=Bone,Name="neck_02"),Weight=0.0005))
    Weight=1
   )
   > &RigUnit_ParentConstraint(
    AdvancedSettings=(InterpolationType=Average,RotationOrderForFilter=XZY)
    Child=(Type=Bone,Name="dyn_longhair_R_A_03")
    Parents=((Item=(Type=Bone,Name="head"),Weight=0.9435),(Item=(Type=Bone,Name="neck_02"),Weight=0.056),(Item=(Type=Bone,Name="neck_01"),Weight=0.0005))
    Weight=1
   )
   > n2 = &RigUnit_ParentConstraint(
    AdvancedSettings=(InterpolationType=Average,RotationOrderForFilter=XZY)
    Child=(Type=Bone,Name="dyn_longhair_R_A_04")
    Parents=((Item=(Type=Bone,Name="head"),Weight=0.004),(Item=(Type=Bone,Name="neck_02"),Weight=0.1084),(Item=(Type=Bone,Name="neck_01"),Weight=0.4716),(Item=(Type=Bone,Name="spine_05"),Weight=0.407),(Item=(Type=Bone,Name="spine_04"),Weight=0.004))
    Weight=1
   )
   RigVMFunctionReturnNode(ExecutePin=@n2.ExecutePin)
  ]}
 Multiply_1_ContainedGraph {nodes [
   n1 = RigVMFunctionEntryNode()
   Multiply = &RigVMFunction_MathDoubleMul(A=@n1.A, B=@n1.B)
   RigVMFunctionReturnNode(Result=&RigVMFunction_MathDoubleMul(B=@n1.C, A=@Multiply.Result).Result)
  ]}
 Multiply_5_ContainedGraph {nodes [
   n1 = RigVMFunctionEntryNode()
   RigVMFunctionReturnNode(Result=&RigVMFunction_MathDoubleMul(B=@n1.C, A=&RigVMFunction_MathDoubleMul(A=@n1.A, B=@n1.B).Result).Result)
  ]}
 New Function_ContainedGraph {nodes [
   n1 = RigVMFunctionEntryNode()
   RigVMFunction_MathBoolToInteger = &RigVMFunction_MathBoolToInteger(Value=@n1.onlyTwist)
   RigVMFunction_MathQuaternionSwingTwist = &RigVMFunction_MathQuaternionSwingTwist(
    TwistAxis=@n1.twistAxis
    Input=&RigUnit_GetRelativeTransformForItem(bChildInitial=False, bParentInitial=False, Child=@n1.child, Parent=@n1.parent).RelativeTransform.Rotation
   )
   n2 = RigVMDispatchNode(
    Values=((),())
    Index=@RigVMFunction_MathBoolToInteger.Result
    Values.0=@RigVMFunction_MathQuaternionSwingTwist.Swing
    Values.1=@RigVMFunction_MathQuaternionSwingTwist.Twist
   ) {template="DISPATCH_RigVMDispatch_SelectInt32(in Index,in Values,out Result)"}
   RigVMFunction_MathQuaternionToEuler = &RigVMFunction_MathQuaternionToEuler(Value=@n2.Result, RotationOrder=@n1.Rotation Order)
   > DISPATCH_RigVMDispatch_Print = RigVMDispatchNode(ScreenColor=(R=1,G=1,B=1,A=1), ScreenDuration=1.5, Enabled=@n1.debugRotation, Value=@RigVMFunction_MathQuaternionToEuler.Result) {template="DISPATCH_RigVMDispatch_Print(in Prefix,in Value,in Enabled,in ScreenDuration,in ScreenColor)"}
   Cast = RigVMDispatchNode(Value=@n1.triggerAxis) {template="DISPATCH_RigVMDispatch_CastEnumToInt(in Value,out Result)"}
   DISPATCH_RigVMDispatch_SelectInt32 = RigVMDispatchNode(
    Values=(0,0,0)
    Index=@Cast.Result
    Values.0=@RigVMFunction_MathQuaternionToEuler.Result.X
    Values.1=@RigVMFunction_MathQuaternionToEuler.Result.Y
    Values.2=@RigVMFunction_MathQuaternionToEuler.Result.Z
   ) {template="DISPATCH_RigVMDispatch_SelectInt32(in Index,in Values,out Result)"}
   Remap = &RigVMFunction_MathDoubleRemap(
    bClamp=False
    Value=@DISPATCH_RigVMDispatch_SelectInt32.Result
    TargetMinimum=@n1.outputMinMax.X
    TargetMaximum=@n1.outputMinMax.Y
    SourceMinimum=@n1.rotationMinMax.X
    SourceMaximum=@n1.rotationMinMax.Y
   )
   Absolute = &RigVMFunction_MathDoubleAbs(Value=@Remap.Result)
   Multiply = &RigVMFunction_MathDoubleMul(
    A=&RigVMFunction_MathDoubleClamp(Value=RigVMDispatchNode(True=@Absolute.Result, Condition=@n1.absoluteValue, False=@Remap.Result) {template="DISPATCH_RigVMDispatch_If(in Condition,in True,in False,out Result)"}.Result).Result
    B=@n1.premult
   )
   > RigVMFunctionReturnNode(outputQuat=@n2.Result, outputWeight=@Multiply.Result)
  ]}
 New Function_ContainedGraph {nodes [
   n1 = RigVMFunctionEntryNode()
   > For_Each = RigVMDispatchNode(Array=@n1.transforms) {template="DISPATCH_RigVMDispatch_ArrayIterator(in Array,out Element,out Index,out Count,out Ratio)"}
   > &RigUnit_OffsetTransformForItem(Item=@For_Each.Element.Item, OffsetTransform=@For_Each.Element.Transform, Weight=@n1.weights, bPropagateToChildren=@n1.inheritsTransform)
   RigVMFunctionReturnNode(ExecuteContext=@For_Each.Completed)
  ]}
 RigVMFunctionLibrary {nodes [
   AlterTransform = RigVMCollapseNode(ExecuteContext=(), inheritsTransform=false, transforms=(), weights=1)
   DeltaRotation v4 = RigVMCollapseNode(
    ExecuteContext=()
    Rotation Order=ZYX
    child=(Type=Bone,Name="None")
    outputMinMax=(X=0,Y=1)
    parent=(Type=Bone,Name="None")
    premult=1
    rotationMinMax=(X=0,Y=0)
    twistAxis=(X=1,Y=0,Z=0)
   )
  ]}
 RigVMModel {nodes [
   RigUnit_BeginExecution = &RigUnit_BeginExecution()
   Variables = RigVMCollapseNode(ExecuteContext=@RigUnit_BeginExecution.ExecutePin)
   RigVMFunction_ControlFlowBranch = &RigVMFunction_ControlFlowBranch(Condition=RigVMVariableNode() {variable=Hair Chains Dynamics Disable}.Value, ExecuteContext=@Variables.ExecuteContext)
   Hair_dynamics_chains_fix = RigVMCollapseNode(ExecuteContext=@RigVMFunction_ControlFlowBranch.False)
   Hair_fix_with_backbling_on = RigVMCollapseNode(ExecuteContext=@RigVMFunction_ControlFlowBranch.True)
  ]}
]
hierarchy [
 root {loc=<0 0 0>, rot=<0 0 0 1>}
 pelvis {parent=root, loc=<0 -0.5301 98.0571>, rot=<0.0127 0.707 -0.0127 -0.707>}
 spine_01 {parent=pelvis, loc=<6.1208 -0.9328 0>, rot=<0 0 0.034 -0.9994>}
 spine_02 {parent=spine_01, loc=<6.5997 -0.2095 0>, rot=<0 0 0.0238 -0.9997>}
 spine_03 {parent=spine_02, loc=<6.4914 -0.1237 0>, rot=<0 0 0.0149 -0.9999>}
 spine_04 {parent=spine_03, loc=<6.9598 0.0794 0>, rot=<0 0 -0.0132 -0.9999>}
 spine_05 {parent=spine_04, loc=<6.9987 0 0>, rot=<0 0 -0.0015 -1>}
 neck_01 {parent=spine_05, loc=<11.3138 0.1529 0>, rot=<0 0 -0.2003 -0.9797>}
 neck_02 {parent=neck_01, loc=<4.5627 0 0>, rot=<0 0 0.0504 -0.9987>}
 head {parent=neck_02, loc=<5.3914 0 0>, rot=<0 0 0.115 -0.9934>}
 base_ear_R {parent=head, loc=<10.2065 2.4081 5.7489>, rot=<0.6992 0.4679 -0.0619 -0.5369>}
 dyn_ear_R_01 {parent=base_ear_R, loc=<0 0 0>, rot=<0 0 0 1>}
 dyn_ear_R_02 {parent=dyn_ear_R_01, loc=<6.4394 0 0>, rot=<-0.0024 -0.0713 -0.033 -0.9969>}
 base_ear_L {parent=head, loc=<9.5954 2.8223 -5.6833>, rot=<-0.5029 -0.6005 -0.2642 -0.5628>}
 dyn_ear_L_01 {parent=base_ear_L, loc=<0 0 0>, rot=<0 0 0 1>}
 dyn_ear_L_02 {parent=dyn_ear_L_01, loc=<6.0623 0 0.7581>, rot=<0 0.0242 0 -0.9997>}
 base_earring_A {parent=dyn_ear_L_02, loc=<6.2407 -0.7836 -0.6592>, rot=<0.4942 -0.7134 -0.4325 -0.2443>}
 dyn_earring_A {parent=base_earring_A, loc=<0 0 0>, rot=<0 0 0 1>}
 base_earring_B {parent=dyn_ear_L_02, loc=<3.5137 -1.4116 -0.986>, rot=<0.6439 -0.5668 -0.3715 -0.3552>}
 dyn_earring_B {parent=base_earring_B, loc=<0 0 0>, rot=<0 0 0 1>}
 base_longhair_C {parent=head, loc=<13.7284 0.1748 0>, rot=<0.6559 -0.7549 0 0>}
 dyn_longhair_C_01 {parent=base_longhair_C, loc=<3.7194 0 0>, rot=<0.9366 0.3503 0 0>}
 dyn_longhair_C_02 {parent=dyn_longhair_C_01, loc=<7.3997 0 0>, rot=<0 0 0.2595 -0.9657>}
 dyn_longhair_C_03 {parent=dyn_longhair_C_02, loc=<7.5316 0 0>, rot=<0 0 0.0191 -0.9998>}
 dyn_longhair_C_04 {parent=dyn_longhair_C_03, loc=<7.5367 0 0>, rot=<0 0 -0.2873 -0.9578>}
 base_longhair_L_A {parent=head, loc=<14.2989 -0.9678 -0.9605>, rot=<0.417 -0.837 -0.3184 -0.1553>}
 dyn_longhair_L_A_01 {parent=base_longhair_L_A, loc=<5.533 0 0>, rot=<0 0 -0.2688 -0.9632>}
 dyn_longhair_L_A_02 {parent=dyn_longhair_L_A_01, loc=<5.7486 0 0>, rot=<0.9938 0.1111 0 0>}
 dyn_longhair_L_A_03 {parent=dyn_longhair_L_A_02, loc=<5.7664 0 0>, rot=<0.9999 -0.0107 0 0>}
 dyn_longhair_L_A_04 {parent=dyn_longhair_L_A_03, loc=<5.6823 0 0>, rot=<0 0 0.3325 -0.9431>}
 base_longhair_L_B {parent=head, loc=<14.2923 -0.1733 -1.4202>, rot=<0.3656 -0.6833 -0.5597 -0.2936>}
 dyn_longhair_L_B_01 {parent=base_longhair_L_B, loc=<4.8354 0 0>, rot=<0 0 -0.3206 -0.9472>}
 dyn_longhair_L_B_02 {parent=dyn_longhair_L_B_01, loc=<4.9872 0 0>, rot=<0.9944 0.106 0 0>}
 dyn_longhair_L_B_03 {parent=dyn_longhair_L_B_02, loc=<5.0002 0 0>, rot=<0.9995 -0.0321 0 0>}
 dyn_longhair_L_B_04 {parent=dyn_longhair_L_B_03, loc=<4.9965 0 0>, rot=<0 0 0.3598 -0.933>}
 base_longhair_R_A {parent=head, loc=<14.2993 -0.9678 0.9605>, rot=<0.417 -0.837 0.3184 0.1553>}
 dyn_longhair_R_A_01 {parent=base_longhair_R_A, loc=<5.5331 0 0>, rot=<0 0 -0.2688 -0.9632>}
 dyn_longhair_R_A_02 {parent=dyn_longhair_R_A_01, loc=<5.749 0 0>, rot=<0.9938 0.111 0 0>}
 dyn_longhair_R_A_03 {parent=dyn_longhair_R_A_02, loc=<5.7662 0 0>, rot=<0.9999 -0.0107 0 0>}
 dyn_longhair_R_A_04 {parent=dyn_longhair_R_A_03, loc=<2.5885 0 0>, rot=<0 0 0.3325 -0.9431>}
 base_longhair_R_B {parent=head, loc=<14.2927 -0.1733 1.4202>, rot=<0.3655 -0.6833 0.5598 0.2935>}
 dyn_longhair_R_B_01 {parent=base_longhair_R_B, loc=<4.8359 0 0>, rot=<0 0 -0.3205 -0.9473>}
 dyn_longhair_R_B_02 {parent=dyn_longhair_R_B_01, loc=<4.9868 0 0>, rot=<0 0 -0.106 -0.9944>}
 dyn_longhair_R_B_03 {parent=dyn_longhair_R_B_02, loc=<4.9998 0 0>, rot=<0 0 -0.0321 -0.9995>}
 dyn_longhair_R_B_04 {parent=dyn_longhair_R_B_03, loc=<4.9971 0 0>, rot=<0 0 0.3598 -0.933>}
 base_fronthair_C {parent=head, loc=<12.6326 7.8007 0.3501>, rot=<-0.0323 0.0276 0.9583 0.2825>}
 dyn_fronthair_C_01 {parent=base_fronthair_C, loc=<0 0 0>, rot=<0 0 0 1>}
 dyn_fronthair_C_02 {parent=dyn_fronthair_C_01, loc=<4.2384 0 0>, rot=<0.0192 0.0849 -0.22 -0.9716>}
 base_fronthair_L {parent=head, loc=<13.5772 5.6652 -1.4587>, rot=<-0.1988 -0.9454 -0.0712 -0.2482>}
 dyn_fronthair_L_01 {parent=base_fronthair_L, loc=<0 0 0>, rot=<0 0 0 1>}
 dyn_fronthair_L_02 {parent=dyn_fronthair_L_01, loc=<6.6672 0 0>, rot=<-0.0369 0.0974 0.3518 -0.9302>}
 base_fronthair_R {parent=head, loc=<12.6929 6.3284 3.5746>, rot=<0.3385 -0.4539 0.8094 0.1555>}
 dyn_fronthair_R_01 {parent=base_fronthair_R, loc=<0 0 0>, rot=<0 0 0 1>}
 dyn_fronthair_R_02 {parent=dyn_fronthair_R_01, loc=<4.1355 0 -0.0907>, rot=<0 0.0616 0 -0.9981>}
 base_backhair_A {parent=head, loc=<11.2124 -0.8478 -8.4149>, rot=<0.4965 -0.6917 -0.5064 -0.1363>}
 dyn_backhair_A {parent=base_backhair_A, loc=<0 0 0>, rot=<0 0 0 1>}
 base_backhair_B {parent=head, loc=<11.8087 -3.8929 -8.3666>, rot=<0.6885 -0.4961 -0.4543 -0.2711>}
 dyn_backhair_B {parent=base_backhair_B, loc=<0 0 0>, rot=<0 0 0 1>}
 base_backhair_C {parent=head, loc=<12.2582 -8.8743 -3.6184>, rot=<0.5554 -0.8107 -0.1661 -0.0818>}
 dyn_backhair_C {parent=base_backhair_C, loc=<0 0 0>, rot=<0 0 0 1>}
 base_backhair_D {parent=head, loc=<11.4052 -8.5301 4.0382>, rot=<0.712 -0.6768 0.1213 0.1422>}
 dyn_backhair_D {parent=base_backhair_D, loc=<0 0 0>, rot=<0 0 0 1>}
 base_backhair_E {parent=head, loc=<13.9324 -5.2924 6.2502>, rot=<0.8029 -0.5004 0.1511 0.2864>}
 dyn_backhair_E {parent=base_backhair_E, loc=<0 0 0>, rot=<0 0 0 1>}
 base_backhair_F {parent=head, loc=<11.257 -4.44 9.0949>, rot=<0.4592 -0.7745 0.4349 -0.0094>}
 dyn_backhair_F {parent=base_backhair_F, loc=<0 0 0>, rot=<0 0 0 1>}
 base_backhair_middle_A {parent=head, loc=<4.4111 -0.0808 -10.4652>, rot=<-0.3935 0.456 0.6842 0.4112>}
 dyn_backhair_middle_A {parent=base_backhair_middle_A, loc=<0 0 0>, rot=<0 0 0 1>}
 base_backhair_middle_B {parent=head, loc=<1.3855 -4.1141 -10.5077>, rot=<0.6892 -0.3668 -0.473 -0.4083>}
 dyn_backhair_middle_B {parent=base_backhair_middle_B, loc=<0 0 0>, rot=<0 0 0 1>}
 base_backhair_middle_C {parent=head, loc=<3.7435 -7.1105 -8.0327>, rot=<0.6258 -0.6081 -0.4288 -0.2338>}
 dyn_backhair_middle_C {parent=base_backhair_middle_C, loc=<0 0 0>, rot=<0 0 0 1>}
 base_backhair_middle_D {parent=head, loc=<5.1789 -9.5409 -6.5288>, rot=<0.6495 -0.7531 -0.0985 0.0364>}
 dyn_backhair_middle_D {parent=base_backhair_middle_D, loc=<0 0 0>, rot=<0 0 0 1>}
 base_backhair_middle_E {parent=head, loc=<4.1376 -12.8457 -3.313>, rot=<0.5304 -0.8215 -0.0781 -0.194>}
 dyn_backhair_middle_E {parent=base_backhair_middle_E, loc=<0 0 0>, rot=<0 0 0 1>}
 base_backhair_middle_F {parent=head, loc=<4.1707 -12.0336 4.5811>, rot=<0.6674 -0.7053 0.2178 0.0984>}
 dyn_backhair_middle_F {parent=base_backhair_middle_F, loc=<0 0 0>, rot=<0 0 0 1>}
 base_backhair_middle_G {parent=head, loc=<7.6948 -5.824 7.499>, rot=<0.5312 -0.758 0.3497 0.1449>}
 dyn_backhair_middle_G {parent=base_backhair_middle_G, loc=<0 0 0>, rot=<0 0 0 1>}
 base_backhair_middle_H {parent=head, loc=<4.1513 -1.2012 11.1035>, rot=<0.5585 -0.5501 0.4591 0.418>}
 dyn_backhair_middle_H {parent=base_backhair_middle_H, loc=<0 0 0>, rot=<0 0 0 1>}
 base_backhair_middle_I {parent=head, loc=<-3.4807 -6.2827 -9.5784>, rot=<0.7181 -0.6955 0.0231 0.0122>}
 dyn_backhair_middle_I {parent=base_backhair_middle_I, loc=<0 0 0>, rot=<0 0 0 1>}
 base_backhair_middle_J {parent=head, loc=<-1.0981 -10.6534 -7.2344>, rot=<0.6961 -0.5688 -0.3708 -0.2334>}
 dyn_backhair_middle_J {parent=base_backhair_middle_J, loc=<0 0 0>, rot=<0 0 0 1>}
 base_backhair_middle_K {parent=head, loc=<1.1827 -12.0349 -4.4596>, rot=<0.6844 -0.6826 -0.2377 -0.0952>}
 dyn_backhair_middle_K {parent=base_backhair_middle_K, loc=<0 0 0>, rot=<0 0 0 1>}
 base_backhair_middle_L {parent=head, loc=<-0.2387 -12.9198 2.4903>, rot=<0.671 -0.7091 0.1429 0.163>}
 dyn_backhair_middle_L {parent=base_backhair_middle_L, loc=<0 0 0>, rot=<0 0 0 1>}
 base_backhair_middle_M {parent=head, loc=<1.0254 -9.8507 7.0679>, rot=<0.6502 -0.6444 0.2615 0.3059>}
 dyn_backhair_middle_M {parent=base_backhair_middle_M, loc=<0 0 0>, rot=<0 0 0 1>}
 base_backhair_middle_N {parent=head, loc=<1.191 -4.9411 9.6208>, rot=<0.4966 -0.4881 0.5757 0.4286>}
 dyn_backhair_middle_N {parent=base_backhair_middle_N, loc=<0 0 0>, rot=<0 0 0 1>}
 base_backhair_middle_O {parent=head, loc=<-0.7009 -2.5572 10.8543>, rot=<0.6225 -0.5386 0.533 0.1957>}
 dyn_backhair_middle_O {parent=base_backhair_middle_O, loc=<0 0 0>, rot=<0 0 0 1>}
 base_backhair_bot_A {parent=head, loc=<-5.051 -10.1362 -6.5285>, rot=<0.6159 -0.6684 -0.3028 -0.2868>}
 dyn_backhair_bot_A {parent=base_backhair_bot_A, loc=<0 0 0>, rot=<0 0 0 1>}
 base_backhair_bot_B {parent=head, loc=<-5.6464 -12.7654 -1.3894>, rot=<0.6892 -0.6915 -0.1529 -0.1528>}
 dyn_backhair_bot_B {parent=base_backhair_bot_B, loc=<0 0 0>, rot=<0 0 0 1>}
 base_backhair_bot_C {parent=head, loc=<-6.2392 -12.933 3.0144>, rot=<0.7238 -0.6837 0.0402 0.0843>}
 dyn_backhair_bot_C {parent=base_backhair_bot_C, loc=<0 0 0>, rot=<0 0 0 1>}
 base_backhair_bot_D {parent=head, loc=<-3.6302 -10.4133 6.1912>, rot=<0.6343 -0.711 0.2175 0.2119>}
 dyn_backhair_bot_D {parent=base_backhair_bot_D, loc=<0 0 0>, rot=<0 0 0 1>}
 base_backhair_bot_E {parent=head, loc=<-3.0805 -5.8896 10.637>, rot=<0.493 -0.4999 0.5168 0.4898>}
 dyn_backhair_bot_E {parent=base_backhair_bot_E, loc=<0 0 0>, rot=<0 0 0 1>}
 base_backhair_bot_F {parent=head, loc=<-5.1501 -0.2458 -8.7507>, rot=<-0.4019 0.503 0.6268 0.4388>}
 dyn_backhair_bot_F {parent=base_backhair_bot_F, loc=<0 0 0>, rot=<0 0 0 1>}
 base_backhair_bot_G {parent=head, loc=<-3.4859 -0.5523 8.1228>, rot=<0.3907 -0.6177 0.5712 0.3736>}
 dyn_backhair_bot_G {parent=base_backhair_bot_G, loc=<0 0 0>, rot=<0 0 0 1>}
 base_backhair_bot_H {parent=head, loc=<-9.6132 -10.1762 -7.5291>, rot=<0.431 -0.8537 -0.0381 -0.2897>}
 dyn_backhair_bot_H {parent=base_backhair_bot_H, loc=<0 0 0>, rot=<0 0 0 1>}
 base_backhair_bot_I {parent=head, loc=<-10.6575 -13.6489 -1.4664>, rot=<0.5145 -0.551 0.4976 0.4291>}
 dyn_backhair_bot_I {parent=base_backhair_bot_I, loc=<0 0 0>, rot=<0 0 0 1>}
 base_backhair_bot_J {parent=head, loc=<-9.6362 -12.6255 4.0648>, rot=<0.3638 -0.9016 0.2087 0.1063>}
 dyn_backhair_bot_J {parent=base_backhair_bot_J, loc=<0 0 0>, rot=<0 0 0 1>}
 base_backhair_bot_K {parent=head, loc=<-8.3094 -9.8011 9.8784>, rot=<0.305 -0.8703 0.3229 0.2126>}
 dyn_backhair_bot_K {parent=base_backhair_bot_K, loc=<0 0 0>, rot=<0 0 0 1>}
 base_backhair_bot_L {parent=head, loc=<-8.9024 -3.4434 -8.3659>, rot=<0.128 -0.6376 -0.6236 -0.4338>}
 dyn_backhair_bot_L {parent=base_backhair_bot_L, loc=<0 0 0>, rot=<0 0 0 1>}
 base_backhair_bot_M {parent=head, loc=<-8.3747 -0.4821 6.9351>, rot=<-0.0109 -0.5621 0.7668 0.3098>}
 dyn_backhair_bot_M {parent=base_backhair_bot_M, loc=<0 0 0>, rot=<0 0 0 1>}
 head_null {parent=neck_02, type=Null, loc=<5.3914 0 0>, rot=<0 0 0.115 -0.9934>}
 clavicle_l {parent=spine_05, loc=<5.4831 -0.1053 -3.2946>, rot=<0.0465 -0.7289 -0.009 -0.6829>}
 upperarm_l {parent=clavicle_l, loc=<9.8658 -1.2136 -0.4105>, rot=<0.0164 -0.3245 -0.0269 -0.9454>}
 lowerarm_l {parent=upperarm_l, loc=<26.465 -0.8345 0.259>, rot=<0.0036 0.0161 -0.2298 -0.9731>}
 hand_l {parent=lowerarm_l, loc=<22.7355 -0.3203 -0.6105>, rot=<-0.6784 -0.0223 0.0735 -0.7307>}
 thumb_01_l {parent=hand_l, loc=<1.3118 -0.3191 -2.459>, rot=<0.6546 -0.4042 0.0664 -0.6354>}
 thumb_02_l {parent=thumb_01_l, loc=<5.9336 -0.0448 0.0401>, rot=<-0.0037 0.0213 0.298 -0.9543>}
 thumb_03_l {parent=thumb_02_l, loc=<2.2223 0.3682 -0.0907>, rot=<0.0118 -0.013 0.0066 -0.9998>}
 index_metacarpal_l {parent=hand_l, loc=<1.6535 0.0183 -2.96>, rot=<0.0158 -0.0124 -0.0173 -0.9996>}
 index_01_l {parent=index_metacarpal_l, loc=<8.3008 0.0357 -0.0524>, rot=<-0.0044 0.0485 0.0924 -0.9945>}
 index_02_l {parent=index_01_l, loc=<3.3587 -0.3004 0.0452>, rot=<-0.002 0.0027 0.3225 -0.9466>}
 index_03_l {parent=index_02_l, loc=<2.5578 -0.123 -0.0252>, rot=<-0.002 0.0067 0.2558 -0.9667>}
 middle_metacarpal_l {parent=hand_l, loc=<2.1856 -0.1725 -0.5984>, rot=<0.006 0.0168 -0.0329 -0.9993>}
 middle_01_l {parent=middle_metacarpal_l, loc=<7.3275 -0.2447 0.0129>, rot=<-0.0208 0.0042 0.0858 -0.9961>}
 middle_02_l {parent=middle_01_l, loc=<4.0123 0.1401 0.1175>, rot=<-0.0218 0.0258 0.3512 -0.9357>}
 middle_03_l {parent=middle_02_l, loc=<2.6077 0.0534 0.1373>, rot=<0.0082 0.0188 0.1708 -0.9851>}
 ring_metacarpal_l {parent=hand_l, loc=<2.2074 -0.3625 1.3638>, rot=<-0.0284 0.0722 -0.0458 -0.9959>}
 ring_01_l {parent=ring_metacarpal_l, loc=<7.0295 -0.6197 -0.1446>, rot=<-0.0489 -0.0252 0.1892 -0.9804>}
 ring_02_l {parent=ring_01_l, loc=<3.6174 0.4647 -0.1027>, rot=<0 0 0.2806 -0.9598>}
 ring_03_l {parent=ring_02_l, loc=<2.4152 0.114 0.0199>, rot=<0 0 0.1189 -0.9929>}
 pinky_metacarpal_l {parent=hand_l, loc=<1.7616 -0.9753 3.3814>, rot=<-0.0397 0.0814 -0.069 -0.9935>}
 pinky_01_l {parent=pinky_metacarpal_l, loc=<7.1709 -0.4831 -0.1094>, rot=<-0.0654 -0.0035 0.2242 -0.9723>}
 pinky_02_l {parent=pinky_01_l, loc=<2.8274 0.0893 -0.1043>, rot=<0.0037 0.0011 0.2854 -0.9584>}
 pinky_03_l {parent=pinky_02_l, loc=<2.2046 0.0307 -0.0078>, rot=<-0.0002 0.0098 0.1445 -0.9895>}
 deform_wrist_up_base_l {parent=hand_l, loc=<0 2.3009 0>, rot=<0 0 0 -1>}
 deform_wrist_up_l {parent=deform_wrist_up_base_l, loc=<0 0 0>, rot=<0 0 0 1>}
 lowerarm_twist_02_l {parent=lowerarm_l, loc=<10.6707 0 0>, rot=<0 0 0 1>}
 lowerarm_twist_01_l {parent=lowerarm_l, loc=<18.316 0 0>, rot=<0 0 0 1>}
 upperarm_twist_01_l {parent=upperarm_l, loc=<7.2135 -0.1134 0>, rot=<0 0 0 1>}
 deform_upperarm_bk_base_l {parent=upperarm_twist_01_l, loc=<-6.4053 -4.3746 0.314>, rot=<0 0 0 -1>}
 deform_upperarm_bk_l {parent=deform_upperarm_bk_base_l, loc=<0 0 0>, rot=<0 0 0 1>}
 deform_upperarm_fr_base_l {parent=upperarm_twist_01_l, loc=<-5.6312 5.092 0.3354>, rot=<0 0 0 -1>}
 deform_upperarm_fr_l {parent=deform_upperarm_fr_base_l, loc=<0 0 0>, rot=<0 0 0 1>}
 deform_upperarm_up_base_l {parent=upperarm_twist_01_l, loc=<-8.891 0.4825 3.9517>, rot=<0 0 0 -1>}
 deform_upperarm_up_l {parent=deform_upperarm_up_base_l, loc=<0 0 0>, rot=<0 0 0 1>}
 upperarm_twist_02_l {parent=upperarm_l, loc=<19.7998 -0.1134 0>, rot=<0 0 0 1>}
 clavicle_r {parent=spine_05, loc=<5.4831 -0.1053 3.2946>, rot=<0.7289 0.0465 -0.6829 0.009>}
 upperarm_r {parent=clavicle_r, loc=<-9.8658 1.2136 0.4105>, rot=<0.0164 -0.3245 -0.0269 -0.9454>}
 lowerarm_r {parent=upperarm_r, loc=<-26.465 0.8345 -0.259>, rot=<0.0036 0.0161 -0.2298 -0.9731>}
 hand_r {parent=lowerarm_r, loc=<-22.7355 0.3203 0.6106>, rot=<-0.6784 -0.0223 0.0735 -0.7307>}
 thumb_01_r {parent=hand_r, loc=<-1.3118 0.3191 2.459>, rot=<0.6546 -0.4042 0.0664 -0.6354>}
 thumb_02_r {parent=thumb_01_r, loc=<-5.9336 0.0448 -0.0401>, rot=<-0.0037 0.0213 0.298 -0.9543>}
 thumb_03_r {parent=thumb_02_r, loc=<-2.2223 -0.3682 0.0907>, rot=<0.0118 -0.013 0.0066 -0.9998>}
 index_metacarpal_r {parent=hand_r, loc=<-1.6535 -0.0183 2.96>, rot=<0.0158 -0.0124 -0.0173 -0.9996>}
 index_01_r {parent=index_metacarpal_r, loc=<-8.3008 -0.0357 0.0524>, rot=<-0.0044 0.0485 0.0924 -0.9945>}
 index_02_r {parent=index_01_r, loc=<-3.3587 0.3004 -0.0452>, rot=<-0.002 0.0027 0.3225 -0.9466>}
 index_03_r {parent=index_02_r, loc=<-2.5578 0.123 0.0252>, rot=<-0.002 0.0067 0.2558 -0.9667>}
 middle_metacarpal_r {parent=hand_r, loc=<-2.1856 0.1725 0.5984>, rot=<0.006 0.0168 -0.0329 -0.9993>}
 middle_01_r {parent=middle_metacarpal_r, loc=<-7.3275 0.2447 -0.0129>, rot=<-0.0208 0.0042 0.0858 -0.9961>}
 middle_02_r {parent=middle_01_r, loc=<-4.0123 -0.1401 -0.1175>, rot=<-0.0218 0.0258 0.3512 -0.9357>}
 middle_03_r {parent=middle_02_r, loc=<-2.6077 -0.0534 -0.1373>, rot=<0.0082 0.0188 0.1708 -0.9851>}
 ring_metacarpal_r {parent=hand_r, loc=<-2.2074 0.3625 -1.3638>, rot=<-0.0284 0.0722 -0.0458 -0.9959>}
 ring_01_r {parent=ring_metacarpal_r, loc=<-7.0295 0.6197 0.1446>, rot=<-0.0489 -0.0252 0.1892 -0.9804>}
 ring_02_r {parent=ring_01_r, loc=<-3.6174 -0.4647 0.1027>, rot=<0 0 0.2806 -0.9598>}
 ring_03_r {parent=ring_02_r, loc=<-2.4152 -0.114 -0.0199>, rot=<0 0 0.1189 -0.9929>}
 pinky_metacarpal_r {parent=hand_r, loc=<-1.7616 0.9753 -3.3814>, rot=<-0.0397 0.0814 -0.069 -0.9935>}
 pinky_01_r {parent=pinky_metacarpal_r, loc=<-7.1709 0.4831 0.1094>, rot=<-0.0654 -0.0035 0.2242 -0.9723>}
 pinky_02_r {parent=pinky_01_r, loc=<-2.8274 -0.0893 0.1043>, rot=<0.0037 0.0011 0.2854 -0.9584>}
 pinky_03_r {parent=pinky_02_r, loc=<-2.2046 -0.0307 0.0078>, rot=<-0.0002 0.0098 0.1445 -0.9895>}
 deform_wrist_up_base_r {parent=hand_r, loc=<-0.0002 -2.3007 -0.0001>, rot=<0 0 0 -1>}
 deform_wrist_up_r {parent=deform_wrist_up_base_r, loc=<0 0 0>, rot=<0 0 0 1>}
 lowerarm_twist_02_r {parent=lowerarm_r, loc=<-10.6707 0 0>, rot=<0 0 0 -1>}
 lowerarm_twist_01_r {parent=lowerarm_r, loc=<-18.316 0 0>, rot=<0 0 0 -1>}
 upperarm_twist_01_r {parent=upperarm_r, loc=<-7.2135 0.1134 0>, rot=<0 0 0 -1>}
 deform_upperarm_bk_base_r {parent=upperarm_twist_01_r, loc=<6.4053 4.3746 -0.314>, rot=<0 0 0 -1>}
 deform_upperarm_bk_r {parent=deform_upperarm_bk_base_r, loc=<0 0 0>, rot=<0 0 0 1>}
 deform_upperarm_fr_base_r {parent=upperarm_twist_01_r, loc=<5.6312 -5.092 -0.3354>, rot=<0 0 0 -1>}
 deform_upperarm_fr_r {parent=deform_upperarm_fr_base_r, loc=<0 0 0>, rot=<0 0 0 1>}
 deform_upperarm_up_base_r {parent=upperarm_twist_01_r, loc=<8.891 -0.4825 -3.9517>, rot=<0 0 0 -1>}
 deform_upperarm_up_r {parent=deform_upperarm_up_base_r, loc=<0 0 0>, rot=<0 0 0 1>}
 upperarm_twist_02_r {parent=upperarm_r, loc=<-19.7998 0.1134 0>, rot=<0 0 0 -1>}
 attach_backpack {parent=spine_05, loc=<-0.123 -13.473 0>, rot=<-0.0093 -0.707 0.0093 -0.707>}
 attach_cape {parent=spine_05, loc=<9.017 -11.714 0>, rot=<0.0616 -0.7044 -0.0616 -0.7044>}
 deform_pec_aim_base_l {parent=spine_04, loc=<3.2456 9.3309 -1.2237>, rot=<0.0603 -0.4033 0.3318 -0.8507>}
 deform_pec_aim_l {parent=deform_pec_aim_base_l, loc=<0 0 0>, rot=<0 0 0 1>}
 deform_pec_base_l {parent=deform_pec_aim_l, loc=<9.3665 0 0>, rot=<0 0 0 1>}
 deform_pec_l {parent=deform_pec_base_l, loc=<0 0 0>, rot=<0 0 0 1>}
 deform_pec_aim_base_r {parent=spine_04, loc=<3.2452 9.3309 1.2237>, rot=<-0.4033 -0.0603 0.8507 0.3318>}
 deform_pec_aim_r {parent=deform_pec_aim_base_r, loc=<0 0 0>, rot=<0 0 0 1>}
 deform_pec_base_r {parent=deform_pec_aim_r, loc=<-9.3667 -0.0003 -0.0004>, rot=<0 0 0 1>}
 deform_pec_r {parent=deform_pec_base_r, loc=<0 0 0>, rot=<0 0 0 1>}
 thigh_l {parent=pelvis, loc=<-1.0855 -0.2109 -8.7053>, rot=<0.0009 0.0355 0.0155 -0.9993>}
 calf_l {parent=thigh_l, loc=<-42.5284 -0.0019 -0.7297>, rot=<0.0006 -0.0161 -0.0391 -0.9991>}
 foot_l {parent=calf_l, loc=<-46.7401 -2.5014 -0.592>, rot=<0.0427 -0.0238 0.0353 -0.9982>}
 ball_l {parent=foot_l, loc=<-3.5658 10.8066 -0.1544>, rot=<0.0001 -0.0001 -0.7063 -0.7079>}
 calf_twist_02_l {parent=calf_l, loc=<-19.1546 -1.7943 -0.2638>, rot=<0.0237 0.0095 -0.0035 -0.9997>}
 calf_twist_01_l {parent=calf_l, loc=<-35.8483 -1.9197 -0.5794>, rot=<0.0237 0.0095 -0.0035 -0.9997>}
 deform_knee_in_base_l {parent=calf_l, loc=<0.8031 -5.0231 0.2456>, rot=<0 0 0 -1>}
 deform_knee_in_l {parent=deform_knee_in_base_l, loc=<0 0 0>, rot=<0 0 0 1>}
 deform_knee_out_base_l {parent=calf_l, loc=<0 4 0>, rot=<0 0 0 -1>}
 deform_knee_out_l {parent=deform_knee_out_base_l, loc=<0 0 0>, rot=<0 0 0 1>}
 thigh_twist_01_l {parent=thigh_l, loc=<-27.952 0 0>, rot=<0.0013 0 0 -1>}
 deform_groin_base_l {parent=thigh_l, loc=<0.4801 5.6776 -0.2486>, rot=<0 0 0 -1>}
 deform_groin_l {parent=deform_groin_base_l, loc=<0 0 0>, rot=<0 0 0 1>}
 thigh_r {parent=pelvis, loc=<-1.0855 -0.2109 8.7053>, rot=<0.0355 -0.0009 0.9993 0.0155>}
 calf_r {parent=thigh_r, loc=<42.5284 0.0019 0.7297>, rot=<0.0006 -0.0161 -0.0391 -0.9991>}
 foot_r {parent=calf_r, loc=<46.7401 2.5014 0.592>, rot=<0.0427 -0.0238 0.0353 -0.9982>}
 ball_r {parent=foot_r, loc=<3.5658 -10.8066 0.1544>, rot=<0.0001 -0.0001 -0.7063 -0.7079>}
 calf_twist_02_r {parent=calf_r, loc=<19.1546 1.7943 0.2638>, rot=<0.0237 0.0095 -0.0035 -0.9997>}
 calf_twist_01_r {parent=calf_r, loc=<35.8483 1.9197 0.5794>, rot=<0.0237 0.0095 -0.0035 -0.9997>}
 deform_knee_in_base_r {parent=calf_r, loc=<-0.8031 5.0231 -0.2457>, rot=<0 0 0 -1>}
 deform_knee_in_r {parent=deform_knee_in_base_r, loc=<0 0 0>, rot=<0 0 0 1>}
 deform_knee_out_base_r {parent=calf_r, loc=<0 -4 0>, rot=<0 0 0 -1>}
 deform_knee_out_r {parent=deform_knee_out_base_r, loc=<0 0 0>, rot=<0 0 0 1>}
 thigh_twist_01_r {parent=thigh_r, loc=<27.952 0 0>, rot=<0.0013 0 0 -1>}
 deform_groin_base_r {parent=thigh_r, loc=<-0.4801 -5.6776 0.2486>, rot=<0 0 0 -1>}
 deform_groin_r {parent=deform_groin_base_r, loc=<0 0 0>, rot=<0 0 0 1>}
 deform_glute_base_l {parent=pelvis, loc=<-7.3466 -9.0213 -9.1245>, rot=<0.0172 0.031 -0.4477 -0.8935>}
 deform_glute_l {parent=deform_glute_base_l, loc=<0 0 0>, rot=<0 0 0 1>}
 deform_glute_base_r {parent=pelvis, loc=<-7.3464 -9.0213 9.1245>, rot=<0.031 -0.0172 0.8935 -0.4477>}
 deform_glute_r {parent=deform_glute_base_r, loc=<0 0 0>, rot=<0 0 0 1>}
]
variables [
 CloseGroundRemap {type=double, default=0}
 Hair Chains Dynamics Disable {type=bool, default=False}
 Head right twist {type=double, default=0}
 HeadDownDelta {type=double, default=0}
 HeadRotXLeft {type=double, default=0}
 HeadRotXRight {type=double, default=0}
 HeadUpDelta {type=double, default=0}
 HeadUpExtDelta {type=double, default=0}
 IS Crouching {type=bool, default=False}
 Is Moving 2D {type=bool, default=False}
 Is Targeting {type=double, default=0}
 Melee swing {type=bool, default=False}
 isAR_Equipped {type=bool, default=False}
 isPistolEquipped {type=bool, default=False}
 isWearingBackbling {type=bool, default=False}
]
