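# Physics asset for the GlamClaws female medium body's face accessory mesh, set up for long-hair physics.
# Five hair chains (center, two left, two right) each link three dyn capsules to a tiny base capsule on the head, with narrow swing and zero twist limits.
# Head, neck, clavicle and upper-arm capsules plus an upper-spine box form the body shapes, joined by angularly free constraints, as are the chain bases to the head.
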PhysicsAsset /BRCosmetics/Characters/Player/Female/Medium/Bodies/F_MED_GlamClaws/Meshes/Parts/F_MED_GlamClaws_FaceAcc_Physics
bodies {
 base_longhair_C [{shapeType=Capsule, length=0.505, radius=0.505, shapeName=base_longhair_C_capsule, rotation=<rot 0 -90 0>}]
 base_longhair_L_A [{shapeType=Capsule, length=0.505, radius=0.505, shapeName=base_longhair_L_A_capsule, rotation=<rot 0 -90 0>}]
 base_longhair_L_B [{shapeType=Capsule, length=0.505, radius=0.505, shapeName=base_longhair_L_B_capsule, rotation=<rot 0 -90 0>}]
 base_longhair_R_A [{shapeType=Capsule, length=0.505, radius=0.505, shapeName=base_longhair_R_A_capsule, rotation=<rot 0 -90 0>}]
 base_longhair_R_B [{shapeType=Capsule, length=0.505, radius=0.505, shapeName=base_longhair_R_B_capsule, rotation=<rot 0 -90 0>}]
 clavicle_l [{shapeType=Capsule, length=6.8949, radius=6.1345, shapeName=clavicle_l_capsule, center=<5.3468 0 -0.4079>, rotation=<rot -0.0004 -90 88.4164>}]
 clavicle_r [{
   shapeType=Capsule
   length=6.8949
   radius=6.1345
   shapeName=clavicle_r_capsule
   center=<-5.1885 0 0.4992>
   rotation=<rot -0.0002 -90 -85.1755>
  }]
 dyn_longhair_C_02 [{shapeType=Capsule, length=1.3159, radius=2.6841, shapeName=dyn_longhair_C_02_capsule, center=<1.5 0 0>, rotation=<rot 88.7951 -90 -90>}]
 dyn_longhair_C_03 [{
   shapeType=Capsule
   length=5.7799
   radius=2.7293
   shapeName=dyn_longhair_C_03_capsule
   center=<2.7383 -0.0002 -0.0764>
   rotation=<rot -89.5814 179.9928 180>
  }]
 dyn_longhair_C_04 [{
   shapeType=Capsule
   length=12.8602
   radius=3.4997
   shapeName=dyn_longhair_C_04_capsule
   center=<2.904 0.0032 -0.0504>
   rotation=<rot 2.0552 179.9382 -180>
  }]
 dyn_longhair_L_A_02 [{shapeType=Capsule, length=1.3159, radius=2.6841, shapeName=dyn_longhair_L_A_02_capsule, center=<1.5 0 0>, rotation=<rot 90 0 0>}]
 dyn_longhair_L_A_03 [{
   shapeType=Capsule
   length=4.602
   radius=2.7293
   shapeName=dyn_longhair_L_A_03_capsule
   center=<4.3328 0.0004 -0.1102>
   rotation=<rot -87.7725 179.985 -179.9771>
  }]
 dyn_longhair_L_A_04 [{
   shapeType=Capsule
   length=5.9763
   radius=2.608
   shapeName=dyn_longhair_L_A_04_capsule
   center=<4.0557 0.0487 0.4448>
   rotation=<rot -11.325 -155.5306 174.9066>
  }]
 dyn_longhair_L_B_02 [{shapeType=Capsule, length=1.3159, radius=2.6841, shapeName=dyn_longhair_L_B_02_capsule, center=<1.5 0 0>, rotation=<rot 90 0 0>}]
 dyn_longhair_L_B_03 [{shapeType=Capsule, length=2, radius=2.4553, shapeName=dyn_longhair_L_B_03_capsule, center=<1.5 0 0>, rotation=<rot 90 -8.3008 0>}]
 dyn_longhair_L_B_04 [{
   shapeType=Capsule
   length=3.8996
   radius=2.608
   shapeName=dyn_longhair_L_B_04_capsule
   center=<4.5848 -0.0986 0>
   rotation=<rot -13.5038 -180 -180>
  }]
 dyn_longhair_R_A_02 [{shapeType=Capsule, length=1.3159, radius=2.6841, shapeName=dyn_longhair_R_A_02_capsule, center=<1.5 0 0>, rotation=<rot 90 0 0>}]
 dyn_longhair_R_A_03 [{
   shapeType=Capsule
   length=6.2421
   radius=2.7293
   shapeName=dyn_longhair_R_A_03_capsule
   center=<5.2141 0 0.2476>
   rotation=<rot 86.1854 180 179.9999>
  }]
 dyn_longhair_R_A_04 [{
   shapeType=Capsule
   length=7.2462
   radius=3.2063
   shapeName=dyn_longhair_R_A_04_capsule
   center=<4.5394 0 0.4163>
   rotation=<rot 28.3712 -180 180>
  }]
 dyn_longhair_R_B_02 [{shapeType=Capsule, length=1.3159, radius=2.6841, shapeName=dyn_longhair_R_B_02_capsule, center=<1.5 0 0>, rotation=<rot 90 0 0>}]
 dyn_longhair_R_B_03 [{shapeType=Capsule, length=2, radius=2.7293, shapeName=dyn_longhair_R_B_03_capsule, center=<1.5 0 0>, rotation=<rot 90 0 0>}]
 dyn_longhair_R_B_04 [{
   shapeType=Capsule
   length=1.5794
   radius=3.3668
   shapeName=dyn_longhair_R_B_04_capsule
   center=<3.3327 0.0003 2.2491>
   rotation=<rot -6.1155 -179.9969 179.9923>
  }]
 head [{shapeType=Capsule, length=4.524, radius=5.7795, center=<0.2059 2.64 -0.1009>, rotation=<rot -18.7061 -77.1011 84.0002>}]
 neck_01 [{shapeType=Capsule, length=3.2417, radius=5.2794, shapeName=neck_01_capsule, center=<0 0.1054 0>, rotation=<rot 0 -90.0001 85.9659>}]
 neck_02 [{shapeType=Capsule, length=3.0114, radius=5.1981, shapeName=neck_02_capsule, rotation=<rot 0 -90 80.9469>}]
 spine_05 [{
   shapeType=Box
   extentX=13.5986
   extentY=9.2468
   extentZ=24.8312
   shapeName=spine_05_box
   center=<5.5929 -0.6793 0.651>
   rotation=<rot 0 -98.8876 0>
  }]
 upperarm_l [{
   shapeType=Capsule
   length=8.9096
   radius=5.9041
   shapeName=upperarm_l_capsule
   center=<6.11 1.0499 0.6878>
   rotation=<rot -0.0078 -89.993 83.579>
  }]
 upperarm_r [{
   shapeType=Capsule
   length=8.9096
   radius=5.9041
   shapeName=upperarm_r_capsule
   center=<-6.1099 -1.579 -0.6878>
   rotation=<rot -0.0078 -89.993 -96.421>
  }]
}
constraints [
 {bone1Name=dyn_longhair_C_02, bone2Name=base_longhair_C, swing1LimitDegrees=10, swing2LimitDegrees=2, twistLimitDegrees=0}
 {bone1Name=dyn_longhair_C_03, bone2Name=dyn_longhair_C_02, swing1LimitDegrees=10, swing2LimitDegrees=2, twistLimitDegrees=0}
 {bone1Name=dyn_longhair_C_04, bone2Name=dyn_longhair_C_03, swing1LimitDegrees=10, swing2LimitDegrees=2, twistLimitDegrees=0}
 {bone1Name=dyn_longhair_L_A_02, bone2Name=base_longhair_L_A, swing1LimitDegrees=10, swing2LimitDegrees=2, twistLimitDegrees=0}
 {bone1Name=dyn_longhair_L_A_03, bone2Name=dyn_longhair_L_A_02, swing1LimitDegrees=10, swing2LimitDegrees=2, twistLimitDegrees=0}
 {bone1Name=dyn_longhair_L_A_04, bone2Name=dyn_longhair_L_A_03, swing1LimitDegrees=10, swing2LimitDegrees=2, twistLimitDegrees=0}
 {bone1Name=dyn_longhair_L_B_02, bone2Name=base_longhair_L_B, swing1LimitDegrees=5, swing2LimitDegrees=2, twistLimitDegrees=0}
 {bone1Name=dyn_longhair_L_B_03, bone2Name=dyn_longhair_L_B_02, swing1LimitDegrees=5, swing2LimitDegrees=2, twistLimitDegrees=0}
 {bone1Name=dyn_longhair_L_B_04, bone2Name=dyn_longhair_L_B_03, swing1LimitDegrees=5, swing2LimitDegrees=2, twistLimitDegrees=0}
 {bone1Name=dyn_longhair_R_A_02, bone2Name=base_longhair_R_A, swing1LimitDegrees=10, swing2LimitDegrees=2, twistLimitDegrees=0}
 {bone1Name=dyn_longhair_R_A_03, bone2Name=dyn_longhair_R_A_02, swing1LimitDegrees=10, swing2LimitDegrees=2, twistLimitDegrees=0}
 {bone1Name=dyn_longhair_R_A_04, bone2Name=dyn_longhair_R_A_03, swing1LimitDegrees=10, swing2LimitDegrees=2, twistLimitDegrees=0}
 {bone1Name=dyn_longhair_R_B_02, bone2Name=base_longhair_R_B, swing1LimitDegrees=10, swing2LimitDegrees=5, twistLimitDegrees=0}
 {bone1Name=dyn_longhair_R_B_03, bone2Name=dyn_longhair_R_B_02, swing1LimitDegrees=10, swing2LimitDegrees=5, twistLimitDegrees=0}
 {bone1Name=dyn_longhair_R_B_04, bone2Name=dyn_longhair_R_B_03, swing1LimitDegrees=10, swing2LimitDegrees=5, twistLimitDegrees=0}
 {
  bone1Name=base_longhair_C
  bone2Name=head
  swing1LimitDegrees=45
  swing1Motion=Free
  swing2LimitDegrees=45
  swing2Motion=Free
  twistLimitDegrees=45
  twistMotion=Free
 }
 {
  bone1Name=base_longhair_L_A
  bone2Name=head
  swing1LimitDegrees=45
  swing1Motion=Free
  swing2LimitDegrees=45
  swing2Motion=Free
  twistLimitDegrees=45
  twistMotion=Free
 }
 {
  bone1Name=base_longhair_L_B
  bone2Name=head
  swing1LimitDegrees=45
  swing1Motion=Free
  swing2LimitDegrees=45
  swing2Motion=Free
  twistLimitDegrees=45
  twistMotion=Free
 }
 {
  bone1Name=base_longhair_R_A
  bone2Name=head
  swing1LimitDegrees=45
  swing1Motion=Free
  swing2LimitDegrees=45
  swing2Motion=Free
  twistLimitDegrees=45
  twistMotion=Free
 }
 {
  bone1Name=base_longhair_R_B
  bone2Name=head
  swing1LimitDegrees=45
  swing1Motion=Free
  swing2LimitDegrees=45
  swing2Motion=Free
  twistLimitDegrees=45
  twistMotion=Free
 }
 {
  bone1Name=upperarm_l
  bone2Name=clavicle_l
  swing1LimitDegrees=45
  swing1Motion=Free
  swing2LimitDegrees=45
  swing2Motion=Free
  twistLimitDegrees=45
  twistMotion=Free
 }
 {
  bone1Name=upperarm_r
  bone2Name=clavicle_r
  swing1LimitDegrees=45
  swing1Motion=Free
  swing2LimitDegrees=45
  swing2Motion=Free
  twistLimitDegrees=45
  twistMotion=Free
 }
 {
  bone1Name=head
  bone2Name=neck_02
  swing1LimitDegrees=45
  swing1Motion=Free
  swing2LimitDegrees=45
  swing2Motion=Free
  twistLimitDegrees=45
  twistMotion=Free
 }
 {
  bone1Name=neck_02
  bone2Name=neck_01
  swing1LimitDegrees=45
  swing1Motion=Free
  swing2LimitDegrees=45
  swing2Motion=Free
  twistLimitDegrees=45
  twistMotion=Free
 }
 {
  bone1Name=neck_01
  bone2Name=spine_05
  swing1LimitDegrees=45
  swing1Motion=Free
  swing2LimitDegrees=45
  swing2Motion=Free
  twistLimitDegrees=45
  twistMotion=Free
 }
 {
  bone1Name=clavicle_l
  bone2Name=spine_05
  swing1LimitDegrees=45
  swing1Motion=Free
  swing2LimitDegrees=45
  swing2Motion=Free
  twistLimitDegrees=45
  twistMotion=Free
 }
 {
  bone1Name=clavicle_r
  bone2Name=spine_05
  swing1LimitDegrees=45
  swing1Motion=Free
  swing2LimitDegrees=45
  swing2Motion=Free
  twistLimitDegrees=45
  twistMotion=Free
 }
]
