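# Animation Blueprint for the GlamClaws female face accessory mesh, adding secondary motion to hair, ears and earrings.
# Parent and Body_Mesh poses, blended at the clavicles, feed a Control Rig, then head-space RigidBody simulation and a post-dynamics Control Rig.
# Weighted constraints make back-hair base bones follow head and long-hair bones.
# Chained AnimDynamics nodes finish front hair, back hair, ears and earrings, with neck spherical limits on front-hair tips.
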AnimBlueprint /BRCosmetics/Characters/Player/Female/Medium/Bodies/F_MED_GlamClaws/Meshes/Parts/F_MED_GlamClaws_FaceAcc_AnimBP
target_skeleton=&/F_MED_GlamClaws_FaceAcc_Skeleton
graphs [
 AnimGraph {nodes [
   Output Pose(Result=Use cached pose 'CRCache_dynamics'()) {layerGroup=None}
   n1 = Get Body_Mesh()
   n2 = Control Rig!(
    Source=Layered blend per bone(BasePose=Copy Pose From Mesh() {bUseAttachedParent=true}, BlendPoses_0=Copy Pose From Mesh(SourceMeshComponent=@n1)) {layerSetup [{branchFilters [{boneName=clavicle_r}, {boneName=clavicle_l}]}]}
    Alpha
    Is Targeting:double=0
    IS Crouching:bool=false
    Is Moving 2D:bool=false
    isPistolEquipped:bool=false
    isAR_Equipped:bool=false
    Hair Chains Dynamics Disable:bool=false
    Melee swing:bool=false
   ) -> (Pose) {controlRigAssetReference {blueprintRigClass=&/F_MED_GlamClaws_FaceAcc_CtrlRig.F_MED_GlamClaws_FaceAcc_CtrlRig_C}}
   Save cached pose 'CRCache_InputPose'(Pose=@n2) {cacheName=CRCache_InputPose}
   n3 = Anim Dynamics - Bone: dyn_fronthair_C_01!(
    ComponentPose=Local To Component(LocalPose=Use cached pose 'CRCache_constraints'())
    Alpha=1
    GravityOverride:Vector="0,0,-980"
    ExternalForce="0,0,0"
   ) -> (Pose) {
    angularDampingOverride=0.99
    angularSpringConstant=10000
    bOverrideAngularDamping=true
    bUseGravityOverride=true
    bUsePlanarLimit=false
    simulationSpace=RootRelative
    alphaBoolBlend {blendOption=Linear}
    boundBone {boneName=dyn_fronthair_C_01}
    componentAppliedLinearAccClamp=<3000 3000 3000>
    componentLinearAccScale=<0.4 0.4 0.4>
    componentLinearVelScale=<0.6 0.6 0.6>
    gravityOverride {z=-980}
    physicsBodyDefinitions [{
      boundBone {boneName=dyn_fronthair_C_01}
      boxExtents=<10 10 10>
      constraintSetup {angularLimitsMax=<0 7 1>, angularLimitsMin=<0 -7 -45>, angularTarget=<1 0 0>}
      localJointOffset=<0.8 0 0>
     }]
    simSpaceSettings {maxAngularAcceleration=14, maxAngularVelocity=7, simSpaceAngularAlpha=1}
   }
   n4 = Anim Dynamics - Bone: dyn_fronthair_C_02!(ComponentPose=@n3, Alpha=1, GravityOverride:Vector="0,0,-980", ExternalForce="0,0,0") -> (Pose) {
    angularDampingOverride=0.99
    angularSpringConstant=10000
    bOverrideAngularDamping=true
    bUseGravityOverride=true
    bUsePlanarLimit=false
    simulationSpace=RootRelative
    alphaBoolBlend {blendOption=Linear}
    boundBone {boneName=dyn_fronthair_C_02}
    componentAppliedLinearAccClamp=<3000 3000 3000>
    componentLinearAccScale=<0.4 0.4 0.4>
    componentLinearVelScale=<0.6 0.6 0.6>
    gravityOverride {z=-980}
    physicsBodyDefinitions [{
      collisionType=CustomSphere
      sphereCollisionRadius=1.2
      boundBone {boneName=dyn_fronthair_C_02}
      boxExtents=<40 40 40>
      constraintSetup {angularLimitsMax=<0 15 1>, angularLimitsMin=<0 -15 -50>, angularTarget=<1 0 0>}
      localJointOffset=<4 0.5 0>
     }]
    simSpaceSettings {maxAngularAcceleration=14, maxAngularVelocity=7, simSpaceAngularAlpha=1}
    sphericalLimits [{limitRadius=10, drivingBone {boneName=neck_01}, sphereLocalOffset=<1.6293 -3.3396 -2.7914>}]
   }
   n5 = Anim Dynamics - Bone: dyn_fronthair_R_01!(ComponentPose=@n4, Alpha=1, GravityOverride:Vector="0,0,-980", ExternalForce="0,0,0") -> (Pose) {
    angularDampingOverride=0.99
    angularSpringConstant=10000
    bOverrideAngularDamping=true
    bUseGravityOverride=true
    bUsePlanarLimit=false
    simulationSpace=RootRelative
    alphaBoolBlend {blendOption=Linear}
    boundBone {boneName=dyn_fronthair_R_01}
    componentAppliedLinearAccClamp=<3000 3000 3000>
    componentLinearAccScale=<0.4 0.4 0.4>
    componentLinearVelScale=<0.6 0.6 0.6>
    gravityOverride {z=-980}
    physicsBodyDefinitions [{
      boundBone {boneName=dyn_fronthair_R_01}
      boxExtents=<10 10 10>
      constraintSetup {angularLimitsMax=<0 2 1>, angularLimitsMin=<0 -1 -30>, angularTarget=<1 0 0>}
      localJointOffset=<0.8 0 0>
     }]
    simSpaceSettings {maxAngularAcceleration=14, maxAngularVelocity=7, simSpaceAngularAlpha=1}
   }
   n6 = Anim Dynamics - Bone: dyn_fronthair_R_02!(ComponentPose=@n5, Alpha=1, GravityOverride:Vector="0,0,-980", ExternalForce="0,0,0") -> (Pose) {
    angularDampingOverride=0.99
    angularSpringConstant=10000
    bOverrideAngularDamping=true
    bUseGravityOverride=true
    bUsePlanarLimit=false
    simulationSpace=RootRelative
    alphaBoolBlend {blendOption=Linear}
    boundBone {boneName=dyn_fronthair_R_02}
    componentAppliedLinearAccClamp=<3000 3000 3000>
    componentLinearAccScale=<0.4 0.4 0.4>
    componentLinearVelScale=<0.6 0.6 0.6>
    gravityOverride {z=-980}
    physicsBodyDefinitions [{
      collisionType=CustomSphere
      sphereCollisionRadius=1.2
      boundBone {boneName=dyn_fronthair_R_02}
      boxExtents=<40 40 40>
      constraintSetup {angularLimitsMax=<0 15 1>, angularLimitsMin=<0 -1 -50>, angularTarget=<1 0 0>}
      localJointOffset=<4 0.5 0>
     }]
    simSpaceSettings {maxAngularAcceleration=14, maxAngularVelocity=7, simSpaceAngularAlpha=1}
    sphericalLimits [{limitRadius=10, drivingBone {boneName=neck_01}, sphereLocalOffset=<1.6293 -3.3396 -2.7914>}]
   }
   n7 = Anim Dynamics - Bone: dyn_ear_R_01!(ComponentPose=@n6, Alpha=1, GravityOverride:Vector="0,0,-980", ExternalForce="0,0,0") -> (Pose) {
    angularDampingOverride=1
    angularSpringConstant=20000
    bAngularSpring=true
    bOverrideAngularDamping=true
    bUseGravityOverride=true
    gravityScale=0.35
    simulationSpace=RootRelative
    boundBone {boneName=dyn_ear_R_01}
    componentLinearAccScale=<1 1 1>
    componentLinearVelScale=<1 1 1>
    gravityOverride {z=-980}
    physicsBodyDefinitions [{
      boundBone {boneName=dyn_ear_R_01}
      boxExtents=<25 35 35>
      constraintSetup {angularLimitsMax=<5 5 5>, angularLimitsMin=<-5 -15 -5>, angularTarget=<1 0 0>}
      localJointOffset=<2 0 0>
     }]
    simSpaceSettings {maxAngularAcceleration=14, maxAngularVelocity=7, simSpaceAngularAlpha=1}
   }
   n8 = Anim Dynamics - Bone: dyn_ear_R_02!(ComponentPose=@n7, Alpha=1, GravityOverride:Vector="0,0,-980", ExternalForce="0,0,0") -> (Pose) {
    angularDampingOverride=1
    angularSpringConstant=20000
    bAngularSpring=true
    bOverrideAngularDamping=true
    bUseGravityOverride=true
    gravityScale=0.35
    simulationSpace=RootRelative
    boundBone {boneName=dyn_ear_R_02}
    componentLinearAccScale=<1 1 1>
    componentLinearVelScale=<1 1 1>
    gravityOverride {z=-980}
    physicsBodyDefinitions [{
      boundBone {boneName=dyn_ear_R_02}
      boxExtents=<25 35 35>
      constraintSetup {angularLimitsMax=<5 5 5>, angularLimitsMin=<-5 -15 -5>, angularTarget=<1 0 0>}
      localJointOffset=<2 0 0>
     }]
    simSpaceSettings {maxAngularAcceleration=14, maxAngularVelocity=7, simSpaceAngularAlpha=1}
   }
   n9 = Anim Dynamics - Bone: dyn_ear_L_01!(ComponentPose=@n8, Alpha=1, GravityOverride:Vector="0,0,-980", ExternalForce="0,0,0") -> (Pose) {
    angularDampingOverride=1
    angularSpringConstant=20000
    bAngularSpring=true
    bOverrideAngularDamping=true
    bUseGravityOverride=true
    gravityScale=0.35
    simulationSpace=RootRelative
    boundBone {boneName=dyn_ear_L_01}
    componentLinearAccScale=<1 1 1>
    componentLinearVelScale=<1 1 1>
    gravityOverride {z=-980}
    physicsBodyDefinitions [{
      boundBone {boneName=dyn_ear_L_01}
      boxExtents=<25 35 35>
      constraintSetup {angularLimitsMax=<5 5 5>, angularLimitsMin=<-5 -15 -5>, angularTarget=<1 0 0>}
      localJointOffset=<2 0 0>
     }]
    simSpaceSettings {maxAngularAcceleration=14, maxAngularVelocity=7, simSpaceAngularAlpha=1}
   }
   n10 = Anim Dynamics - Bone: dyn_ear_L_02!(ComponentPose=@n9, Alpha=1, GravityOverride:Vector="0,0,-980", ExternalForce="0,0,0") -> (Pose) {
    angularDampingOverride=1
    angularSpringConstant=20000
    bAngularSpring=true
    bOverrideAngularDamping=true
    bUseGravityOverride=true
    gravityScale=0.35
    simulationSpace=RootRelative
    boundBone {boneName=dyn_ear_L_02}
    componentLinearAccScale=<1 1 1>
    componentLinearVelScale=<1 1 1>
    gravityOverride {z=-980}
    physicsBodyDefinitions [{
      boundBone {boneName=dyn_ear_L_02}
      boxExtents=<25 35 35>
      constraintSetup {angularLimitsMax=<5 5 5>, angularLimitsMin=<-5 -15 -5>, angularTarget=<1 0 0>}
      localJointOffset=<2 0 0>
     }]
    simSpaceSettings {maxAngularAcceleration=14, maxAngularVelocity=7, simSpaceAngularAlpha=1}
   }
   n11 = Anim Dynamics - Bone: dyn_backhair_A!(ComponentPose=@n10, Alpha=1, GravityOverride:Vector="0,0,-980", ExternalForce="0,0,0") -> (Pose) {
    angularDampingOverride=0.99
    angularSpringConstant=10000
    bOverrideAngularDamping=true
    bUseGravityOverride=true
    bUsePlanarLimit=false
    simulationSpace=RootRelative
    alphaBoolBlend {blendOption=Linear}
    boundBone {boneName=dyn_backhair_A}
    componentAppliedLinearAccClamp=<5000 5000 5000>
    componentLinearAccScale=<0.4 0.4 0.3>
    componentLinearVelScale=<0.8 0.8 0.6>
    gravityOverride {z=-980}
    physicsBodyDefinitions [{
      boundBone {boneName=dyn_backhair_A}
      boxExtents=<10 10 10>
      constraintSetup {angularLimitsMax=<0 5 10>, angularLimitsMin=<0 -20 -25>, angularTarget=<1 0 0>}
      localJointOffset=<0.8 0 0>
     }]
    simSpaceSettings {maxAngularAcceleration=10, maxAngularVelocity=5, simSpaceAngularAlpha=1}
   }
   n12 = Anim Dynamics - Bone: dyn_backhair_B!(ComponentPose=@n11, Alpha=1, GravityOverride:Vector="0,0,-980", ExternalForce="0,0,0") -> (Pose) {
    angularDampingOverride=0.99
    angularSpringConstant=10000
    bOverrideAngularDamping=true
    bUseGravityOverride=true
    bUsePlanarLimit=false
    simulationSpace=RootRelative
    alphaBoolBlend {blendOption=Linear}
    boundBone {boneName=dyn_backhair_B}
    componentAppliedLinearAccClamp=<5000 5000 5000>
    componentLinearAccScale=<0.4 0.4 0.3>
    componentLinearVelScale=<0.8 0.8 0.6>
    gravityOverride {z=-980}
    physicsBodyDefinitions [{
      boundBone {boneName=dyn_backhair_B}
      boxExtents=<10 10 10>
      constraintSetup {angularLimitsMax=<0 5 10>, angularLimitsMin=<0 -20 -25>, angularTarget=<1 0 0>}
      localJointOffset=<0.8 0 0>
     }]
    simSpaceSettings {maxAngularAcceleration=10, maxAngularVelocity=5, simSpaceAngularAlpha=1}
   }
   n13 = Anim Dynamics - Bone: dyn_backhair_C!(ComponentPose=@n12, Alpha=1, GravityOverride:Vector="0,0,-980", ExternalForce="0,0,0") -> (Pose) {
    angularDampingOverride=0.99
    angularSpringConstant=10000
    bOverrideAngularDamping=true
    bUseGravityOverride=true
    bUsePlanarLimit=false
    simulationSpace=RootRelative
    alphaBoolBlend {blendOption=Linear}
    boundBone {boneName=dyn_backhair_C}
    componentAppliedLinearAccClamp=<5000 5000 5000>
    componentLinearAccScale=<0.4 0.4 0.3>
    componentLinearVelScale=<0.8 0.8 0.6>
    gravityOverride {z=-980}
    physicsBodyDefinitions [{
      boundBone {boneName=dyn_backhair_C}
      boxExtents=<10 10 10>
      constraintSetup {angularLimitsMax=<0 5 10>, angularLimitsMin=<0 -20 -25>, angularTarget=<1 0 0>}
      localJointOffset=<0.8 0 0>
     }]
    simSpaceSettings {maxAngularAcceleration=10, maxAngularVelocity=5, simSpaceAngularAlpha=1}
   }
   n14 = Anim Dynamics - Bone: dyn_backhair_D!(ComponentPose=@n13, Alpha=1, GravityOverride:Vector="0,0,-980", ExternalForce="0,0,0") -> (Pose) {
    angularDampingOverride=0.99
    angularSpringConstant=10000
    bOverrideAngularDamping=true
    bUseGravityOverride=true
    bUsePlanarLimit=false
    simulationSpace=RootRelative
    alphaBoolBlend {blendOption=Linear}
    boundBone {boneName=dyn_backhair_D}
    componentAppliedLinearAccClamp=<5000 5000 5000>
    componentLinearAccScale=<0.4 0.4 0.3>
    componentLinearVelScale=<0.8 0.8 0.6>
    gravityOverride {z=-980}
    physicsBodyDefinitions [{
      boundBone {boneName=dyn_backhair_D}
      boxExtents=<10 10 10>
      constraintSetup {angularLimitsMax=<0 5 10>, angularLimitsMin=<0 -20 -25>, angularTarget=<1 0 0>}
      localJointOffset=<0.8 0 0>
     }]
    simSpaceSettings {maxAngularAcceleration=10, maxAngularVelocity=5, simSpaceAngularAlpha=1}
   }
   n15 = Anim Dynamics - Bone: dyn_backhair_E!(ComponentPose=@n14, Alpha=1, GravityOverride:Vector="0,0,-980", ExternalForce="0,0,0") -> (Pose) {
    angularDampingOverride=0.99
    angularSpringConstant=10000
    bOverrideAngularDamping=true
    bUseGravityOverride=true
    bUsePlanarLimit=false
    simulationSpace=RootRelative
    alphaBoolBlend {blendOption=Linear}
    boundBone {boneName=dyn_backhair_E}
    componentAppliedLinearAccClamp=<5000 5000 5000>
    componentLinearAccScale=<0.4 0.4 0.3>
    componentLinearVelScale=<0.8 0.8 0.6>
    gravityOverride {z=-980}
    physicsBodyDefinitions [{
      boundBone {boneName=dyn_backhair_E}
      boxExtents=<10 10 10>
      constraintSetup {angularLimitsMax=<0 5 10>, angularLimitsMin=<0 -20 -25>, angularTarget=<1 0 0>}
      localJointOffset=<0.8 0 0>
     }]
    simSpaceSettings {maxAngularAcceleration=10, maxAngularVelocity=5, simSpaceAngularAlpha=1}
   }
   n16 = Anim Dynamics - Bone: dyn_backhair_F!(ComponentPose=@n15, Alpha=1, GravityOverride:Vector="0,0,-980", ExternalForce="0,0,0") -> (Pose) {
    angularDampingOverride=0.99
    angularSpringConstant=10000
    bOverrideAngularDamping=true
    bUseGravityOverride=true
    bUsePlanarLimit=false
    simulationSpace=RootRelative
    alphaBoolBlend {blendOption=Linear}
    boundBone {boneName=dyn_backhair_F}
    componentAppliedLinearAccClamp=<5000 5000 5000>
    componentLinearAccScale=<0.4 0.4 0.3>
    componentLinearVelScale=<0.8 0.8 0.6>
    gravityOverride {z=-980}
    physicsBodyDefinitions [{
      boundBone {boneName=dyn_backhair_F}
      boxExtents=<10 10 10>
      constraintSetup {angularLimitsMax=<0 5 10>, angularLimitsMin=<0 -20 -25>, angularTarget=<1 0 0>}
      localJointOffset=<0.8 0 0>
     }]
    simSpaceSettings {maxAngularAcceleration=10, maxAngularVelocity=5, simSpaceAngularAlpha=1}
   }
   n17 = Anim Dynamics - Bone: dyn_backhair_middle_A!(ComponentPose=@n16, Alpha=1, GravityOverride:Vector="0,0,-980", ExternalForce="0,0,0") -> (Pose) {
    angularDampingOverride=0.99
    angularSpringConstant=1000
    bAngularSpring=true
    bOverrideAngularDamping=true
    bUseGravityOverride=true
    bUsePlanarLimit=false
    simulationSpace=RootRelative
    alphaBoolBlend {blendOption=Linear}
    boundBone {boneName=dyn_backhair_middle_A}
    componentAppliedLinearAccClamp=<5000 5000 5000>
    componentLinearAccScale=<0.4 0.4 0.3>
    componentLinearVelScale=<0.8 0.8 0.6>
    gravityOverride {z=-980}
    physicsBodyDefinitions [{
      boundBone {boneName=dyn_backhair_middle_A}
      boxExtents=<10 10 10>
      constraintSetup {angularLimitsMax=<0 5 10>, angularLimitsMin=<0 -20 -25>, angularTarget=<1 0 0>}
      localJointOffset=<0.8 0 0>
     }]
    simSpaceSettings {maxAngularAcceleration=10, maxAngularVelocity=5, simSpaceAngularAlpha=1}
   }
   n18 = Anim Dynamics - Bone: dyn_backhair_middle_B!(ComponentPose=@n17, Alpha=1, GravityOverride:Vector="0,0,-980", ExternalForce="0,0,0") -> (Pose) {
    angularDampingOverride=0.99
    angularSpringConstant=1000
    bAngularSpring=true
    bOverrideAngularDamping=true
    bUseGravityOverride=true
    bUsePlanarLimit=false
    simulationSpace=RootRelative
    alphaBoolBlend {blendOption=Linear}
    boundBone {boneName=dyn_backhair_middle_B}
    componentAppliedLinearAccClamp=<5000 5000 5000>
    componentLinearAccScale=<0.4 0.4 0.3>
    componentLinearVelScale=<0.8 0.8 0.6>
    gravityOverride {z=-980}
    physicsBodyDefinitions [{
      boundBone {boneName=dyn_backhair_middle_B}
      boxExtents=<10 10 10>
      constraintSetup {angularLimitsMax=<0 5 10>, angularLimitsMin=<0 -20 -25>, angularTarget=<1 0 0>}
      localJointOffset=<0.8 0 0>
     }]
    simSpaceSettings {maxAngularAcceleration=10, maxAngularVelocity=5, simSpaceAngularAlpha=1}
   }
   n19 = Anim Dynamics - Bone: dyn_backhair_middle_C!(ComponentPose=@n18, Alpha=1, GravityOverride:Vector="0,0,-980", ExternalForce="0,0,0") -> (Pose) {
    angularDampingOverride=0.99
    angularSpringConstant=1000
    bAngularSpring=true
    bOverrideAngularDamping=true
    bUseGravityOverride=true
    bUsePlanarLimit=false
    simulationSpace=RootRelative
    alphaBoolBlend {blendOption=Linear}
    boundBone {boneName=dyn_backhair_middle_C}
    componentAppliedLinearAccClamp=<5000 5000 5000>
    componentLinearAccScale=<0.4 0.4 0.3>
    componentLinearVelScale=<0.8 0.8 0.6>
    gravityOverride {z=-980}
    physicsBodyDefinitions [{
      boundBone {boneName=dyn_backhair_middle_C}
      boxExtents=<10 10 10>
      constraintSetup {angularLimitsMax=<0 5 10>, angularLimitsMin=<0 -20 -25>, angularTarget=<1 0 0>}
      localJointOffset=<0.8 0 0>
     }]
    simSpaceSettings {maxAngularAcceleration=10, maxAngularVelocity=5, simSpaceAngularAlpha=1}
   }
   n20 = Anim Dynamics - Bone: dyn_backhair_middle_D!(ComponentPose=@n19, Alpha=1, GravityOverride:Vector="0,0,-980", ExternalForce="0,0,0") -> (Pose) {
    angularDampingOverride=0.99
    angularSpringConstant=1000
    bAngularSpring=true
    bOverrideAngularDamping=true
    bUseGravityOverride=true
    bUsePlanarLimit=false
    simulationSpace=RootRelative
    alphaBoolBlend {blendOption=Linear}
    boundBone {boneName=dyn_backhair_middle_D}
    componentAppliedLinearAccClamp=<5000 5000 5000>
    componentLinearAccScale=<0.4 0.4 0.3>
    componentLinearVelScale=<0.8 0.8 0.6>
    gravityOverride {z=-980}
    physicsBodyDefinitions [{
      boundBone {boneName=dyn_backhair_middle_D}
      boxExtents=<10 10 10>
      constraintSetup {angularLimitsMax=<0 5 10>, angularLimitsMin=<0 -20 -25>, angularTarget=<1 0 0>}
      localJointOffset=<0.8 0 0>
     }]
    simSpaceSettings {maxAngularAcceleration=10, maxAngularVelocity=5, simSpaceAngularAlpha=1}
   }
   n21 = Anim Dynamics - Bone: dyn_backhair_middle_E!(ComponentPose=@n20, Alpha=1, GravityOverride:Vector="0,0,-980", ExternalForce="0,0,0") -> (Pose) {
    angularDampingOverride=0.99
    angularSpringConstant=1000
    bAngularSpring=true
    bOverrideAngularDamping=true
    bUseGravityOverride=true
    bUsePlanarLimit=false
    simulationSpace=RootRelative
    alphaBoolBlend {blendOption=Linear}
    boundBone {boneName=dyn_backhair_middle_E}
    componentAppliedLinearAccClamp=<5000 5000 5000>
    componentLinearAccScale=<0.4 0.4 0.3>
    componentLinearVelScale=<0.8 0.8 0.6>
    gravityOverride {z=-980}
    physicsBodyDefinitions [{
      boundBone {boneName=dyn_backhair_middle_E}
      boxExtents=<10 10 10>
      constraintSetup {angularLimitsMax=<0 5 10>, angularLimitsMin=<0 -20 -25>, angularTarget=<1 0 0>}
      localJointOffset=<0.8 0 0>
     }]
    simSpaceSettings {maxAngularAcceleration=10, maxAngularVelocity=5, simSpaceAngularAlpha=1}
   }
   n22 = Anim Dynamics - Bone: dyn_backhair_middle_F!(ComponentPose=@n21, Alpha=1, GravityOverride:Vector="0,0,-980", ExternalForce="0,0,0") -> (Pose) {
    angularDampingOverride=0.99
    angularSpringConstant=1000
    bAngularSpring=true
    bOverrideAngularDamping=true
    bUseGravityOverride=true
    bUsePlanarLimit=false
    simulationSpace=RootRelative
    alphaBoolBlend {blendOption=Linear}
    boundBone {boneName=dyn_backhair_middle_F}
    componentAppliedLinearAccClamp=<5000 5000 5000>
    componentLinearAccScale=<0.4 0.4 0.3>
    componentLinearVelScale=<0.8 0.8 0.6>
    gravityOverride {z=-980}
    physicsBodyDefinitions [{
      boundBone {boneName=dyn_backhair_middle_F}
      boxExtents=<10 10 10>
      constraintSetup {angularLimitsMax=<0 5 10>, angularLimitsMin=<0 -20 -25>, angularTarget=<1 0 0>}
      localJointOffset=<0.8 0 0>
     }]
    simSpaceSettings {maxAngularAcceleration=10, maxAngularVelocity=5, simSpaceAngularAlpha=1}
   }
   n23 = Anim Dynamics - Bone: dyn_backhair_middle_G!(ComponentPose=@n22, Alpha=1, GravityOverride:Vector="0,0,-980", ExternalForce="0,0,0") -> (Pose) {
    angularDampingOverride=0.99
    angularSpringConstant=1000
    bAngularSpring=true
    bOverrideAngularDamping=true
    bUseGravityOverride=true
    bUsePlanarLimit=false
    simulationSpace=RootRelative
    alphaBoolBlend {blendOption=Linear}
    boundBone {boneName=dyn_backhair_middle_G}
    componentAppliedLinearAccClamp=<5000 5000 5000>
    componentLinearAccScale=<0.4 0.4 0.3>
    componentLinearVelScale=<0.8 0.8 0.6>
    gravityOverride {z=-980}
    physicsBodyDefinitions [{
      boundBone {boneName=dyn_backhair_middle_G}
      boxExtents=<10 10 10>
      constraintSetup {angularLimitsMax=<0 5 10>, angularLimitsMin=<0 -20 -25>, angularTarget=<1 0 0>}
      localJointOffset=<0.8 0 0>
     }]
    simSpaceSettings {maxAngularAcceleration=10, maxAngularVelocity=5, simSpaceAngularAlpha=1}
   }
   n24 = Anim Dynamics - Bone: dyn_backhair_middle_H!(ComponentPose=@n23, Alpha=1, GravityOverride:Vector="0,0,-980", ExternalForce="0,0,0") -> (Pose) {
    angularDampingOverride=0.99
    angularSpringConstant=1000
    bAngularSpring=true
    bOverrideAngularDamping=true
    bUseGravityOverride=true
    bUsePlanarLimit=false
    simulationSpace=RootRelative
    alphaBoolBlend {blendOption=Linear}
    boundBone {boneName=dyn_backhair_middle_H}
    componentAppliedLinearAccClamp=<5000 5000 5000>
    componentLinearAccScale=<0.4 0.4 0.3>
    componentLinearVelScale=<0.8 0.8 0.6>
    gravityOverride {z=-980}
    physicsBodyDefinitions [{
      boundBone {boneName=dyn_backhair_middle_H}
      boxExtents=<10 10 10>
      constraintSetup {angularLimitsMax=<0 5 10>, angularLimitsMin=<0 -20 -25>, angularTarget=<1 0 0>}
      localJointOffset=<0.8 0 0>
     }]
    simSpaceSettings {maxAngularAcceleration=10, maxAngularVelocity=5, simSpaceAngularAlpha=1}
   }
   n25 = Anim Dynamics - Bone: dyn_backhair_middle_I!(ComponentPose=@n24, Alpha=1, GravityOverride:Vector="0,0,-980", ExternalForce="0,0,0") -> (Pose) {
    angularDampingOverride=0.99
    angularSpringConstant=1000
    bAngularSpring=true
    bOverrideAngularDamping=true
    bUseGravityOverride=true
    bUsePlanarLimit=false
    simulationSpace=RootRelative
    alphaBoolBlend {blendOption=Linear}
    boundBone {boneName=dyn_backhair_middle_I}
    componentAppliedLinearAccClamp=<5000 5000 5000>
    componentLinearAccScale=<0.4 0.4 0.3>
    componentLinearVelScale=<0.8 0.8 0.6>
    gravityOverride {z=-980}
    physicsBodyDefinitions [{
      boundBone {boneName=dyn_backhair_middle_I}
      boxExtents=<10 10 10>
      constraintSetup {angularLimitsMax=<0 5 10>, angularLimitsMin=<0 -20 -25>, angularTarget=<1 0 0>}
      localJointOffset=<0.8 0 0>
     }]
    simSpaceSettings {maxAngularAcceleration=10, maxAngularVelocity=5, simSpaceAngularAlpha=1}
   }
   n26 = Anim Dynamics - Bone: dyn_backhair_middle_J!(ComponentPose=@n25, Alpha=1, GravityOverride:Vector="0,0,-980", ExternalForce="0,0,0") -> (Pose) {
    angularDampingOverride=0.99
    angularSpringConstant=1000
    bAngularSpring=true
    bOverrideAngularDamping=true
    bUseGravityOverride=true
    bUsePlanarLimit=false
    simulationSpace=RootRelative
    alphaBoolBlend {blendOption=Linear}
    boundBone {boneName=dyn_backhair_middle_J}
    componentAppliedLinearAccClamp=<5000 5000 5000>
    componentLinearAccScale=<0.4 0.4 0.3>
    componentLinearVelScale=<0.8 0.8 0.6>
    gravityOverride {z=-980}
    physicsBodyDefinitions [{
      boundBone {boneName=dyn_backhair_middle_J}
      boxExtents=<10 10 10>
      constraintSetup {angularLimitsMax=<0 5 10>, angularLimitsMin=<0 -20 -25>, angularTarget=<1 0 0>}
      localJointOffset=<0.8 0 0>
     }]
    simSpaceSettings {maxAngularAcceleration=10, maxAngularVelocity=5, simSpaceAngularAlpha=1}
   }
   n27 = Anim Dynamics - Bone: dyn_backhair_middle_K!(ComponentPose=@n26, Alpha=1, GravityOverride:Vector="0,0,-980", ExternalForce="0,0,0") -> (Pose) {
    angularDampingOverride=0.99
    angularSpringConstant=1000
    bAngularSpring=true
    bOverrideAngularDamping=true
    bUseGravityOverride=true
    bUsePlanarLimit=false
    simulationSpace=RootRelative
    alphaBoolBlend {blendOption=Linear}
    boundBone {boneName=dyn_backhair_middle_K}
    componentAppliedLinearAccClamp=<5000 5000 5000>
    componentLinearAccScale=<0.4 0.4 0.3>
    componentLinearVelScale=<0.8 0.8 0.6>
    gravityOverride {z=-980}
    physicsBodyDefinitions [{
      boundBone {boneName=dyn_backhair_middle_K}
      boxExtents=<10 10 10>
      constraintSetup {angularLimitsMax=<0 5 10>, angularLimitsMin=<0 -20 -25>, angularTarget=<1 0 0>}
      localJointOffset=<0.8 0 0>
     }]
    simSpaceSettings {maxAngularAcceleration=10, maxAngularVelocity=5, simSpaceAngularAlpha=1}
   }
   n28 = Anim Dynamics - Bone: dyn_backhair_middle_L!(ComponentPose=@n27, Alpha=1, GravityOverride:Vector="0,0,-980", ExternalForce="0,0,0") -> (Pose) {
    angularDampingOverride=0.99
    angularSpringConstant=1000
    bAngularSpring=true
    bOverrideAngularDamping=true
    bUseGravityOverride=true
    bUsePlanarLimit=false
    simulationSpace=RootRelative
    alphaBoolBlend {blendOption=Linear}
    boundBone {boneName=dyn_backhair_middle_L}
    componentAppliedLinearAccClamp=<5000 5000 5000>
    componentLinearAccScale=<0.4 0.4 0.3>
    componentLinearVelScale=<0.8 0.8 0.6>
    gravityOverride {z=-980}
    physicsBodyDefinitions [{
      boundBone {boneName=dyn_backhair_middle_L}
      boxExtents=<10 10 10>
      constraintSetup {angularLimitsMax=<0 5 10>, angularLimitsMin=<0 -20 -25>, angularTarget=<1 0 0>}
      localJointOffset=<0.8 0 0>
     }]
    simSpaceSettings {maxAngularAcceleration=10, maxAngularVelocity=5, simSpaceAngularAlpha=1}
   }
   n29 = Anim Dynamics - Bone: dyn_backhair_middle_M!(ComponentPose=@n28, Alpha=1, GravityOverride:Vector="0,0,-980", ExternalForce="0,0,0") -> (Pose) {
    angularDampingOverride=0.99
    angularSpringConstant=1000
    bAngularSpring=true
    bOverrideAngularDamping=true
    bUseGravityOverride=true
    bUsePlanarLimit=false
    simulationSpace=RootRelative
    alphaBoolBlend {blendOption=Linear}
    boundBone {boneName=dyn_backhair_middle_M}
    componentAppliedLinearAccClamp=<5000 5000 5000>
    componentLinearAccScale=<0.4 0.4 0.3>
    componentLinearVelScale=<0.8 0.8 0.6>
    gravityOverride {z=-980}
    physicsBodyDefinitions [{
      boundBone {boneName=dyn_backhair_middle_M}
      boxExtents=<10 10 10>
      constraintSetup {angularLimitsMax=<0 5 10>, angularLimitsMin=<0 -20 -25>, angularTarget=<1 0 0>}
      localJointOffset=<0.8 0 0>
     }]
    simSpaceSettings {maxAngularAcceleration=10, maxAngularVelocity=5, simSpaceAngularAlpha=1}
   }
   n30 = Anim Dynamics - Bone: dyn_backhair_middle_N!(ComponentPose=@n29, Alpha=1, GravityOverride:Vector="0,0,-980", ExternalForce="0,0,0") -> (Pose) {
    angularDampingOverride=0.99
    angularSpringConstant=1000
    bAngularSpring=true
    bOverrideAngularDamping=true
    bUseGravityOverride=true
    bUsePlanarLimit=false
    simulationSpace=RootRelative
    alphaBoolBlend {blendOption=Linear}
    boundBone {boneName=dyn_backhair_middle_N}
    componentAppliedLinearAccClamp=<5000 5000 5000>
    componentLinearAccScale=<0.4 0.4 0.3>
    componentLinearVelScale=<0.8 0.8 0.6>
    gravityOverride {z=-980}
    physicsBodyDefinitions [{
      boundBone {boneName=dyn_backhair_middle_N}
      boxExtents=<10 10 10>
      constraintSetup {angularLimitsMax=<0 5 10>, angularLimitsMin=<0 -20 -25>, angularTarget=<1 0 0>}
      localJointOffset=<0.8 0 0>
     }]
    simSpaceSettings {maxAngularAcceleration=10, maxAngularVelocity=5, simSpaceAngularAlpha=1}
   }
   n31 = Anim Dynamics - Bone: dyn_backhair_middle_O!(ComponentPose=@n30, Alpha=1, GravityOverride:Vector="0,0,-980", ExternalForce="0,0,0") -> (Pose) {
    angularDampingOverride=0.99
    angularSpringConstant=1000
    bAngularSpring=true
    bOverrideAngularDamping=true
    bUseGravityOverride=true
    bUsePlanarLimit=false
    simulationSpace=RootRelative
    alphaBoolBlend {blendOption=Linear}
    boundBone {boneName=dyn_backhair_middle_O}
    componentAppliedLinearAccClamp=<5000 5000 5000>
    componentLinearAccScale=<0.4 0.4 0.3>
    componentLinearVelScale=<0.8 0.8 0.6>
    gravityOverride {z=-980}
    physicsBodyDefinitions [{
      boundBone {boneName=dyn_backhair_middle_O}
      boxExtents=<10 10 10>
      constraintSetup {angularLimitsMax=<0 5 10>, angularLimitsMin=<0 -20 -25>, angularTarget=<1 0 0>}
      localJointOffset=<0.8 0 0>
     }]
    simSpaceSettings {maxAngularAcceleration=10, maxAngularVelocity=5, simSpaceAngularAlpha=1}
   }
   n32 = Anim Dynamics - Bone: dyn_backhair_bot_A!(ComponentPose=@n31, Alpha=1, GravityOverride:Vector="0,0,-980", ExternalForce="0,0,0") -> (Pose) {
    angularDampingOverride=0.99
    angularSpringConstant=1000
    bAngularSpring=true
    bOverrideAngularDamping=true
    bUseGravityOverride=true
    bUsePlanarLimit=false
    simulationSpace=RootRelative
    alphaBoolBlend {blendOption=Linear}
    boundBone {boneName=dyn_backhair_bot_A}
    componentAppliedLinearAccClamp=<5000 5000 5000>
    componentLinearAccScale=<0.4 0.4 0.3>
    componentLinearVelScale=<0.8 0.8 0.6>
    gravityOverride {z=-980}
    physicsBodyDefinitions [{
      boundBone {boneName=dyn_backhair_bot_A}
      boxExtents=<10 10 10>
      constraintSetup {angularLimitsMax=<0 5 10>, angularLimitsMin=<0 -20 -25>, angularTarget=<1 0 0>}
      localJointOffset=<0.8 0 0>
     }]
    simSpaceSettings {maxAngularAcceleration=10, maxAngularVelocity=5, simSpaceAngularAlpha=1}
   }
   n33 = Anim Dynamics - Bone: dyn_backhair_bot_B!(ComponentPose=@n32, Alpha=1, GravityOverride:Vector="0,0,-980", ExternalForce="0,0,0") -> (Pose) {
    angularDampingOverride=0.99
    angularSpringConstant=1000
    bAngularSpring=true
    bOverrideAngularDamping=true
    bUseGravityOverride=true
    bUsePlanarLimit=false
    simulationSpace=RootRelative
    alphaBoolBlend {blendOption=Linear}
    boundBone {boneName=dyn_backhair_bot_B}
    componentAppliedLinearAccClamp=<5000 5000 5000>
    componentLinearAccScale=<0.4 0.4 0.3>
    componentLinearVelScale=<0.8 0.8 0.6>
    gravityOverride {z=-980}
    physicsBodyDefinitions [{
      boundBone {boneName=dyn_backhair_bot_B}
      boxExtents=<10 10 10>
      constraintSetup {angularLimitsMax=<0 5 10>, angularLimitsMin=<0 -20 -25>, angularTarget=<1 0 0>}
      localJointOffset=<0.8 0 0>
     }]
    simSpaceSettings {maxAngularAcceleration=10, maxAngularVelocity=5, simSpaceAngularAlpha=1}
   }
   n34 = Anim Dynamics - Bone: dyn_backhair_bot_C!(ComponentPose=@n33, Alpha=1, GravityOverride:Vector="0,0,-980", ExternalForce="0,0,0") -> (Pose) {
    angularDampingOverride=0.99
    angularSpringConstant=1000
    bAngularSpring=true
    bOverrideAngularDamping=true
    bUseGravityOverride=true
    bUsePlanarLimit=false
    simulationSpace=RootRelative
    alphaBoolBlend {blendOption=Linear}
    boundBone {boneName=dyn_backhair_bot_C}
    componentAppliedLinearAccClamp=<5000 5000 5000>
    componentLinearAccScale=<0.4 0.4 0.3>
    componentLinearVelScale=<0.8 0.8 0.6>
    gravityOverride {z=-980}
    physicsBodyDefinitions [{
      boundBone {boneName=dyn_backhair_bot_C}
      boxExtents=<10 10 10>
      constraintSetup {angularLimitsMax=<0 5 10>, angularLimitsMin=<0 -20 -25>, angularTarget=<1 0 0>}
      localJointOffset=<0.8 0 0>
     }]
    simSpaceSettings {maxAngularAcceleration=10, maxAngularVelocity=5, simSpaceAngularAlpha=1}
   }
   n35 = Anim Dynamics - Bone: dyn_backhair_bot_D!(ComponentPose=@n34, Alpha=1, GravityOverride:Vector="0,0,-980", ExternalForce="0,0,0") -> (Pose) {
    angularDampingOverride=0.99
    angularSpringConstant=1000
    bAngularSpring=true
    bOverrideAngularDamping=true
    bUseGravityOverride=true
    bUsePlanarLimit=false
    simulationSpace=RootRelative
    alphaBoolBlend {blendOption=Linear}
    boundBone {boneName=dyn_backhair_bot_D}
    componentAppliedLinearAccClamp=<5000 5000 5000>
    componentLinearAccScale=<0.4 0.4 0.3>
    componentLinearVelScale=<0.8 0.8 0.6>
    gravityOverride {z=-980}
    physicsBodyDefinitions [{
      boundBone {boneName=dyn_backhair_bot_D}
      boxExtents=<10 10 10>
      constraintSetup {angularLimitsMax=<0 5 10>, angularLimitsMin=<0 -20 -25>, angularTarget=<1 0 0>}
      localJointOffset=<0.8 0 0>
     }]
    simSpaceSettings {maxAngularAcceleration=10, maxAngularVelocity=5, simSpaceAngularAlpha=1}
   }
   n36 = Anim Dynamics - Bone: dyn_backhair_bot_E!(ComponentPose=@n35, Alpha=1, GravityOverride:Vector="0,0,-980", ExternalForce="0,0,0") -> (Pose) {
    angularDampingOverride=0.99
    angularSpringConstant=1000
    bAngularSpring=true
    bOverrideAngularDamping=true
    bUseGravityOverride=true
    bUsePlanarLimit=false
    simulationSpace=RootRelative
    alphaBoolBlend {blendOption=Linear}
    boundBone {boneName=dyn_backhair_bot_E}
    componentAppliedLinearAccClamp=<5000 5000 5000>
    componentLinearAccScale=<0.4 0.4 0.3>
    componentLinearVelScale=<0.8 0.8 0.6>
    gravityOverride {z=-980}
    physicsBodyDefinitions [{
      boundBone {boneName=dyn_backhair_bot_E}
      boxExtents=<10 10 10>
      constraintSetup {angularLimitsMax=<0 5 10>, angularLimitsMin=<0 -20 -25>, angularTarget=<1 0 0>}
      localJointOffset=<1 0 0>
     }]
    simSpaceSettings {maxAngularAcceleration=10, maxAngularVelocity=5, simSpaceAngularAlpha=1}
   }
   n37 = Anim Dynamics - Bone: dyn_backhair_bot_F!(ComponentPose=@n36, Alpha=1, GravityOverride:Vector="0,0,-980", ExternalForce="0,0,0") -> (Pose) {
    angularDampingOverride=0.99
    angularSpringConstant=1000
    bAngularSpring=true
    bOverrideAngularDamping=true
    bUseGravityOverride=true
    bUsePlanarLimit=false
    simulationSpace=RootRelative
    alphaBoolBlend {blendOption=Linear}
    boundBone {boneName=dyn_backhair_bot_F}
    componentAppliedLinearAccClamp=<5000 5000 5000>
    componentLinearAccScale=<0.4 0.4 0.3>
    componentLinearVelScale=<0.8 0.8 0.6>
    gravityOverride {z=-980}
    physicsBodyDefinitions [{
      boundBone {boneName=dyn_backhair_bot_F}
      boxExtents=<20 20 20>
      constraintSetup {angularLimitsMax=<0 5 10>, angularLimitsMin=<0 -20 -25>, angularTarget=<1 0 0>}
      localJointOffset=<0.8 0 0>
     }]
    simSpaceSettings {maxAngularAcceleration=10, maxAngularVelocity=5, simSpaceAngularAlpha=1}
   }
   n38 = Anim Dynamics - Bone: dyn_backhair_bot_G!(ComponentPose=@n37, Alpha=1, GravityOverride:Vector="0,0,-980", ExternalForce="0,0,0") -> (Pose) {
    angularDampingOverride=0.99
    angularSpringConstant=1000
    bAngularSpring=true
    bOverrideAngularDamping=true
    bUseGravityOverride=true
    bUsePlanarLimit=false
    simulationSpace=RootRelative
    alphaBoolBlend {blendOption=Linear}
    boundBone {boneName=dyn_backhair_bot_G}
    componentAppliedLinearAccClamp=<5000 5000 5000>
    componentLinearAccScale=<0.4 0.4 0.3>
    componentLinearVelScale=<0.8 0.8 0.6>
    gravityOverride {z=-980}
    physicsBodyDefinitions [{
      boundBone {boneName=dyn_backhair_bot_G}
      boxExtents=<10 10 10>
      constraintSetup {angularLimitsMax=<0 5 10>, angularLimitsMin=<0 -20 -25>, angularTarget=<1 0 0>}
      localJointOffset=<0.8 0 0>
     }]
    simSpaceSettings {maxAngularAcceleration=10, maxAngularVelocity=5, simSpaceAngularAlpha=1}
   }
   n39 = Anim Dynamics - Bone: dyn_earring_A!(ComponentPose=@n38, Alpha=1, GravityOverride:Vector="0,0,-980", ExternalForce="0,0,0") -> (Pose) {
    angularDampingOverride=1
    angularSpringConstant=20000
    bAngularSpring=true
    bOverrideAngularDamping=true
    bUseGravityOverride=true
    gravityScale=0.35
    simulationSpace=RootRelative
    boundBone {boneName=dyn_earring_A}
    componentLinearAccScale=<1 1 1>
    componentLinearVelScale=<1 1 1>
    gravityOverride {z=-980}
    physicsBodyDefinitions [{
      boundBone {boneName=dyn_earring_A}
      boxExtents=<20 30 30>
      constraintSetup {angularLimitsMax=<0 0 5>, angularLimitsMin=<0 0 -45>, angularTarget=<1 0 0>}
      localJointOffset=<4 0 0>
     }]
    simSpaceSettings {maxAngularAcceleration=20, maxAngularVelocity=10, simSpaceAngularAlpha=1}
   }
   n40 = Anim Dynamics - Bone: dyn_earring_B!(ComponentPose=@n39, Alpha=1, GravityOverride:Vector="0,0,-980", ExternalForce="0,0,0") -> (Pose) {
    angularDampingOverride=1
    angularSpringConstant=20000
    bAngularSpring=true
    bOverrideAngularDamping=true
    bUseGravityOverride=true
    gravityScale=0.35
    simulationSpace=RootRelative
    boundBone {boneName=dyn_earring_B}
    componentLinearAccScale=<1 1 1>
    componentLinearVelScale=<1 1 1>
    gravityOverride {z=-980}
    physicsBodyDefinitions [{
      boundBone {boneName=dyn_earring_B}
      boxExtents=<20 30 30>
      constraintSetup {angularLimitsMax=<0 0 25>, angularLimitsMin=<0 0 -25>, angularTarget=<1 0 0>}
      localJointOffset=<4 0 0>
     }]
    simSpaceSettings {maxAngularAcceleration=20, maxAngularVelocity=10, simSpaceAngularAlpha=1}
   }
   Save cached pose 'CRCache_dynamics'(Pose=Component To Local(ComponentPose=@n40)) {cacheName=CRCache_dynamics}
   n41 = RigidBody!(ComponentPose=Local To Component(LocalPose=Use cached pose 'CRCache_InputPose'()), Alpha, OverrideWorldGravity:Vector="0,0,0", ExternalForce) -> (Pose) {
    bOverrideWorldGravity=true
    overridePhysicsAsset=&/F_MED_GlamClaws_FaceAcc_Physics
    simulationSpace=BaseBoneSpace
    baseBoneRef {boneName=head}
    componentAppliedLinearAccClamp=<3000 3000 3000>
    simSpaceSettings {
     dampingAlpha=0.5
     maxAngularAcceleration=6
     maxAngularVelocity=3
     maxLinearAcceleration=450
     maxLinearVelocity=200
     velocityScaleZ=0.5
     worldAlpha=1
    }
   }
   n42 = Control Rig!(Source=Component To Local(ComponentPose=@n41), Alpha, Is Crouching:bool=false, Is Targeting=0, Is Moving 2D=false) -> (Pose) {controlRigAssetReference {blueprintRigClass=&/F_MED_GlamClaws_postdyn_FaceAcc_CtrlRig.F_MED_GlamClaws_postdyn_FaceAcc_CtrlRig_C}}
   n43 = Use cached pose 'CRCache_InputPose'()
   n44 = "Constraint - base_backhair_A (PPPP)"!(
    ComponentPose=Local To Component(LocalPose=Blend Poses by bool(BlendPose_0=@n43, BlendPose_1=@n42, bActiveValue=true))
    Alpha=1
    ConstraintWeights_0=0.6607
    ConstraintWeights_1:float=0.3379
    ConstraintWeights_2:float=0.0007
    ConstraintWeights_3:float=0.0007
   ) -> (Pose) {
    boneToModify {boneName=base_backhair_A}
    constraintSetup [
     {targetBone {boneName=head}}
     {targetBone {boneName=dyn_longhair_L_B_01}}
     {targetBone {boneName=dyn_longhair_L_B_02}}
     {targetBone {boneName=dyn_longhair_L_A_01}}
    ]
    constraintWeights [1,1,1,1]
   }
   n45 = "Constraint - base_backhair_B (PPPPP)"!(
    ComponentPose=@n44
    Alpha=1
    ConstraintWeights_0=0.543
    ConstraintWeights_1:float=0.3948
    ConstraintWeights_2:float=0.0602
    ConstraintWeights_3:float=0.0016
    ConstraintWeights_4:float=0.0004
   ) -> (Pose) {
    boneToModify {boneName=base_backhair_B}
    constraintSetup [
     {targetBone {boneName=head}}
     {targetBone {boneName=dyn_longhair_L_B_01}}
     {targetBone {boneName=dyn_longhair_L_A_01}}
     {targetBone {boneName=dyn_longhair_L_B_02}}
     {targetBone {boneName=dyn_longhair_L_A_02}}
    ]
    constraintWeights [1,1,1,1,1]
   }
   n46 = "Constraint - base_backhair_C (PPPP)"!(
    ComponentPose=@n45
    Alpha=1
    ConstraintWeights_0=0.5722
    ConstraintWeights_1:float=0.2248
    ConstraintWeights_2:float=0.2016
    ConstraintWeights_3:float=0.0014
   ) -> (Pose) {
    boneToModify {boneName=base_backhair_C}
    constraintSetup [
     {targetBone {boneName=head}}
     {targetBone {boneName=dyn_longhair_C_01}}
     {targetBone {boneName=dyn_longhair_L_A_01}}
     {targetBone {boneName=dyn_longhair_C_02}}
    ]
    constraintWeights [1,1,1,1]
   }
   n47 = "Constraint - base_backhair_D (PPPPP)"!(
    ComponentPose=@n46
    Alpha=1
    ConstraintWeights_0=0.5749
    ConstraintWeights_1:float=0.2863
    ConstraintWeights_2:float=0.1363
    ConstraintWeights_3:float=0.0023
    ConstraintWeights_4:float=0.0002
   ) -> (Pose) {
    boneToModify {boneName=base_backhair_D}
    constraintSetup [
     {targetBone {boneName=head}}
     {targetBone {boneName=dyn_longhair_R_A_01}}
     {targetBone {boneName=dyn_longhair_C_01}}
     {targetBone {boneName=dyn_longhair_C_02}}
     {targetBone {boneName=dyn_longhair_R_A_02}}
    ]
    constraintWeights [1,1,1,1,1]
   }
   n48 = "Constraint - base_backhair_E (PPPPP)"!(
    ComponentPose=@n47
    Alpha=1
    ConstraintWeights_0=0.6775
    ConstraintWeights_1:float=0.2972
    ConstraintWeights_2:float=0.025
    ConstraintWeights_3:float=0.0001
    ConstraintWeights_4:float=0.0001
   ) -> (Pose) {
    boneToModify {boneName=base_backhair_E}
    constraintSetup [
     {targetBone {boneName=head}}
     {targetBone {boneName=dyn_longhair_R_A_01}}
     {targetBone {boneName=dyn_longhair_C_01}}
     {targetBone {boneName=dyn_longhair_R_A_02}}
     {targetBone {boneName=dyn_longhair_R_B_01}}
    ]
    constraintWeights [1,1,1,1,1]
   }
   n49 = "Constraint - base_backhair_F (PPPPP)"!(
    ComponentPose=@n48
    Alpha=1
    ConstraintWeights_0=0.5364
    ConstraintWeights_1:float=0.4223
    ConstraintWeights_2:float=0.0307
    ConstraintWeights_3:float=0.0093
    ConstraintWeights_4:float=0.0012
   ) -> (Pose) {
    boneToModify {boneName=base_backhair_F}
    constraintSetup [
     {targetBone {boneName=head}}
     {targetBone {boneName=dyn_longhair_R_A_01}}
     {targetBone {boneName=dyn_longhair_R_B_01}}
     {targetBone {boneName=dyn_longhair_R_B_02}}
     {targetBone {boneName=dyn_longhair_R_A_02}}
    ]
    constraintWeights [1,1,1,1,1]
   }
   n50 = "Constraint - base_backhair_middle_A (PPPPP)"!(
    ComponentPose=@n49
    Alpha=1
    ConstraintWeights_0=0.7724
    ConstraintWeights_1:float=0.1434
    ConstraintWeights_2:float=0.0823
    ConstraintWeights_3:float=0.0011
    ConstraintWeights_4:float=0.0007
   ) -> (Pose) {
    boneToModify {boneName=base_backhair_middle_A}
    constraintSetup [
     {targetBone {boneName=dyn_longhair_L_B_02}}
     {targetBone {boneName=head}}
     {targetBone {boneName=dyn_longhair_L_B_03}}
     {targetBone {boneName=dyn_longhair_L_B_01}}
     {targetBone {boneName=dyn_longhair_L_A_02}}
    ]
    constraintWeights [1,1,1,1,1]
   }
   n51 = "Constraint - base_backhair_middle_B (PPPPPP)"!(
    ComponentPose=@n50
    Alpha=1
    ConstraintWeights_0=0.8273
    ConstraintWeights_1:float=0.0841
    ConstraintWeights_2:float=0.0487
    ConstraintWeights_3:float=0.0257
    ConstraintWeights_4:float=0.014
    ConstraintWeights_5:float=0.0002
   ) -> (Pose) {
    boneToModify {boneName=base_backhair_middle_B}
    constraintSetup [
     {targetBone {boneName=dyn_longhair_L_B_03}}
     {targetBone {boneName=dyn_longhair_L_A_02}}
     {targetBone {boneName=head}}
     {targetBone {boneName=dyn_longhair_L_B_02}}
     {targetBone {boneName=dyn_longhair_L_A_03}}
     {targetBone {boneName=dyn_longhair_L_B_04}}
    ]
    constraintWeights [1,1,1,1,1,1]
   }
   n52 = "Constraint - base_backhair_middle_C (PPPPPPP)"!(
    ComponentPose=@n51
    Alpha=1
    ConstraintWeights_0=0.8487
    ConstraintWeights_1:float=0.1221
    ConstraintWeights_2:float=0.0128
    ConstraintWeights_3:float=0.0084
    ConstraintWeights_4:float=0.0054
    ConstraintWeights_5:float=0.0025
    ConstraintWeights_6:float=0.0001
   ) -> (Pose) {
    boneToModify {boneName=base_backhair_middle_C}
    constraintSetup [
     {targetBone {boneName=dyn_longhair_L_A_02}}
     {targetBone {boneName=head}}
     {targetBone {boneName=dyn_longhair_L_B_02}}
     {targetBone {boneName=dyn_longhair_L_A_01}}
     {targetBone {boneName=dyn_longhair_L_B_03}}
     {targetBone {boneName=dyn_longhair_L_A_03}}
     {targetBone {boneName=dyn_longhair_L_B_01}}
    ]
    constraintWeights [1,1,1,1,1,1,1]
   }
   n53 = "Constraint - base_backhair_middle_D (PPPPPPPP)"!(
    ComponentPose=@n52
    Alpha=1
    ConstraintWeights_0=0.7132
    ConstraintWeights_1:float=0.1404
    ConstraintWeights_2:float=0.1347
    ConstraintWeights_3:float=0.0085
    ConstraintWeights_4:float=0.0019
    ConstraintWeights_5:float=0.0007
    ConstraintWeights_6:float=0.0003
    ConstraintWeights_7:float=0.0002
   ) -> (Pose) {
    boneToModify {boneName=base_backhair_middle_D}
    constraintSetup [
     {targetBone {boneName=dyn_longhair_L_A_02}}
     {targetBone {boneName=head}}
     {targetBone {boneName=dyn_longhair_L_A_01}}
     {targetBone {boneName=dyn_longhair_C_02}}
     {targetBone {boneName=dyn_longhair_L_A_03}}
     {targetBone {boneName=dyn_longhair_C_01}}
     {targetBone {boneName=dyn_longhair_L_B_02}}
     {targetBone {boneName=dyn_longhair_C_03}}
    ]
    constraintWeights [1,1,1,1,1,1,1,1]
   }
   n54 = "Constraint - base_backhair_middle_E (PPPPPPP)"!(
    ComponentPose=@n53
    Alpha=1
    ConstraintWeights_0=0.8051
    ConstraintWeights_1:float=0.0824
    ConstraintWeights_2:float=0.066
    ConstraintWeights_3:float=0.041
    ConstraintWeights_4:float=0.0037
    ConstraintWeights_5:float=0.0017
    ConstraintWeights_6:float=0.0002
   ) -> (Pose) {
    boneToModify {boneName=base_backhair_middle_E}
    constraintSetup [
     {targetBone {boneName=dyn_longhair_C_02}}
     {targetBone {boneName=dyn_longhair_L_A_02}}
     {targetBone {boneName=head}}
     {targetBone {boneName=dyn_longhair_C_03}}
     {targetBone {boneName=dyn_longhair_L_A_03}}
     {targetBone {boneName=dyn_longhair_L_A_01}}
     {targetBone {boneName=dyn_longhair_C_01}}
    ]
    constraintWeights [1,1,1,1,1,1,1]
   }
   n55 = "Constraint - base_backhair_middle_F (PPPPPP)"!(
    ComponentPose=@n54
    Alpha=1
    ConstraintWeights_0=0.5328
    ConstraintWeights_1:float=0.381
    ConstraintWeights_2:float=0.0619
    ConstraintWeights_3:float=0.0173
    ConstraintWeights_4:float=0.0053
    ConstraintWeights_5:float=0.0017
   ) -> (Pose) {
    boneToModify {boneName=base_backhair_middle_F}
    constraintSetup [
     {targetBone {boneName=dyn_longhair_C_02}}
     {targetBone {boneName=dyn_longhair_R_A_02}}
     {targetBone {boneName=head}}
     {targetBone {boneName=dyn_longhair_C_03}}
     {targetBone {boneName=dyn_longhair_R_A_01}}
     {targetBone {boneName=dyn_longhair_R_A_03}}
    ]
    constraintWeights [1,1,1,1,1,1]
   }
   n56 = "Constraint - base_backhair_middle_G (PPPPPPP)"!(
    ComponentPose=@n55
    Alpha=1
    ConstraintWeights_0=0.6413
    ConstraintWeights_1:float=0.2145
    ConstraintWeights_2:float=0.138
    ConstraintWeights_3:float=0.0056
    ConstraintWeights_4:float=0.0003
    ConstraintWeights_5:float=0.0001
    ConstraintWeights_6:float=0.0001
   ) -> (Pose) {
    boneToModify {boneName=base_backhair_middle_G}
    constraintSetup [
     {targetBone {boneName=dyn_longhair_R_A_02}}
     {targetBone {boneName=head}}
     {targetBone {boneName=dyn_longhair_R_A_01}}
     {targetBone {boneName=dyn_longhair_R_B_02}}
     {targetBone {boneName=dyn_longhair_C_02}}
     {targetBone {boneName=dyn_longhair_R_B_01}}
     {targetBone {boneName=dyn_longhair_R_A_03}}
    ]
    constraintWeights [1,1,1,1,1,1,1]
   }
   n57 = "Constraint - base_backhair_middle_H (PPPPPP)"!(
    ComponentPose=@n56
    Alpha=1
    ConstraintWeights_0=0.8015
    ConstraintWeights_1:float=0.1717
    ConstraintWeights_2:float=0.0156
    ConstraintWeights_3:float=0.0084
    ConstraintWeights_4:float=0.0016
    ConstraintWeights_5:float=0.0012
   ) -> (Pose) {
    boneToModify {boneName=base_backhair_middle_H}
    constraintSetup [
     {targetBone {boneName=dyn_longhair_R_B_02}}
     {targetBone {boneName=head}}
     {targetBone {boneName=dyn_longhair_R_B_01}}
     {targetBone {boneName=dyn_longhair_R_B_03}}
     {targetBone {boneName=dyn_longhair_R_A_01}}
     {targetBone {boneName=dyn_longhair_R_A_02}}
    ]
    constraintWeights [1,1,1,1,1,1]
   }
   n58 = "Constraint - base_backhair_middle_I (PPPPPPPP)"!(
    ComponentPose=@n57
    Alpha=1
    ConstraintWeights_0=0.6713
    ConstraintWeights_1:float=0.1731
    ConstraintWeights_2:float=0.1291
    ConstraintWeights_3:float=0.0193
    ConstraintWeights_4:float=0.0055
    ConstraintWeights_5:float=0.0013
    ConstraintWeights_6:float=0.0002
    ConstraintWeights_7:float=0.0002
   ) -> (Pose) {
    boneToModify {boneName=base_backhair_middle_I}
    constraintSetup [
     {targetBone {boneName=dyn_longhair_L_A_03}}
     {targetBone {boneName=dyn_longhair_L_A_04}}
     {targetBone {boneName=dyn_longhair_L_B_04}}
     {targetBone {boneName=dyn_longhair_L_B_03}}
     {targetBone {boneName=head}}
     {targetBone {boneName=dyn_longhair_L_A_02}}
     {targetBone {boneName=dyn_longhair_C_04}}
     {targetBone {boneName=dyn_longhair_C_03}}
    ]
    constraintWeights [1,1,1,1,1,1,1,1]
   }
   n59 = "Constraint - base_backhair_middle_J (PPPPPPPP)"!(
    ComponentPose=@n58
    Alpha=1
    ConstraintWeights_0=0.8721
    ConstraintWeights_1:float=0.0631
    ConstraintWeights_2:float=0.0229
    ConstraintWeights_3:float=0.0198
    ConstraintWeights_4:float=0.015
    ConstraintWeights_5:float=0.0062
    ConstraintWeights_6:float=0.0006
    ConstraintWeights_7:float=0.0003
   ) -> (Pose) {
    boneToModify {boneName=base_backhair_middle_J}
    constraintSetup [
     {targetBone {boneName=dyn_longhair_L_A_03}}
     {targetBone {boneName=dyn_longhair_L_A_02}}
     {targetBone {boneName=dyn_longhair_L_A_04}}
     {targetBone {boneName=dyn_longhair_C_03}}
     {targetBone {boneName=head}}
     {targetBone {boneName=dyn_longhair_L_B_03}}
     {targetBone {boneName=dyn_longhair_L_B_04}}
     {targetBone {boneName=dyn_longhair_C_02}}
    ]
    constraintWeights [1,1,1,1,1,1,1,1]
   }
   n60 = "Constraint - base_backhair_middle_K (PPPPPPP)"!(
    ComponentPose=@n59
    Alpha=1
    ConstraintWeights_0=0.4832
    ConstraintWeights_1:float=0.3368
    ConstraintWeights_2:float=0.1078
    ConstraintWeights_3:float=0.0443
    ConstraintWeights_4:float=0.0271
    ConstraintWeights_5:float=0.0006
    ConstraintWeights_6:float=0.0002
   ) -> (Pose) {
    boneToModify {boneName=base_backhair_middle_K}
    constraintSetup [
     {targetBone {boneName=dyn_longhair_C_03}}
     {targetBone {boneName=dyn_longhair_L_A_03}}
     {targetBone {boneName=dyn_longhair_L_A_02}}
     {targetBone {boneName=dyn_longhair_C_02}}
     {targetBone {boneName=head}}
     {targetBone {boneName=dyn_longhair_L_A_04}}
     {targetBone {boneName=dyn_longhair_C_04}}
    ]
    constraintWeights [1,1,1,1,1,1,1]
   }
   n61 = "Constraint - base_backhair_middle_L (PPPPPPPP)"!(
    ComponentPose=@n60
    Alpha=1
    ConstraintWeights_0=0.7717
    ConstraintWeights_1:float=0.0888
    ConstraintWeights_2:float=0.0729
    ConstraintWeights_3:float=0.045
    ConstraintWeights_4:float=0.0119
    ConstraintWeights_5:float=0.009
    ConstraintWeights_6:float=0.0005
    ConstraintWeights_7:float=0.0002
   ) -> (Pose) {
    boneToModify {boneName=base_backhair_middle_L}
    constraintSetup [
     {targetBone {boneName=dyn_longhair_C_03}}
     {targetBone {boneName=dyn_longhair_R_A_02}}
     {targetBone {boneName=dyn_longhair_R_A_03}}
     {targetBone {boneName=dyn_longhair_C_02}}
     {targetBone {boneName=head}}
     {targetBone {boneName=dyn_longhair_C_04}}
     {targetBone {boneName=dyn_longhair_R_A_04}}
     {targetBone {boneName=dyn_longhair_L_A_03}}
    ]
    constraintWeights [1,1,1,1,1,1,1,1]
   }
   n62 = "Constraint - base_backhair_middle_M (PPPPPPP)"!(
    ComponentPose=@n61
    Alpha=1
    ConstraintWeights_0=0.7148
    ConstraintWeights_1:float=0.1452
    ConstraintWeights_2:float=0.0747
    ConstraintWeights_3:float=0.0346
    ConstraintWeights_4:float=0.0299
    ConstraintWeights_5:float=0.0004
    ConstraintWeights_6:float=0.0003
   ) -> (Pose) {
    boneToModify {boneName=base_backhair_middle_M}
    constraintSetup [
     {targetBone {boneName=dyn_longhair_R_A_02}}
     {targetBone {boneName=dyn_longhair_R_A_03}}
     {targetBone {boneName=dyn_longhair_C_03}}
     {targetBone {boneName=head}}
     {targetBone {boneName=dyn_longhair_C_02}}
     {targetBone {boneName=dyn_longhair_R_A_04}}
     {targetBone {boneName=dyn_longhair_C_04}}
    ]
    constraintWeights [1,1,1,1,1,1,1]
   }
   n63 = "Constraint - base_backhair_middle_N (PPPPPPP)"!(
    ComponentPose=@n62
    Alpha=1
    ConstraintWeights_0=0.3643
    ConstraintWeights_1:float=0.3454
    ConstraintWeights_2:float=0.2452
    ConstraintWeights_3:float=0.039
    ConstraintWeights_4:float=0.0046
    ConstraintWeights_5:float=0.0012
    ConstraintWeights_6:float=0.0005
   ) -> (Pose) {
    boneToModify {boneName=base_backhair_middle_N}
    constraintSetup [
     {targetBone {boneName=dyn_longhair_R_A_03}}
     {targetBone {boneName=dyn_longhair_R_B_03}}
     {targetBone {boneName=dyn_longhair_R_A_02}}
     {targetBone {boneName=head}}
     {targetBone {boneName=dyn_longhair_R_B_02}}
     {targetBone {boneName=dyn_longhair_R_A_04}}
     {targetBone {boneName=dyn_longhair_R_B_04}}
    ]
    constraintWeights [1,1,1,1,1,1,1]
   }
   n64 = "Constraint - base_backhair_middle_O (PPPPPPP)"!(
    ComponentPose=@n63
    Alpha=1
    ConstraintWeights_0=0.4711
    ConstraintWeights_1:float=0.3273
    ConstraintWeights_2:float=0.1638
    ConstraintWeights_3:float=0.0323
    ConstraintWeights_4:float=0.0031
    ConstraintWeights_5:float=0.0016
    ConstraintWeights_6:float=0.0008
   ) -> (Pose) {
    boneToModify {boneName=base_backhair_middle_O}
    constraintSetup [
     {targetBone {boneName=dyn_longhair_R_B_03}}
     {targetBone {boneName=dyn_longhair_R_A_03}}
     {targetBone {boneName=dyn_longhair_R_B_04}}
     {targetBone {boneName=head}}
     {targetBone {boneName=dyn_longhair_R_B_02}}
     {targetBone {boneName=dyn_longhair_R_A_02}}
     {targetBone {boneName=dyn_longhair_R_A_04}}
    ]
    constraintWeights [1,1,1,1,1,1,1]
   }
   n65 = "Constraint - base_backhair_bot_A (PPPPPP)"!(
    ComponentPose=@n64
    Alpha=1
    ConstraintWeights_0=0.4492
    ConstraintWeights_1:float=0.2704
    ConstraintWeights_2:float=0.1419
    ConstraintWeights_3:float=0.1352
    ConstraintWeights_4:float=0.003
    ConstraintWeights_5:float=0.0004
   ) -> (Pose) {
    boneToModify {boneName=base_backhair_bot_A}
    constraintSetup [
     {targetBone {boneName=dyn_longhair_L_A_04}}
     {targetBone {boneName=dyn_longhair_C_04}}
     {targetBone {boneName=dyn_longhair_L_A_03}}
     {targetBone {boneName=dyn_longhair_C_03}}
     {targetBone {boneName=head}}
     {targetBone {boneName=dyn_longhair_L_B_04}}
    ]
    constraintWeights [1,1,1,1,1,1]
   }
   n66 = "Constraint - base_backhair_bot_B (PPPPPP)"!(
    ComponentPose=@n65
    Alpha=1
    ConstraintWeights_0=0.5761
    ConstraintWeights_1:float=0.3954
    ConstraintWeights_2:float=0.0239
    ConstraintWeights_3:float=0.0026
    ConstraintWeights_4:float=0.0019
    ConstraintWeights_5:float=0.0001
   ) -> (Pose) {
    boneToModify {boneName=base_backhair_bot_B}
    constraintSetup [
     {targetBone {boneName=dyn_longhair_C_03}}
     {targetBone {boneName=dyn_longhair_C_04}}
     {targetBone {boneName=dyn_longhair_L_A_04}}
     {targetBone {boneName=head}}
     {targetBone {boneName=dyn_longhair_L_A_03}}
     {targetBone {boneName=dyn_longhair_R_A_03}}
    ]
    constraintWeights [1,1,1,1,1,1]
   }
   n67 = "Constraint - base_backhair_bot_C (PPPPPPP)"!(
    ComponentPose=@n66
    Alpha=1
    ConstraintWeights_0=0.5758
    ConstraintWeights_1:float=0.3333
    ConstraintWeights_2:float=0.0814
    ConstraintWeights_3:float=0.0074
    ConstraintWeights_4:float=0.0016
    ConstraintWeights_5:float=0.0003
    ConstraintWeights_6:float=0.0001
   ) -> (Pose) {
    boneToModify {boneName=base_backhair_bot_C}
    constraintSetup [
     {targetBone {boneName=dyn_longhair_C_04}}
     {targetBone {boneName=dyn_longhair_C_03}}
     {targetBone {boneName=dyn_longhair_R_A_03}}
     {targetBone {boneName=dyn_longhair_R_A_04}}
     {targetBone {boneName=head}}
     {targetBone {boneName=dyn_longhair_R_A_02}}
     {targetBone {boneName=dyn_longhair_L_A_04}}
    ]
    constraintWeights [1,1,1,1,1,1,1]
   }
   n68 = "Constraint - base_backhair_bot_D (PPPPPP)"!(
    ComponentPose=@n67
    Alpha=1
    ConstraintWeights_0=0.3396
    ConstraintWeights_1:float=0.3027
    ConstraintWeights_2:float=0.26
    ConstraintWeights_3:float=0.0938
    ConstraintWeights_4:float=0.0024
    ConstraintWeights_5:float=0.0015
   ) -> (Pose) {
    boneToModify {boneName=base_backhair_bot_D}
    constraintSetup [
     {targetBone {boneName=dyn_longhair_R_A_03}}
     {targetBone {boneName=dyn_longhair_C_04}}
     {targetBone {boneName=dyn_longhair_R_A_04}}
     {targetBone {boneName=dyn_longhair_C_03}}
     {targetBone {boneName=head}}
     {targetBone {boneName=dyn_longhair_R_A_02}}
    ]
    constraintWeights [1,1,1,1,1,1]
   }
   n69 = "Constraint - base_backhair_bot_E (PPPPPPP)"!(
    ComponentPose=@n68
    Alpha=1
    ConstraintWeights_0=0.8044
    ConstraintWeights_1:float=0.1038
    ConstraintWeights_2:float=0.0676
    ConstraintWeights_3:float=0.0169
    ConstraintWeights_4:float=0.0047
    ConstraintWeights_5:float=0.0023
    ConstraintWeights_6:float=0.0004
   ) -> (Pose) {
    boneToModify {boneName=base_backhair_bot_E}
    constraintSetup [
     {targetBone {boneName=dyn_longhair_R_A_03}}
     {targetBone {boneName=dyn_longhair_R_B_03}}
     {targetBone {boneName=dyn_longhair_R_A_02}}
     {targetBone {boneName=head}}
     {targetBone {boneName=dyn_longhair_R_A_04}}
     {targetBone {boneName=dyn_longhair_R_B_04}}
     {targetBone {boneName=dyn_longhair_R_B_02}}
    ]
    constraintWeights [1,1,1,1,1,1,1]
   }
   n70 = "Constraint - base_backhair_bot_F (PPPPP)"!(
    ComponentPose=@n69
    Alpha=1
    ConstraintWeights_0=0.7128
    ConstraintWeights_1:float=0.2644
    ConstraintWeights_2:float=0.0214
    ConstraintWeights_3:float=0.0013
    ConstraintWeights_4:float=0.0002
   ) -> (Pose) {
    boneToModify {boneName=base_backhair_bot_F}
    constraintSetup [
     {targetBone {boneName=dyn_longhair_L_B_04}}
     {targetBone {boneName=dyn_longhair_L_B_03}}
     {targetBone {boneName=head}}
     {targetBone {boneName=dyn_longhair_L_B_02}}
     {targetBone {boneName=dyn_longhair_L_A_03}}
    ]
    constraintWeights [1,1,1,1,1]
   }
   n71 = "Constraint - base_backhair_bot_G (PPPPP)"!(
    ComponentPose=@n70
    Alpha=1
    ConstraintWeights_0=0.9261
    ConstraintWeights_1:float=0.0379
    ConstraintWeights_2:float=0.0148
    ConstraintWeights_3:float=0.0145
    ConstraintWeights_4:float=0.0066
   ) -> (Pose) {
    boneToModify {boneName=base_backhair_bot_G}
    constraintSetup [
     {targetBone {boneName=dyn_longhair_R_B_04}}
     {targetBone {boneName=dyn_longhair_R_A_04}}
     {targetBone {boneName=dyn_longhair_R_A_03}}
     {targetBone {boneName=head}}
     {targetBone {boneName=dyn_longhair_R_B_03}}
    ]
    constraintWeights [1,1,1,1,1]
   }
   n72 = "Constraint - base_earring_B (PP)"!(ComponentPose=@n71, Alpha=1, ConstraintWeights_0=0.7952, ConstraintWeights_1:float=0.2048) -> (Pose) {
    boneToModify {boneName=base_earring_B}
    constraintSetup [{targetBone {boneName=dyn_ear_L_02}}, {targetBone {boneName=dyn_ear_L_01}}]
    constraintWeights [1,1]
   }
   Save cached pose 'CRCache_constraints'(Pose=Component To Local(ComponentPose=@n72)) {cacheName=CRCache_constraints}
  ]}
 EventGraph {nodes [
   n1 = Event BlueprintUpdateAnimation()
   > n2 = Sequence()
   > n3 = "Gravity Override State Function (Parameter Asset) (Nativized)"(
    CharacterPartAnimInstance=Self-Reference()
    GameTimeInSeconds=GetGameTimeInSeconds()
    Parameters=&/F_MED_GlamClaws_FaceAcc_Parameters
   )
   > Set GravityOverride(GravityOverride=@n3)
   n4 = CheckForBackbling()
   > n5 = Set isWearingBlackbling(isWearingBlackbling:bool=OR Boolean!(A=@n4, B=@n4.bIsBackblingACape, C:bool=@n4.bIsBackblingLong) -> (ReturnValue)) -> (Output_Get:bool)
   n6 = Get bRigidBodyEnabled()
   > Set Hair Chains Dynamics Disable(Hair Chains Dynamics Disable:bool=OR Boolean!(A=@n5, B=NOT Boolean(A=@n6)) -> (ReturnValue)) -> (Output_Get:bool)
   n7 = Get bIsTargeting()
   Set isTargeting float(
    execute=@n2.then_1
    isTargeting float:double=FInterpTo(
     Current=Get isTargeting float() -> (isTargeting float:double)
     Target="To Float (Boolean)"(InBool=@n7)
     DeltaTime=@n1.DeltaTimeX
     InterpSpeed=15
    )
   ) -> (Output_Get:double)
   n8 = Get bIsPlayingMeleeAnim()
   > Set fMeleeSwing(fMeleeSwing:double=FInterpTo(Current=Get fMeleeSwing() -> (fMeleeSwing:double), Target="To Float (Boolean)"(InBool=@n8), DeltaTime=@n1.DeltaTimeX, InterpSpeed=100)) -> (Output_Get:double)
   n9 = Event OnCharacterCustomizationLoaded()
   > Set Body_Mesh(Body_Mesh=Get Array Item with EFort Custom Part Type(Array=@n9.CustomMeshParts))
   Event OnWeaponChanged()
   > Set isAR Equipped(isAR Equipped:bool=false) -> (Output_Get:bool)
   > Set isPistol Equipped(isPistol Equipped:bool=false) -> (Output_Get:bool)
   > n10 = Switch on EFortWeaponCoreAnimation!(execute, Selection:EFortWeaponCoreAnimation Enum=Get WeaponCoreAnim!() -> (WeaponCoreAnim)) -> (
    Melee:Exec
    Pistol:Exec
    Shotgun:Exec
    PaperBlueprint:Exec
    Rifle:Exec
    MeleeOneHand:Exec
    MachinePistol:Exec
    RocketLauncher:Exec
    GrenadeLauncher:Exec
    GoingCommando:Exec
    AssaultRifle:Exec
    TacticalShotgun:Exec
    SniperRifle:Exec
    TrapPlacement:Exec
    ShoulderLauncher:Exec
    AbilityDecoTool:Exec
    Crossbow:Exec
    C4:Exec
    RemoteControl:Exec
    DualWield:Exec
    AR_BullPup:Exec
    AR_ForwardGrip:Exec
    MedPackPaddles:Exec
    SMG_P90:Exec
    AR_DrumGun:Exec
    Consumable_Small:Exec
    Consumable_Large:Exec
    Balloon:Exec
    MountedTurret:Exec
    CreativeTool:Exec
    ExplosiveBow:Exec
    AshtonIndigo:Exec
    AshtonChicago:Exec
    MeleeDualWield:Exec
    Unarmed:Exec
   )
   Set isAR Equipped(execute=[@n10.AssaultRifle, @n10.SMG_P90], isAR Equipped=true)
   Set isPistol Equipped(execute=@n10.Pistol, isPistol Equipped=true)
  ]}
]
variables [
 Body_Mesh {type=SkeletalMeshComponent}
 fMeleeSwing {type=number}
 GravityOverride {type=Vector}
 Hair Chains Dynamics Disable {type=boolean}
 isAR Equipped {type=boolean}
 isPistol Equipped {type=boolean}
 isTargeting float {type=number}
 isWearingBlackbling {type=boolean}
]
